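# Environment_PanningEmissive is a material function that adds a scrolling emissive layer to an incoming set of material attributes, for shared environment materials.
# A panner with separate X and Y speed parameters scrolls the UVs of a texture parameter, whose selectable channel (red by default) is multiplied by a color parameter.
# The result adds to the input EmissiveColor, and the black default texture and color contribute no glow until overridden.
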MaterialFunction /Game/Packages/Fortress_SharedMaterials/Base_Material_Functions/Environment_PanningEmissive
graph {nodes [
  n1 = GetMaterialAttributes(None=FunctionInput() {inputName=Material, inputType=FunctionInput_MaterialAttributes}) {attributeGetTypes [#1]}
  n2 = Panner(
   Coordinate=TextureCoordinate()
   Speed=AppendVector(
    A=ScalarParameter() {group=Panning Emissive, parameterName=Panning Speed X}
    B=ScalarParameter() {group=Panning Emissive, parameterName=Panning Speed Y}
   )
  )
  n3 = Multiply(
   A=StaticComponentMaskParameter(None=TextureSampleParameter2D(UVs=@n2) {group=Panning Emissive, parameterName=Panning Emissive, texture=/Engine/EngineResources/Black_Low.Black}.RGB) {defaultR=true, group=Panning Emissive, parameterName=Panning Emissive Channel}
   B=VectorParameter() {group=Panning Emissive, parameterName=Panning Emissive Color, defaultValue=<rgba 0 0 0 1>}.RGB
  )
  FunctionOutput(None=SetMaterialAttributes!(MaterialAttributes=@n1.MaterialAttributes, EmissiveColor=Add(A=@n1.EmissiveColor, B=@n3)) {attributeSetTypes [#1]})
 ]}
