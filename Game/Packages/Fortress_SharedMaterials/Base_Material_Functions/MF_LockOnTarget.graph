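# Material function that adds a lock-on indicator to a material's emissive, toggled by the isLockOn parameter.
# When active, it zeroes emissive in the material attributes and modulates a separately output cropped emissive mask.
# UV1 red marks the affected geometry, while UV1 green is quantized into segments and compared against remapped lock-on progress.
# Named reroutes carry that segmented progress and an optional time-based square-wave blink.
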MaterialFunction /Game/Packages/Fortress_SharedMaterials/Base_Material_Functions/MF_LockOnTarget
graph {nodes [
  Constant3Vector() {constant=<rgba 1.7228 0 3 1>}
  Constant() {r=4}
  n1 = GetMaterialAttributes(None=FunctionInput() {inputName=MA, inputType=FunctionInput_MaterialAttributes}) {attributeGetTypes [#1]}
  n2 = Constant()
  n3 = ScalarParameter() {group=LockOn, parameterName=isLockOn}
  FunctionOutput(None=BlendMaterialAttributes(
    A=@n1.MaterialAttributes
    B=SetMaterialAttributes!(MaterialAttributes=@n1.MaterialAttributes, EmissiveColor=@n2) {attributeSetTypes [#1]}
    Alpha=@n3
   ))
  n4 = FunctionInput() {inputName=Cropped Emissive mask}
  n5 = NamedRerouteUsage()
  n6 = Ceil(None=ComponentMask(None=TextureCoordinate() {coordinateIndex=1}) {r=true}) {desc=Using the geo in UV1 as mask}
  n7 = LinearInterpolate(A=@n4, B=Multiply(A=@n4, B=@n5), Alpha=@n6)
  n8 = NamedRerouteUsage()
  FunctionOutput(None=LinearInterpolate(A=@n4, B=Multiply(A=@n7, B=@n8), Alpha=@n3)) {outputName=Cropped Emissive out}
  Multiply(A=LinearInterpolate(A=@n1.EmissiveColor, B=Multiply(A=@n1.EmissiveColor, B=@n5), Alpha=@n6), B=@n8)
  n9 = ScalarParameter() {defaultValue=4, group=LockOn, parameterName=Lockon_Segments}
  n10 = MaterialFunctionCall(
   Input=MaterialFunctionCall(Input=ScalarParameter() {group=LockOn, parameterName=Lockon_Progress}, Input Low=Constant(), Input High=@n9) {materialFunction=&RemapValue_0-1}.Result
   Input Low=ScalarParameter() {parameterName=Low}
   Input High=ScalarParameter() {defaultValue=1, parameterName=High}
  ) {materialFunction=&RemapValue_0-1}
  n11 = LinearInterpolate(Alpha=ComponentMask(None=TextureCoordinate() {coordinateIndex=1}) {g=true}) {constA=0.2}
  n12 = Divide(A=Round(None=Multiply(A=@n11, B=@n9)), B=@n9) {constB=4}
  NamedRerouteDeclaration(None=Add(A=Saturate(None=@n10.Result), B=Subtract(A=@n12))) {name=Lockon_Progress, nodeColor=<rgba 0 1 0.5547 1>}
  n13 = Frac(None=Multiply(A=Time() {period=2.213}, B=ScalarParameter() {defaultValue=2, parameterName=Lockon_BlinkSpeed}))
  NamedRerouteDeclaration(None=LinearInterpolate(B=Round(None=@n13), Alpha=ScalarParameter() {group=LockOn, parameterName=Lockon_Blink}) {constA=1}) {name=Lockon_Blink, nodeColor=<rgba 1 0 0.9141 1>}
 ]}
