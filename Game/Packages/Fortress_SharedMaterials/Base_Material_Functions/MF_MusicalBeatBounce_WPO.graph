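# Experimental material function that outputs a vertical vertex offset, bouncing meshes to musical beats for the SoundBurst event.
# It reads beat, measure and time signature from MPC_Musical_Parameters, using period, phase and decay parameters to pick active beats and shape a squared falloff scaled by Bounce Scale.
# A static switch and dynamic branches on the listening flag and IsPlaying zero the offset.
# Leftover cosine and local-height remap nodes are unconnected to the output.
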MaterialFunction /Game/Packages/Fortress_SharedMaterials/Base_Material_Functions/MF_MusicalBeatBounce_WPO
graph {
 properties {bExposeToLibrary=true, description=EXPERIMENTAL effect causing vertex bounce synced to musical beat for SoundBurst Event}
 nodes [
  CollectionParameter() {collection=/Game/Athena/Events/AudioAnalysis/MPC_Musical_Parameters, parameterName=BeatLifetime}
  Fmod(A=Time(), B=Constant() {r=1})
  n1 = ScalarParameter() {parameterName=IsListenningToMusicalBeat}
  n2 = Constant3Vector()
  n3 = CollectionParameter() {collection=&MPC_Musical_Parameters, parameterName=MusicMeasure}
  n4 = ScalarParameter() {parameterName=MeasurePhaseOffset}
  n5 = Min(
   A=ScalarParameter() {defaultValue=1, parameterName=BeatPeriod}
   B=CollectionParameter() {collection=&MPC_Musical_Parameters, parameterName=TimeSignatureNumerator}
  )
  n6 = Fmod(A=CollectionParameter() {collection=&MPC_Musical_Parameters, parameterName=MeasureBeat}, B=@n5)
  n7 = Fmod(A=ScalarParameter() {parameterName=BeatPhaseOffset}, B=@n5)
  n8 = Min(A=Add(A=@n7, B=ScalarParameter() {defaultValue=1, parameterName=BeatDecay}), B=@n5)
  n9 = Constant() {r=1}
  n10 = If(A=@n6, B=@n8, "A > B"=@n9, "A == B"=@n9, "A < B"=Divide(A=Subtract(A=@n6, B=@n7), B=Subtract(A=@n8, B=@n7))) {constB=1, equalsThreshold=0}
  n11 = Constant() {r=1}
  n12 = If(
   A=Fmod(A=Add(A=@n3, B=@n4), B=ScalarParameter() {defaultValue=1, parameterName=MeasurePeriod})
   B=@n11
   "A > B"=@n11
   "A == B"=@n11
   "A < B"=If(A=@n7, B=@n6, "A > B"=Constant() {r=1}, "A == B"=@n10, "A < B"=@n10) {constB=1, equalsThreshold=0}
  ) {constB=1, equalsThreshold=0}
  n13 = OneMinus(None=Multiply(A=@n12, B=@n12))
  n14 = ScalarParameter() {defaultValue=5, parameterName=Bounce Scale}
  n15 = MaterialFunctionCall!(
   Alpha=CollectionParameter() {collection=&MPC_Musical_Parameters, parameterName=IsPlaying}
   Color Input 0=Constant3Vector()
   Color Input 1=AppendVector(A=Constant2Vector(), B=Multiply(A=@n13, B=@n14) {constB=10})
  ) {materialFunction=/Engine/Functions/Engine_MaterialFunctions02/Utility/DynamicBranch}
  FunctionOutput(None=StaticSwitchParameter(
    True=MaterialFunctionCall(Alpha=@n1, Color Input 0=@n2, Color Input 1=@n15.Result) {materialFunction=&DynamicBranch}.Result
    False=Constant3Vector()
   ) {defaultValue=true, parameterName=Use MusicalBeatBounce Animation})
  n16 = TransformPosition(None=WorldPosition() {worldPositionShaderOffset=WPT_ExcludeAllShaderOffsets}.XYZ) {transformSourceType=TRANSFORMPOSSOURCE_World, transformType=TRANSFORMPOSSOURCE_Instance}
  n17 = Constant()
  MaterialFunctionCall(
   Input=ComponentMask(None=Max(A=@n16) {constB=0}) {b=true}
   Input Low=@n17
   Input High=ComponentMask(None=MaterialFunctionCall() {materialFunction=&ObjectLocalBounds}.Local Bounds Max) {b=true}
   Target Low=@n17
   Target High=Constant() {r=10}
  ) {materialFunction=&RemapValueRange}
  n18 = Cosine(None=CollectionParameter() {collection=&MPC_Musical_Parameters, parameterName=BeatLifetime})
  n19 = Power(Base=Multiply(A=Add(A=@n18)) {constB=0.5}) {constExponent=3}
  n20 = Constant3Vector() {constant=<rgba 0 0 1 0>}
  Reroute(None=Multiply(A=Multiply(A=@n19, B=@n20), B=Constant() {r=10}))
 ]
}
