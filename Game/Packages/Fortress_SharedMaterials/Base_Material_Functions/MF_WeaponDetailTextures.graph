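# Material function that layers detail textures onto weapon material attributes, toggled by UseDetailTextures and bypassed on mobile.
# A mask texture supplies AO/curvature and two masks that select between three parameter sets controlling array slice, scale and strength.
# Overlay and normal arrays are sampled by tiled UVs or triplanar projection from pre-skinned position and normal.
# These derive overlay color, roughness, angle-corrected normals and metallic damage, with AO darkening.
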MaterialFunction /Game/Packages/Fortress_SharedMaterials/Base_Material_Functions/MF_WeaponDetailTextures
graph {nodes [
  Constant()
  n1 = FunctionInput() {inputType=FunctionInput_MaterialAttributes}
  n2 = GetMaterialAttributes(None=@n1) {attributeGetTypes [
    #1
    #2
    #3
    #4
    #5
   ]}
  n3 = LinearInterpolate(A=ScalarParameter() {defaultValue=0.846, group=DetailTextures, parameterName=DT_GLOBAL_AO_Darkening}, Alpha=NamedRerouteUsage()) {constA=1}
  n4 = MaterialFunctionCall(Base=@n2.BaseColor, Blend=NamedRerouteUsage()) {materialFunction=&Blend_Overlay}
  n5 = ScalarParameter() {defaultValue=0.344, group=DetailTextures, parameterName=DT_GLOBAL_Metallic_Brightness}
  n6 = NamedRerouteUsage()
  n7 = NamedRerouteUsage()
  n8 = Constant3Vector() {constant=<rgba 0 0 1 0>}
  n9 = NamedRerouteUsage()
  n10 = NamedRerouteUsage()
  n11 = SetMaterialAttributes!(
   MaterialAttributes=@n2.MaterialAttributes
   BaseColor=Multiply(A=@n3, B=LinearInterpolate(A=@n4.Result, B=@n5, Alpha=@n6))
   Normal=MaterialFunctionCall(BaseNormal=@n2.Normal, AdditionalNormal=LinearInterpolate(A=@n7, B=@n8, Alpha=@n9)) {materialFunction=&BlendAngleCorrectedNormals}.Result
   Roughness=LinearInterpolate(
    A=MaterialFunctionCall(Base=@n2.Roughness, Blend=@n10) {materialFunction=&Blend_Overlay}.Result
    B=ScalarParameter() {defaultValue=0.3, group=DetailTextures, parameterName=DT_GLOBAL_Metallic_Roughness}
    Alpha=@n6
   )
   Metallic=LinearInterpolate(A=@n2.Metallic, Alpha=NamedRerouteUsage())
   Specular=Multiply(A=@n3, B=@n2.Specular)
  ) {attributeSetTypes [
    #1
    #2
    #3
    #4
    #5
   ]}
  FunctionOutput(None=ShadingPathSwitch(Default=StaticSwitchParameter(True=@n11, False=@n1) {group=DetailTextures, parameterName=UseDetailTextures}, Mobile=@n1) {desc=Migrated from FeatureLevelSwitch})
  Multiply(A=VectorParameter() {parameterName=ColorTestGunMetal, defaultValue=<rgba 0.02 0.02 0.02 1>}.RGB, B=@n2.BaseColor)
  n12 = TextureSampleParameter2D() {
   group=DetailTextures
   parameterName=DetailTextureMask
   samplerType=SAMPLERTYPE_Masks
   texture=&T_BlackMask
   channelNames {a="", b=Mask3, g=AO_Curvature, r=Mask2}
  }
  NamedRerouteDeclaration(None=@n12.AO_Curvature) {name=Curvature, nodeColor=<rgba 0.0547 0 1 1>}
  NamedRerouteDeclaration(None=@n12.Mask2) {name=Mask1, nodeColor=<rgba 1 0 0.5859 1>}
  n13 = LocalPosition() {localOrigin=InstancePreSkinning}
  n14 = ShaderStageSwitch(PixelShader=VertexInterpolator(VS=@n13.XYZ).PS, VertexShader=@n13.XYZ)
  n15 = LinearInterpolate(
   A=LinearInterpolate(
    A=ScalarParameter() {defaultValue=32, group=DetailTextures Mask1, parameterName=DT1_SizeWS}
    B=ScalarParameter() {defaultValue=32, group=DetailTextures Mask2, parameterName=DT2_SizeWS}
    Alpha=NamedRerouteUsage()
   )
   B=ScalarParameter() {defaultValue=32, group=DetailTextures Mask3, parameterName=DT3_SizeWS}
   Alpha=NamedRerouteUsage()
  )
  n16 = Divide(A=NamedRerouteDeclaration(None=@n14) {name=PreSkinnedLocation, nodeColor=<rgba 0 1 0.4609 1>}, B=Multiply(A=Abs(None=@n15)) {constB=-1})
  n17 = NamedRerouteUsage()
  n18 = TextureSampleParameter2DArray!(UVs=AppendVector(A=ComponentMask(None=@n16) {b=true, r=true}, B=@n17), Apply View MipBias) {
   group=DetailTextures
   parameterName=DetailTextureArray_Overlay
   samplerType=SAMPLERTYPE_LinearColor
   texture=/Game/Packages/Fortress_SharedMaterials/Textures/DetailTextures/DetailTextureArray_Overlay
  }
  n19 = StaticBool()
  n20 = StaticSwitch(True=AppendVector(A=@n18.RGB, B=@n18.RGB), False=@n18.RGB, Value=@n19)
  n21 = TextureSampleParameter2DArray(UVs=AppendVector(A=ComponentMask(None=@n16) {b=true, g=true}, B=@n17)) {
   group=DetailTextures
   parameterName=DetailTextureArray_Overlay
   samplerType=SAMPLERTYPE_LinearColor
   texture=&DetailTextureArray_Overlay
  }
  n22 = StaticSwitch(True=AppendVector(A=@n21.RGB, B=@n21.RGB), False=@n21.RGB, Value=@n19)
  n23 = PreSkinnedNormal()
  n24 = NamedRerouteDeclaration(None=ShaderStageSwitch(PixelShader=VertexInterpolator(VS=@n23).PS, VertexShader=@n23)) {name=PreSkinnedNormal, nodeColor=<rgba 0.1953 1 0 1>}
  n25 = ComponentMask(None=@n24) {r=true}
  n26 = ScalarParameter() {defaultValue=1, group=DetailTextures, parameterName=DT_GLOBAL_WorldAlignedBlendContrast}
  n27 = ComponentMask(None=MaterialFunctionCall(In=Abs(None=@n25), Contrast=@n26) {materialFunction=&CheapContrast}.Result) {r=true}
  n28 = TextureSampleParameter2DArray(UVs=AppendVector(A=ComponentMask(None=@n16) {g=true, r=true}, B=@n17)) {
   group=DetailTextures
   parameterName=DetailTextureArray_Overlay
   samplerType=SAMPLERTYPE_LinearColor
   texture=&DetailTextureArray_Overlay
  }
  n29 = AppendVector(A=@n28.RGB, B=@n28.RGB)
  n30 = MaterialFunctionCall(In=Abs(None=ComponentMask(None=@n24) {b=true}), Contrast=@n26) {materialFunction=&CheapContrast}
  NamedRerouteDeclaration(None=LinearInterpolate(
    A=LinearInterpolate(A=@n20, B=@n22, Alpha=@n27) {desc=Float3}
    B=StaticSwitch(True=@n29, False=@n28.RGB, Value=@n19)
    Alpha=ComponentMask(None=@n30.Result) {r=true}
   )) {name=TriPlanarDetailOverlay, nodeColor=<rgba 1 0 0.0937 1>}
  n31 = LinearInterpolate(
   A=LinearInterpolate(
    A=ScalarParameter() {defaultValue=32, group=DetailTextures Mask1, parameterName=DT1_SizeWS_Normal}
    B=ScalarParameter() {defaultValue=32, group=DetailTextures Mask2, parameterName=DT2_SizeWS_Normal}
    Alpha=NamedRerouteUsage()
   )
   B=ScalarParameter() {defaultValue=32, group=DetailTextures Mask3, parameterName=DT3_SizeWS_Normal}
   Alpha=NamedRerouteUsage()
  )
  n32 = Divide(A=NamedRerouteUsage(), B=Multiply(A=Abs(None=@n31)) {constA=-1, constB=-1})
  n33 = NamedRerouteUsage()
  n34 = TextureSampleParameter2DArray(UVs=AppendVector(A=ComponentMask(None=@n32) {b=true, r=true}, B=@n33)) {
   group=DetailTextures
   parameterName=DetailTextureArray_Normal
   samplerType=SAMPLERTYPE_Normal
   texture=/Game/Packages/Fortress_SharedMaterials/Textures/DetailTextures/DetailTextureArray_Normal
  }
  n35 = NamedRerouteUsage()
  n36 = MaterialFunctionCall(Float3=MaterialFunctionCall(Normal=@n34.RGB, Flatness=@n35) {materialFunction=&FlattenNormal}.Result) {materialFunction=&BreakOutFloat3Components}
  n37 = Multiply(A=NamedRerouteUsage(), B=Constant() {r=1})
  n38 = AppendVector(A=Multiply(A=@n36.R) {constB=-1}, B=Multiply(A=@n36.B, B=ComponentMask(None=@n37) {g=true}))
  n39 = Multiply(A=@n36.G) {constB=-1}
  n40 = ComponentMask(None=@n37) {r=true}
  n41 = TextureSampleParameter2DArray(UVs=AppendVector(A=ComponentMask(None=@n32) {b=true, g=true}, B=@n33)) {
   group=DetailTextures
   parameterName=DetailTextureArray_Normal
   samplerType=SAMPLERTYPE_Normal
   texture=&DetailTextureArray_Normal
  }
  n42 = MaterialFunctionCall(Float3=MaterialFunctionCall(Normal=@n41.RGB, Flatness=@n35) {materialFunction=&FlattenNormal}.Result) {materialFunction=&BreakOutFloat3Components}
  n43 = AppendVector(A=Multiply(A=@n40, B=@n42.B), B=Multiply(A=@n42.R) {constB=-1})
  n44 = Multiply(A=@n42.G) {constB=-1}
  n45 = Abs(None=@n40)
  n46 = Constant()
  n47 = TextureSampleParameter2DArray(UVs=AppendVector(A=ComponentMask(None=@n32) {g=true, r=true}, B=@n33)) {
   group=DetailTextures
   parameterName=DetailTextureArray_Normal
   samplerType=SAMPLERTYPE_Normal
   texture=&DetailTextureArray_Normal
  }
  n48 = MaterialFunctionCall(Normal=@n47.RGB, Flatness=@n35) {materialFunction=&FlattenNormal}
  n49 = Multiply(A=@n48.Result, B=Constant3Vector() {constant=<rgba -1 -1 1 0>})
  n50 = MaterialFunctionCall(Float3=@n48.Result) {materialFunction=&BreakOutFloat3Components}
  n51 = ComponentMask(None=@n37) {b=true}
  n52 = LinearInterpolate(
   A=LinearInterpolate(
    A=AppendVector(A=@n38, B=@n39)
    B=AppendVector(A=@n43, B=@n44)
    Alpha=MaterialFunctionCall(In=@n45, Contrast=@n46) {materialFunction=&CheapContrast}.Result
   )
   B=AppendVector(A=ComponentMask(None=@n49) {g=true, r=true}, B=Multiply(A=@n50.B, B=@n51))
   Alpha=MaterialFunctionCall(In=Abs(None=@n51), Contrast=@n46) {materialFunction=&CheapContrast}.Result
  )
  NamedRerouteDeclaration(None=Transform(None=Normalize(VectorInput=@n52)) {transformSourceType=TRANSFORMSOURCE_World, transformType=TRANSFORM_Tangent}) {name=TriPlanarDetailNormal, nodeColor=<rgba 0.8984 1 0 1>}
  NamedRerouteDeclaration(None=@n12.Mask3) {name=Mask2, nodeColor=<rgba 0 0.9062 1 1>}
  NamedRerouteDeclaration(None=@n12.AO_Curvature) {name=AO, nodeColor=<rgba 0.9922 1 0 1>}
  n53 = TextureCoordinate()
  n54 = LinearInterpolate(
   A=LinearInterpolate(
    A=ScalarParameter() {defaultValue=4, group=DetailTextures Mask1, parameterName=DT1_Tiling}
    B=ScalarParameter() {defaultValue=4, group=DetailTextures Mask2, parameterName=DT2_Tiling}
    Alpha=NamedRerouteUsage()
   )
   B=ScalarParameter() {defaultValue=4, group=DetailTextures Mask3, parameterName=DT3_Tiling}
   Alpha=NamedRerouteUsage()
  )
  n55 = LinearInterpolate(
   A=LinearInterpolate(
    A=ScalarParameter() {desc="only use full Intergernumbers 0, 1, 2 ,3 ,4", group=DetailTextures Mask1, parameterName=DT1_Selecton_Overlay}
    B=ScalarParameter() {desc="only use full Intergernumbers 0, 1, 2 ,3 ,4", group=DetailTextures Mask2, parameterName=DT2_Selecton_Overlay}
    Alpha=NamedRerouteUsage()
   )
   B=ScalarParameter() {desc="only use full Intergernumbers 0, 1, 2 ,3 ,4", group=DetailTextures Mask3, parameterName=DT3_Selecton_Overlay}
   Alpha=NamedRerouteUsage()
  )
  n56 = TextureSampleParameter2DArray(UVs=AppendVector(A=Multiply(A=@n53, B=@n54), B=NamedRerouteDeclaration(None=@n55) {name=TextureSelectionOverlay, nodeColor=<rgba 0 0.5781 1 1>})) {
   group=DetailTextures
   parameterName=DetailTextureArray_Overlay
   samplerType=SAMPLERTYPE_LinearColor
   texture=&DetailTextureArray_Overlay
  }
  n57 = ComponentMask(None=StaticSwitchParameter(True=NamedRerouteUsage(), False=@n56.RGB) {group=DetailTextures, parameterName=UseTriPlanarDetailTextures}) {r=true}
  n58 = LinearInterpolate(
   A=LinearInterpolate(
    A=ScalarParameter() {defaultValue=0.3968, group=DetailTextures Mask1, parameterName=D1_MetallicDamageThreshold}
    B=ScalarParameter() {defaultValue=0.3968, group=DetailTextures Mask2, parameterName=D2_MetallicDamageThreshold}
    Alpha=NamedRerouteUsage()
   )
   B=ScalarParameter() {defaultValue=0.3968, group=DetailTextures Mask3, parameterName=D3_MetallicDamageThreshold}
   Alpha=NamedRerouteUsage()
  )
  n59 = Multiply(A=Subtract(A=@n57, B=OneMinus(None=@n58)), B=NamedRerouteUsage())
  n60 = LinearInterpolate(
   A=LinearInterpolate(
    A=ScalarParameter() {defaultValue=0.2564, group=DetailTextures Mask1, parameterName=DetailTexture_MetallicDamage_Contrast}
    B=ScalarParameter() {defaultValue=0.2564, group=DetailTextures Mask2, parameterName=DT2_MetallicDamage_Contrast}
    Alpha=NamedRerouteUsage()
   )
   B=ScalarParameter() {defaultValue=0.2564, group=DetailTextures Mask3, parameterName=DT3_MetallicDamage_Contrast}
   Alpha=NamedRerouteUsage()
  )
  NamedRerouteDeclaration(None=Saturate(None=Power(Base=@n59, Exp=@n60))) {name=Metallic_DetailMask, nodeColor=<rgba 0 1 0.4375 1>}
  n61 = ScalarParameter() {defaultValue=-1.1946, group=DetailTextures, parameterName=DT_GLOBAL_Strength_Surface}
  n62 = ScalarParameter() {defaultValue=13.1057, group=DetailTextures, parameterName=DT_GLOBAL_Strength_Edge}
  n63 = NamedRerouteUsage()
  n64 = Saturate(None=Divide(A=Subtract(A=@n63) {constB=0.5}))
  NamedRerouteDeclaration(None=Saturate(None=LinearInterpolate(A=@n61, B=@n62, Alpha=@n64))) {name=CurvatureEdges, nodeColor=<rgba 1 0.5156 0 1>}
  n65 = Multiply(A=ScalarParameter() {defaultValue=0.2068, group=DetailTextures, parameterName=DT_GLOBAL_Strength_AO}) {constB=-1}
  n66 = NamedRerouteUsage()
  NamedRerouteDeclaration(None=Saturate(None=LinearInterpolate(A=@n65, Alpha=@n66))) {name=AO_Mask, nodeColor=<rgba 1 0 0.1641 1>}
  n67 = TextureCoordinate()
  n68 = LinearInterpolate(
   A=LinearInterpolate(
    A=ScalarParameter() {defaultValue=4, group=DetailTextures Mask1, parameterName=DT1_Tiling_Normal}
    B=ScalarParameter() {defaultValue=4, group=DetailTextures Mask2, parameterName=DT2_Tiling_Normal}
    Alpha=NamedRerouteUsage()
   )
   B=ScalarParameter() {defaultValue=4, group=DetailTextures Mask3, parameterName=DT3_Tiling_Normal}
   Alpha=NamedRerouteUsage()
  )
  n69 = LinearInterpolate(
   A=LinearInterpolate(
    A=ScalarParameter() {desc="only use full Intergernumbers 0, 1, 2 ,3 ,4", group=DetailTextures Mask1, parameterName=DT1_Selecton_Normal}
    B=ScalarParameter() {desc="only use full Intergernumbers 0, 1, 2 ,3 ,4", group=DetailTextures Mask2, parameterName=DT2_Selecton_Normal}
    Alpha=NamedRerouteUsage()
   )
   B=ScalarParameter() {desc="only use full Intergernumbers 0, 1, 2 ,3 ,4", group=DetailTextures Mask3, parameterName=DT3_Selecton_Normal}
   Alpha=NamedRerouteUsage()
  )
  n70 = TextureSampleParameter2DArray(UVs=AppendVector(A=Multiply(A=@n67, B=@n68), B=NamedRerouteDeclaration(None=@n69) {name=TextureSelectionNormal, nodeColor=<rgba 0.7813 0 1 1>})) {
   group=DetailTextures
   parameterName=DetailTextureArray_Normal
   samplerType=SAMPLERTYPE_Normal
   texture=&DetailTextureArray_Normal
  }
  n71 = ScalarParameter() {group=DetailTextures Mask1, parameterName=DT1_NormalFlatten}
  n72 = ScalarParameter() {group=DetailTextures Mask2, parameterName=DT2_NormalFlatten}
  n73 = NamedRerouteUsage()
  n74 = NamedRerouteDeclaration(None=LinearInterpolate(
    A=LinearInterpolate(A=@n71, B=@n72, Alpha=@n73)
    B=ScalarParameter() {group=DetailTextures Mask3, parameterName=DT3_NormalFlatten}
    Alpha=NamedRerouteUsage()
   )) {name=NormalFlatten, nodeColor=<rgba 1 0.4219 0 1>}
  NamedRerouteDeclaration(None=StaticSwitchParameter(
    True=NamedRerouteUsage()
    False=MaterialFunctionCall(Normal=@n70.RGB, Flatness=@n74) {materialFunction=&FlattenNormal}.Result
   ) {group=DetailTextures, parameterName=UseTriPlanarDetailTextures}) {name=DetailNormalBlend, nodeColor=<rgba 0 1 0.9063 1>}
  n75 = ScalarParameter() {defaultValue=-0.136, group=DetailTextures Mask1, parameterName=DT1_Roughness_Birght}
  n76 = ScalarParameter() {defaultValue=-0.136, group=DetailTextures Mask2, parameterName=DT2_Roughness_Birght}
  n77 = NamedRerouteUsage()
  n78 = Multiply(
   A=NamedRerouteUsage()
   B=LinearInterpolate(
    A=LinearInterpolate(A=@n75, B=@n76, Alpha=@n77)
    B=ScalarParameter() {defaultValue=-0.136, group=DetailTextures Mask3, parameterName=DT3_Roughness_Birght}
    Alpha=NamedRerouteUsage()
   )
  )
  n79 = ScalarParameter() {defaultValue=0.691, group=DetailTextures Mask1, parameterName=DT1_Roughness_Dark}
  n80 = ScalarParameter() {defaultValue=0.691, group=DetailTextures Mask2, parameterName=DT2_Roughness_Dark}
  n81 = NamedRerouteUsage()
  n82 = Multiply(
   A=NamedRerouteUsage()
   B=LinearInterpolate(
    A=LinearInterpolate(A=@n79, B=@n80, Alpha=@n81)
    B=ScalarParameter() {defaultValue=0.691, group=DetailTextures Mask3, parameterName=DT3_Roughness_Dark}
    Alpha=NamedRerouteUsage()
   )
  )
  NamedRerouteDeclaration(None=Add(A=Add(A=@n78, B=@n82)) {constB=0.5}) {name=DetailRoughness_Blend, nodeColor=<rgba 1 0 0.5391 1>}
  n83 = Subtract(A=@n57) {constB=0.5}
  n84 = ScalarParameter() {defaultValue=1, group=DetailTextures Mask1, parameterName=DT1_OverlayBrighten}
  n85 = ScalarParameter() {defaultValue=1, group=DetailTextures Mask2, parameterName=DT2_OverlayBrighten}
  n86 = NamedRerouteUsage()
  n87 = Multiply(
   A=NamedRerouteDeclaration(None=Saturate(None=@n83)) {name=Bright, nodeColor=<rgba 0.3828 1 0 1>}
   B=LinearInterpolate(
    A=LinearInterpolate(A=@n84, B=@n85, Alpha=@n86)
    B=ScalarParameter() {defaultValue=1, group=DetailTextures Mask3, parameterName=DT3_OverlayBrighten}
    Alpha=NamedRerouteUsage()
   )
  )
  n88 = Saturate(None=Multiply(A=@n83) {constB=-1})
  n89 = LinearInterpolate(
   A=LinearInterpolate(
    A=ScalarParameter() {defaultValue=5, group=DetailTextures Mask1, parameterName=DT1_OverlayDarkenAO}
    B=ScalarParameter() {defaultValue=5, group=DetailTextures Mask2, parameterName=DT2_OverlayDarkenAO}
    Alpha=NamedRerouteUsage()
   )
   B=ScalarParameter() {defaultValue=5, group=DetailTextures Mask3, parameterName=DT3_OverlayDarkenAO}
   Alpha=NamedRerouteUsage()
  )
  n90 = LinearInterpolate(
   A=LinearInterpolate(
    A=ScalarParameter() {defaultValue=-2.6063, group=DetailTextures Mask1, parameterName=DT1_OverlayDarken}
    B=ScalarParameter() {defaultValue=-2.6063, group=DetailTextures Mask2, parameterName=DT2_OverlayDarken}
    Alpha=NamedRerouteUsage()
   )
   B=ScalarParameter() {defaultValue=-2.6063, group=DetailTextures Mask3, parameterName=DT3_OverlayDarken}
   Alpha=NamedRerouteUsage()
  )
  n91 = NamedRerouteUsage()
  n92 = Multiply(A=Multiply(A=NamedRerouteDeclaration(None=@n88) {name=Dark, nodeColor=<rgba 1 0 0.7031 1>}, B=LinearInterpolate(A=@n89, B=@n90, Alpha=@n91))) {constB=-1}
  NamedRerouteDeclaration(None=Add(A=Add(A=@n87, B=@n92)) {constB=0.5}) {name=DetailOverlay_Blend, nodeColor=<rgba 1 0 0.5391 1>}
  PreSkinnedNormal()
  Reroute(None=@n14) {desc=WorldPosition}
  ScalarParameter() {defaultValue=0.352, parameterName=RoughnessTestse}
  ScalarParameter() {defaultValue=0.5, parameterName=SpecTest}
  ScalarParameter() {defaultValue=1, group=DetailTextures, parameterName=DT_GLOBAL_WorldAlignedBlendContrast}
 ]}
