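# MF_WindGlobals is a material function that exposes Fortnite's global wind values to materials.
# It reads two vectors from the FortniteMaterialParameters collection: WeatherWindDirectionAndStrength, and WindDirection_RGB Speed_A, whose name indicates direction in RGB and speed in alpha.
# Each passes straight to a same-named output with no processing, giving shaders one shared access point for wind data.
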MaterialFunction /Game/Packages/Fortress_SharedMaterials/Base_Material_Functions/MF_WindGlobals
graph {nodes [
  n1 = CollectionParameter() {
   collection=&FortniteMaterialParameters
   parameterName=WeatherWindDirectionAndStrength
  }
  FunctionOutput(None=@n1) {outputName=WeatherWindDirectionAndStrength}
  n2 = CollectionParameter() {
   collection=&FortniteMaterialParameters
   parameterName=WindDirection_RGB Speed_A
  }
  FunctionOutput(None=@n2) {outputName=WindDirection_RGB Speed_A}
 ]}
