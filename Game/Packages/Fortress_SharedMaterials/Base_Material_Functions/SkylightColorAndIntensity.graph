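# SkyLightColorAndIntensity is a Fortnite material function, exposed to the library, that gives materials the sky light's color and intensity.
# It reads the SkyLightColorAndIntensity vector from the FortniteMaterialParameters collection, then masks RGB to a Color output and alpha to an Intensity output.
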MaterialFunction /Game/Packages/Fortress_SharedMaterials/Base_Material_Functions/SkylightColorAndIntensity.SkyLightColorAndIntensity
graph {
 properties {bExposeToLibrary=true, libraryCategoriesText [Misc,Fortnite]}
 nodes [
  n1 = CollectionParameter() {
   collection=&FortniteMaterialParameters
   parameterName=SkyLightColorAndIntensity
  }
  FunctionOutput(None=ComponentMask(None=@n1) {b=true, g=true, r=true}) {outputName=Color}
  FunctionOutput(None=ComponentMask(None=@n1) {a=true}) {outputName=Intensity, sortPriority=1}
 ]
}
