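# Material function that outputs a two-component tan(half field of view) for first-person rendering.
# It pushes a normalized 45-degree view-space vector through MF_FirstPersonTransforms at full blend and back to view space, with named reroutes deriving scale and FOV-correction ratios from vector components.
# The engine FOV function's output is divided by a rerouted value to give X.
# Y multiplies X by buffer-size G over itself, so it equals X, without aspect correction.
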MaterialFunction /Game/Global/MaterialFunctions/MF_FirstPerson_TanHalfFov
graph {nodes [
  n1 = NamedRerouteDeclaration(None=Divide(A=MaterialFunctionCall() {materialFunction=/Engine/Functions/Engine_MaterialFunctions02/Utility/FOV}."FOV ", B=NamedRerouteUsage())) {name=FPTanHalfFov, nodeColor=<rgba 1 0.2812 0 1>}
  n2 = MaterialFunctionCall(Float2=ViewProperty() {property=MEVP_BufferSize}.Property) {materialFunction=&BreakOutFloat2Components}
  n3 = Divide(A=@n2.G, B=@n2.G)
  FunctionOutput(None=MaterialFunctionCall(X=@n1, Y=Multiply(A=@n1, B=@n3)) {materialFunction=&MakeFloat2}.Result) {outputName="Tan(0.5 * Field Of View)"}
  n4 = NamedRerouteUsage()
  n5 = ComponentMask(None=NamedRerouteUsage()) {r=true}
  n6 = NamedRerouteUsage()
  NamedRerouteDeclaration(None=Divide(A=ComponentMask(None=@n4) {r=true}, B=Multiply(A=@n5, B=@n6)) {constB=0.707}) {name=FOVCorrectionFactor, nodeColor=<rgba 0 1 0.6016 1>}
  n7 = NamedRerouteUsage()
  n8 = NamedRerouteUsage()
  NamedRerouteDeclaration(None=Divide(A=ComponentMask(None=@n7) {b=true}, B=ComponentMask(None=@n8) {b=true})) {name=FPScale, nodeColor=<rgba 0.2891 1 0 1>}
  NamedRerouteDeclaration(None=Normalize(VectorInput=Constant3Vector() {constant=<rgba 1 0 1 0>})) {name=ViewSpace45Deg, nodeColor=<rgba 0 0.5312 1 1>}
  n9 = TransformPosition(None=NamedRerouteUsage()) {transformSourceType=TRANSFORMPOSSOURCE_View, transformType=TRANSFORMPOSSOURCE_TranslatedWorld}
  n10 = Constant() {r=1}
  NamedRerouteDeclaration(None=TransformPosition(None=MaterialFunctionCall!(CameraRelativePosition=@n9, Scale, FirstPersonLerpAlpha=@n10) {materialFunction=&/MF_FirstPersonTransforms}.CameraRealtiveFP) {transformSourceType=TRANSFORMPOSSOURCE_TranslatedWorld, transformType=TRANSFORMPOSSOURCE_View}) {name=ViewSpace45DegFP, nodeColor=<rgba 0.8281 0 1 1>}
 ]}
