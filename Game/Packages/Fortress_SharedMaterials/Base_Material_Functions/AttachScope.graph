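# AttachScope material function pins a weapon scope mesh to the camera and outputs scope-view UVs, masks and opacity.
# World Position Offset remaps pre-skinned positions into view space with a camera offset and depth-scaled legacy offset, cancelling world placement.
# A static switch adds masked dial rotation to positions and normals, which get matching reorientation.
# Projected scope corners from FortniteMaterialParameters feed TextureCropping for screen-aligned and mesh UVs.
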MaterialFunction /Game/Packages/Fortress_SharedMaterials/Base_Material_Functions/AttachScope
graph {
 properties {
  description=Attaches a static mesh to the camera and allows the user to rotate the mesh by altering the object bases and also provides a camera space offset variable.
  libraryCategoriesText [Misc, World Position Offset]
 }
 nodes [
  n1 = VectorParameter() {parameterName=CrosshairOffset, defaultValue=<rgba 0 0 0 1>}
  n2 = CollectionParameter() {collection=&FortniteMaterialParameters, parameterName=TopLeftScope}
  n3 = Constant3Vector() {constant=<rgba -1 0 0 0>}
  n4 = Constant3Vector() {constant=<rgba 0 0 1 0>}
  n5 = Constant3Vector() {constant=<rgba 0 1 0 0>}
  n6 = AppendVector(A=Constant2Vector(), B=ScalarParameter() {defaultValue=12, parameterName=WeaponSightsCameraOffset})
  n7 = MaterialFunctionCall!(
   View Space Position=MaterialFunctionCall(VectorToTransform=ComponentMask(None=@n2) {b=true, g=true, r=true}, BasisX=@n3, BasisY=@n4, BasisZ=@n5, BasisPosition=@n6) {materialFunction=&Transform3x3Matrix}.Transformed Vector
   Buffer UV Pixel Offset
  ) {materialFunction=&/ViewSpaceTransformToClipSpace}
  n8 = CollectionParameter() {collection=&FortniteMaterialParameters, parameterName=BottomRightScope}
  n9 = MaterialFunctionCall(View Space Position=MaterialFunctionCall(VectorToTransform=ComponentMask(None=@n8) {b=true, g=true, r=true}, BasisX=@n3, BasisY=@n4, BasisZ=@n5, BasisPosition=@n6) {materialFunction=&Transform3x3Matrix}.Transformed Vector) {materialFunction=&ViewSpaceTransformToClipSpace}
  n10 = MaterialFunctionCall!(
   UVs=Add(
    A=MaterialFunctionCall(X=@n1.R, Y=@n1.R) {materialFunction=&MakeFloat2}.Result
    B=MaterialFunctionCall() {materialFunction=&ScreenAlignedUVs}."X 100%, Y 100%"
   )
   TextureIn
   UpperLeftCorner=@n7.Screen aligned 0-1 UV
   LowerRightCorner=@n9.Screen aligned 0-1 UV
   ExportFloat4
  ) {materialFunction=/Engine/Functions/Engine_MaterialFunctions01/Texturing/TextureCropping}
  FunctionOutput(None=@n10.Crop UVs) {outputName=Post process uvs 0-1 uvs}
  FunctionOutput(None=@n10.Crop Mask) {outputName=Post Process UV Crop Mask}
  n11 = MaterialFunctionCall(UVs=TextureCoordinate(), UpperLeftCorner=@n7.Screen aligned 0-1 UV, LowerRightCorner=@n9.Screen aligned 0-1 UV) {materialFunction=&TextureCropping}
  FunctionOutput(None=@n11.Crop Mask) {outputName=TexCoord 0-1 Mask, sortPriority=10}
  n12 = Constant3Vector() {constant=<rgba 1 0 0 0>}
  n13 = ScalarParameter() {parameterName=ScopeRotateDialAngle}
  n14 = VectorParameter() {parameterName=offset, defaultValue=<rgba 0 0 0 1>}
  n15 = PreSkinnedPosition()
  n16 = ShaderStageSwitch(PixelShader=VertexInterpolator(VS=@n15).PS, VertexShader=@n15)
  n17 = FunctionInput() {bUsePreviewValueAsDefault=true, inputName=RotateDialMask, inputType=FunctionInput_Scalar, sortPriority=1, previewValue=<1 0 0 1>}
  n18 = Multiply(
   A=RotateAboutAxis(
    NormalizedRotationAxis=@n12
    RotationAngle=@n13
    PivotPoint=NamedRerouteDeclaration(None=@n14.RGB) {name=DialPivot, nodeColor=<rgba 1 0 0.1641 1>}
    Position=@n16
   )
   B=NamedRerouteDeclaration(None=@n17) {name=DialMask, nodeColor=<rgba 1 0 0.2812 1>}
  )
  n19 = MaterialFunctionCall(
   VectorToTransform=StaticSwitchParameter(True=Add(A=@n18, B=@n16), False=@n16) {parameterName=UseScopeRotateMeshDialAnim}
   BasisX=@n3
   BasisY=@n4
   BasisZ=@n5
   BasisPosition=@n6
  ) {materialFunction=&Transform3x3Matrix}
  n20 = Add(
   A=Multiply(
    A=VectorParameter() {parameterName=LegacyScopeOffset, defaultValue=<rgba 0 0 0 1>}.RGB
    B=ComponentMask(None=@n19.Transformed Vector) {b=true}
   )
   B=@n19.Transformed Vector
  )
  n21 = CameraPositionWS()
  n22 = WorldPosition()
  FunctionOutput(None=Add(A=Transform(None=@n20) {transformSourceType=TRANSFORMSOURCE_View}, B=Subtract(A=@n21, B=@n22.XYZ))) {outputName=World Position Offset}
  n23 = Transform(None=FunctionInput(Preview=VertexNormalWS()) {bUsePreviewValueAsDefault=true, inputName=Normal WS, sortPriority=3}) {transformSourceType=TRANSFORMSOURCE_World, transformType=TRANSFORM_Local}
  n24 = NamedRerouteUsage()
  n25 = NamedRerouteUsage()
  n26 = Constant()
  n27 = Normalize(VectorInput=@n23)
  n28 = StaticSwitchParameter(
   True=LinearInterpolate(A=@n23, B=RotateAboutAxis(NormalizedRotationAxis=@n24, RotationAngle=@n25, PivotPoint=@n26, Position=@n27), Alpha=NamedRerouteUsage())
   False=@n23
  ) {parameterName=UseScopeRotateMeshDialAnim}
  FunctionOutput(None=Transform(None=MaterialFunctionCall(VectorToTransform=@n28, BasisX=@n3, BasisY=@n4, BasisZ=@n5) {materialFunction=&Transform3x3Matrix}.Transformed Vector) {transformSourceType=TRANSFORMSOURCE_View}) {outputName=World Normal}
  FunctionOutput(None=@n11.Crop UVs) {outputName=TexCoord 0-1, sortPriority=50}
  FunctionOutput(None=ScalarParameter() {defaultValue=1, parameterName=BlendAmount}) {outputName=Opacity, sortPriority=10}
  MaterialFunctionCall!(
   Vector1=Constant3Vector() {constant=<rgba -1 0 0 0>}
   Vector2=Constant3Vector() {constant=<rgba 0 0 1 0>}
   Vector3=Constant3Vector() {constant=<rgba 0 1 0 0>}
  ) {materialFunction=/Engine/Functions/Engine_MaterialFunctions02/Math/MakeVectorsOrthogonal}
  Multiply(A=@n25) {constB=-1}
  NamedRerouteDeclaration(None=@n12) {name=DialrotationAxis, nodeColor=<rgba 0.0547 1 0 1>}
  NamedRerouteDeclaration(None=@n13) {name=DialAngle, nodeColor=<rgba 0 0.5312 1 1>}
  NamedRerouteUsage()
  TransformPosition(None=WorldPosition().XYZ) {transformSourceType=TRANSFORMPOSSOURCE_World}
 ]
}
