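# Material function, exposed to the library, that returns white where a model lies within the cascaded shadow map casting region.
# It derives depth from world position and compares it to the collection's ShadowDistance, scaled by the fadeout fraction.
# The clamped result fades from white to black just short of the shadow distance.
# The AreDistanceFieldsEnabled? parameter is added before a final clamp, so enabled distance fields force the mask white.
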MaterialFunction /Game/Packages/Fortress_SharedMaterials/Base_Material_Functions/CascadingShadowMapMask
graph {
 properties {bExposeToLibrary=true, description=Returns white if the model is within the casting cascading shadow region.}
 nodes [
  n1 = CollectionParameter() {collection=&FortniteMaterialParameters, parameterName=ShadowDistance}
  n2 = CollectionParameter() {
   collection=&FortniteMaterialParameters
   parameterName=ShadowDistanceFadeoutFraction
  }
  n3 = Divide(
   A=Subtract(
    A=MaterialFunctionCall!(World Position) {materialFunction=/Engine/Functions/Engine_MaterialFunctions02/Utility/DepthFromWorldPosition}.Result
    B=@n1
   )
   B=Multiply(A=@n1, B=@n2)
  )
  n4 = Clamp(None=Subtract(A=Multiply(A=@n3) {constB=-1}))
  n5 = CollectionParameter() {
   collection=&FortniteMaterialParameters
   parameterName=AreDistanceFieldsEnabled?
  }
  FunctionOutput(None=Clamp(None=Add(A=@n4, B=@n5))) {description=Returns white if the model is casting shadows., outputName=Cascading Shadow Mask}
 ]
}
