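# Material function that adjusts an incoming world position offset for characters flagged by the IsSmallerCharacter scalar parameter.
# A BlendDistance input drives a zero-to-one lerp, which multiplies the world-space vertex normal and then the incoming offset component-wise.
# IsSmallerCharacter blends between the untouched offset and this modified one.
# The blend value is also exposed as a second Distance Blend output.
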MaterialFunction /Game/Packages/Fortress_SharedMaterials/Base_Material_Functions/CharacterSizeBalanceAdjust
graph {nodes [
  n1 = FunctionInput() {inputName=WPO}
  n2 = LinearInterpolate(B=Constant() {desc=1, r=1}, Alpha=FunctionInput() {bUsePreviewValueAsDefault=true, inputName=BlendDistance, inputType=FunctionInput_Scalar}) {constB=0}
  n3 = Multiply(A=@n2, B=VertexNormalWS())
  FunctionOutput(None=LinearInterpolate(A=@n1, B=Multiply(A=@n1, B=@n3), Alpha=ScalarParameter() {group=WorldPositionOffset, parameterName=IsSmallerCharacter}))
  FunctionOutput(None=@n2) {outputName=Distance Blend}
 ]}
