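# Material function that applies a camera-distance-driven size balance adjustment to character materials.
# An engine Distance_Blend with a constant range and negative start offset feeds CharacterSizeBalanceAdjust along with the incoming World Position Offset.
# The result replaces that offset in the passed-through material attributes, and the blend factor is exposed as a second output.
# Despite the name, no brightness change happens inside this function.
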MaterialFunction /Game/Packages/Fortress_SharedMaterials/Base_Material_Functions/CharacterBrightnessandScale_Balance
graph {nodes [
  n1 = GetMaterialAttributes(None=FunctionInput() {inputName=Material, inputType=FunctionInput_MaterialAttributes}) {attributeGetTypes [#1]}
  n2 = MaterialFunctionCall!(Blend Range=Constant() {desc=4500, r=3000}, Start Offset=Constant() {r=-500}) {materialFunction=/Engine/Functions/Engine_MaterialFunctions03/Particles/Distance_Blend}
  FunctionOutput(None=SetMaterialAttributes!(
    MaterialAttributes=@n1.MaterialAttributes
    WorldPositionOffset=MaterialFunctionCall!(BlendDistance=@n2.Result, WPO=@n1.WorldPositionOffset) {materialFunction=&/CharacterSizeBalanceAdjust}.Result
   ) {attributeSetTypes [#1]})
  FunctionOutput(None=@n2.Result) {outputName=Distance Blend, sortPriority=2}
 ]}
