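# CharacterDeRez_Transition is a material function that shifts a character's material toward the blue de-rez look.
# It feeds the input material attributes and a rim term, defaulting to a pixel-normal Fresnel, into MF_CharacterDerez_BaseBlue.
# It blends from the original attributes to that result using the NewDissolveGradient parameter, doubled and clamped, so the blue look fully applies halfway through the parameter's zero-to-one range.
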MaterialFunction /Game/Packages/Fortress_SharedMaterials/Base_Material_Functions/CharacterDeRez_Transition
graph {nodes [
  n1 = FunctionInput() {inputName=Material, inputType=FunctionInput_MaterialAttributes}
  n2 = FunctionInput(Preview=Fresnel(Normal=PixelNormalWS()) {baseReflectFraction=0, exponent=1}) {bUsePreviewValueAsDefault=true, inputName=Fresnel, inputType=FunctionInput_Scalar}
  n3 = Multiply(A=ScalarParameter() {group=DeRez, parameterName=NewDissolveGradient}) {constB=2}
  FunctionOutput(None=BlendMaterialAttributes(
    A=@n1
    B=MaterialFunctionCall!(BaseMaterial=@n1, Rim=@n2) {materialFunction=&/MF_CharacterDerez_BaseBlue}.Result
    Alpha=Clamp(None=@n3)
   ))
 ]}
