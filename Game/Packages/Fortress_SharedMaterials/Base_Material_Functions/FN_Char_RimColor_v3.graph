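# Material function for character rim lighting that tints a rim mask with the sun's color.
# It reads SunLightColor from the FortniteMaterialParameters collection, keeps RGB and saturates it to the 0–1 range.
# That color is scaled by the Rim Light Overall Boost scalar parameter and multiplied by the Rim Mask input, which previews at 1.
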MaterialFunction /Game/Packages/Fortress_SharedMaterials/Base_Material_Functions/FN_Char_RimColor_v3
graph {nodes [
  n1 = Constant() {r=1}
  n2 = CollectionParameter() {
   collection=&FortniteMaterialParameters
   desc=aka 'Directional Light Color'
   parameterName=SunLightColor
  }
  n3 = Saturate(None=ComponentMask(None=@n2) {b=true, g=true, r=true})
  n4 = ScalarParameter() {desc=Rim Light Overall Boost, group=Rim, parameterName=Rim Light Overall Boost}
  FunctionOutput(None=Multiply(
    A=FunctionInput(Preview=@n1) {bUsePreviewValueAsDefault=true, inputName=Rim Mask, inputType=FunctionInput_Scalar, sortPriority=4}
    B=Multiply(A=@n3, B=@n4)
   ))
 ]}
