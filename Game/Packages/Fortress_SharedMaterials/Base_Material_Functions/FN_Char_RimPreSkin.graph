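# Material function that outputs a single-channel character rim mask built from pre-skinned normals and position, so it is independent of animation pose.
# Normal Y and Z terms form the directional mask, which is remapped by radius and opacity constants and saturated.
# A vertical gradient of local Z over object radius fades that mask by height.
# The math runs in the vertex shader and passes through a VertexInterpolator, keeping only red.
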MaterialFunction /Game/Packages/Fortress_SharedMaterials/Base_Material_Functions/FN_Char_RimPreSkin
graph {nodes [
  n1 = Constant() {desc="Rim Mask Radius (Sharpens)", r=0.1}
  n2 = Constant() {desc="Rim Mask Opacity (Softens)", r=0.2}
  n3 = PreSkinnedNormal() {desc="outputs where the mesh normals\r\nare facing, but in vertex space"}
  n4 = ComponentMask(None=@n3) {desc=restricts to y axis, g=true}
  n5 = Add(
   A=ComponentMask(None=@n3) {b=true, desc=restricts to z axis}
   B=Constant() {desc=adjust angle at which forward facing normals output 0, r=0.4}
  )
  n6 = Constant() {r=10}
  n7 = Multiply(
   A=OneMinus(None=Abs(None=@n4) {desc="forces all negative values to be positive;\r\nresults in white values being mirrored to back half of mesh"})
   B=Saturate(None=Multiply(A=@n5, B=@n6))
  )
  n8 = Constant() {desc=Rim Mask Vertical Gradient Balance, r=-0.85}
  n9 = Constant() {desc=Rim Mask Vertical Gradient Contrast, r=5}
  n10 = Divide(
   A=MaterialFunctionCall(Float3=PreSkinnedPosition()) {desc=chooses gradient direction, materialFunction=&BreakOutFloat3Components}.B
   B=ObjectRadius()
  ) {desc=sets gradient scale based on mesh size}
  n11 = LinearInterpolate(
   B=Saturate(None=LinearInterpolate(A=@n1, B=@n2, Alpha=@n7)) {desc=necessary to keep rim mask values even across entire mesh}
   Alpha=Saturate(None=LinearInterpolate(A=@n8, B=@n9, Alpha=@n10)) {desc=necessary to keep upper portions of gradient from blowing out}
  )
  FunctionOutput(None=ComponentMask(None=VertexInterpolator(VS=@n11).PS) {desc="only outputs the first channel from vertex shader; also fixes red tint bug", r=true})
 ]}
