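# Material function that builds a character rim mask, going by its name, from Fresnel, a pre-skin rim mask and AO.
# Fresnel, overridable by input, blends from a small boost constant toward a near/far parameter pair selected by DistanceBlend.
# The near parameter's note says negative values add normal map influence and positive ones soften.
# The saturated blend is multiplied by FN_Char_RimPreSkin's result, then by AO, which defaults to one.
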MaterialFunction /Game/Packages/Fortress_SharedMaterials/Base_Material_Functions/FN_Char_RimMask
graph {nodes [
  n1 = Constant() {r=1}
  n2 = ScalarParameter() {defaultValue=-0.2, desc="negative numbers mean more normal map influence\r\npositive numbers are softer", parameterName=hjklhkj}
  n3 = ScalarParameter() {defaultValue=0.04, parameterName=Far}
  n4 = FunctionInput() {bUsePreviewValueAsDefault=true, inputName=DistanceBlend, inputType=FunctionInput_Scalar}
  n5 = Fresnel()
  n6 = Saturate(None=LinearInterpolate(
    A=Constant() {desc=Rim Light Fresnel Boost, r=0.1}
    B=LinearInterpolate(A=@n2, B=@n3, Alpha=@n4)
    Alpha=FunctionInput(Preview=@n5) {bUsePreviewValueAsDefault=true, inputName=Fresnel, inputType=FunctionInput_Scalar}
   ))
  n7 = MaterialFunctionCall() {materialFunction=&/FN_Char_RimPreSkin}
  FunctionOutput(None=Multiply(
    A=FunctionInput(Preview=@n1) {bUsePreviewValueAsDefault=true, inputName=AO}
    B=Multiply(A=@n6, B=@n7.Result) {desc=multiply fresnel with rim mask}
   ) {desc=multiply by AO})
 ]}
