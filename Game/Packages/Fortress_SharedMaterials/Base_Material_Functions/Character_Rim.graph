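# Material function that computes character rim lighting, scaled by a Rim Intensity parameter and the bEnableCharacterRimLight value from FortniteMaterialParameters.
# FN_Char_RimMask builds a mask from AO, Fresnel and BalanceDistance inputs.
# A quality switch colors the mask through FN_Char_RimColor_v3, or a flat mobile color on Low.
# BalanceDistance then blends this toward a brighter flat distance rim; the Diffuse Texture input is unconnected.
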MaterialFunction /Game/Packages/Fortress_SharedMaterials/Base_Material_Functions/Character_Rim
graph {nodes [
  FunctionInput(Preview=Constant() {r=1}) {bUsePreviewValueAsDefault=true, inputName=Diffuse Texture, sortPriority=2}
  n1 = FunctionInput() {bUsePreviewValueAsDefault=true, inputName=BalanceDistance, inputType=FunctionInput_Scalar}
  n2 = MaterialFunctionCall!(
   AO=FunctionInput(Preview=Constant() {r=1}) {bUsePreviewValueAsDefault=true, inputName=AO Texture, sortPriority=1}
   DistanceBlend=@n1
   Fresnel=FunctionInput(Preview=Fresnel()) {bUsePreviewValueAsDefault=true, inputName=Fresnel N dot C, inputType=FunctionInput_Scalar, sortPriority=3}
  ) {materialFunction=&/FN_Char_RimMask}
  n3 = Constant3Vector() {desc=Mobile Rim Light Color, constant=<rgba 2 2 2 1>}
  n4 = LinearInterpolate(
   A=QualitySwitch(
    Default=MaterialFunctionCall!(Rim Mask=@n2.Result) {materialFunction=&/FN_Char_RimColor_v3}.Result
    Low=Multiply(A=@n2.Result, B=@n3)
   )
   B=Multiply(A=@n2.Result, B=Constant3Vector() {desc=Distance Rim Color, constant=<rgba 15 15 15 1>})
   Alpha=@n1
  )
  n5 = ScalarParameter() {defaultValue=1, group=Rim, parameterName=Rim Intensity}
  n6 = CollectionParameter() {
   collection=&FortniteMaterialParameters
   parameterName=bEnableCharacterRimLight
  }
  FunctionOutput(None=Multiply(A=Multiply(A=@n4, B=@n5), B=@n6))
 ]}
