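# Deprecated Fortnite base-shader material function that blends base color, specular mask and normal texture sets and adds hit-effect vertex deformation.
# Fortnite_TextureSetLerp, gated by static switches, outputs the surface channels; its diffuse also feeds FresnelFunction with Emissive.
# Fortnite_HitEffects, driven by destruction amount and custom primitive data parameters, adds hit-and-shrink offset to caller-supplied World Position Offset.
# Specular mask and normal inputs 7 and 8 are unconnected.
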MaterialFunction /Game/Packages/Fortress_SharedMaterials/Base_Material_Functions/DEPRECATED_BaseShaderFunction
graph {
 properties {libraryCategoriesText [Misc,Fortnite]}
 nodes [
  FunctionInput() {
   bUsePreviewValueAsDefault=true
   description=UI seperator
   inputName=-----------Base Shader-----------
   inputType=FunctionInput_StaticBool
   sortPriority=-1
   previewValue=<1 1 1 1>
  }
  FunctionInput() {
   bUsePreviewValueAsDefault=true
   description=UI seperator
   inputName=-----------Destruction------------
   inputType=FunctionInput_StaticBool
   sortPriority=24
  }
  FunctionInput() {bUsePreviewValueAsDefault=true, inputName=" - - - - - - - -", inputType=FunctionInput_StaticBool, sortPriority=60}
  FunctionInput() {bUsePreviewValueAsDefault=true, inputName=SpecularMasks7, inputType=FunctionInput_Vector4, sortPriority=66, previewValue=<1 0 0.2667 1>}
  FunctionInput() {bUsePreviewValueAsDefault=true, inputName=SpecularMasks8, inputType=FunctionInput_Vector4, sortPriority=67, previewValue=<1 0 0.2667 1>}
  FunctionInput() {
   bUsePreviewValueAsDefault=true
   description=UI seperator
   inputName=------Multiple Diffuse Textures-----
   inputType=FunctionInput_StaticBool
   sortPriority=50
   previewValue=<1 1 1 1>
  }
  FunctionInput() {bUsePreviewValueAsDefault=true, inputName=Normal7, inputType=FunctionInput_Vector4, sortPriority=78, previewValue=<0 0 1 1>}
  FunctionInput() {bUsePreviewValueAsDefault=true, inputName=Normal8, inputType=FunctionInput_Vector4, sortPriority=79, previewValue=<0 0 1 1>}
  FunctionInput() {bUsePreviewValueAsDefault=true, inputName=" - - - - - - - - ", inputType=FunctionInput_StaticBool, sortPriority=71}
  FunctionInput() {bUsePreviewValueAsDefault=true, inputName=" - - - - - - - - 1", inputType=FunctionInput_StaticBool, sortPriority=80}
  n1 = FunctionInput() {bUsePreviewValueAsDefault=true, inputName=SpecularMasks1, sortPriority=2, previewValue=<1 0 0.2667 1>}
  n2 = FunctionInput() {bUsePreviewValueAsDefault=true, inputName=Base Color Alpha, inputType=FunctionInput_Scalar, sortPriority=1, previewValue=<0.5 0.5 0.5 1>}
  n3 = FunctionInput() {bUsePreviewValueAsDefault=true, desc=Normals, inputName=Normals1, sortPriority=5, previewValue=<0 0 1 1>}
  n4 = FunctionInput() {bUsePreviewValueAsDefault=true, inputName=normal2, inputType=FunctionInput_Vector4, sortPriority=73, previewValue=<0 0 1 1>}
  n5 = FunctionInput() {bUsePreviewValueAsDefault=true, inputName=SpecularMasks2, inputType=FunctionInput_Vector4, sortPriority=61, previewValue=<1 0 0.2667 1>}
  n6 = FunctionInput() {bUsePreviewValueAsDefault=true, inputName=SpecularMasks3, inputType=FunctionInput_Vector4, sortPriority=62, previewValue=<1 0 0.2667 1>}
  n7 = FunctionInput() {bUsePreviewValueAsDefault=true, inputName=Normal3, inputType=FunctionInput_Vector4, sortPriority=74, previewValue=<0 0 1 1>}
  n8 = FunctionInput() {bUsePreviewValueAsDefault=true, inputName=SpecularMasks4, inputType=FunctionInput_Vector4, sortPriority=63, previewValue=<1 0 0.2667 1>}
  n9 = FunctionInput() {bUsePreviewValueAsDefault=true, inputName=Normal4, inputType=FunctionInput_Vector4, sortPriority=75, previewValue=<0 0 1 1>}
  n10 = FunctionInput() {bUsePreviewValueAsDefault=true, inputName=Normal5, inputType=FunctionInput_Vector4, sortPriority=76, previewValue=<0 0 1 1>}
  n11 = FunctionInput() {bUsePreviewValueAsDefault=true, inputName=SpecularMasks5, inputType=FunctionInput_Vector4, sortPriority=64, previewValue=<1 0 0.2667 1>}
  n12 = FunctionInput() {bUsePreviewValueAsDefault=true, inputName=SpecularMasks6, inputType=FunctionInput_Vector4, sortPriority=65, previewValue=<1 0 0.2667 1>}
  n13 = FunctionInput() {bUsePreviewValueAsDefault=true, inputName=Normal6, inputType=FunctionInput_Vector4, sortPriority=77, previewValue=<0 0 1 1>}
  n14 = StaticBool()
  n15 = MaterialFunctionCall!(
   Spec1=@n1
   Texture1=AppendVector(A=FunctionInput() {bUsePreviewValueAsDefault=true, inputName=BaseColor1, previewValue=<0.5 0.5 0.5 1>}, B=@n2)
   Normal1=@n3
   Normal2=@n4
   Texture2=FunctionInput() {bUsePreviewValueAsDefault=true, inputName=BaseColor2, inputType=FunctionInput_Vector4, sortPriority=51}
   Spec2=@n5
   Texture3=FunctionInput() {bUsePreviewValueAsDefault=true, inputName=BaseColor3, inputType=FunctionInput_Vector4, sortPriority=52}
   Spec3=@n6
   Normal3=@n7
   Texture4=FunctionInput() {bUsePreviewValueAsDefault=true, inputName=BaseColor4, inputType=FunctionInput_Vector4, sortPriority=53}
   Spec4=@n8
   Normal4=@n9
   Texture5=FunctionInput() {bUsePreviewValueAsDefault=true, inputName=BaseColor5, inputType=FunctionInput_Vector4, sortPriority=54}
   Normal5=@n10
   Spec5=@n11
   Texture6=FunctionInput() {bUsePreviewValueAsDefault=true, inputName=BaseColor6, inputType=FunctionInput_Vector4, sortPriority=55}
   Spec6=@n12
   Normal6=@n13
   Texture7=FunctionInput() {bUsePreviewValueAsDefault=true, inputName=BaseColor7, inputType=FunctionInput_Vector4, sortPriority=56}
   Texture8=FunctionInput() {bUsePreviewValueAsDefault=true, inputName=BaseColor8, inputType=FunctionInput_Vector4, sortPriority=57}
   Texture9
   Use_Dot_Texture9
   Texture10
   Use 2 Diffuse Textures=FunctionInput(Preview=@n14) {bUsePreviewValueAsDefault=true, inputName=Use 2 Textures, inputType=FunctionInput_StaticBool, sortPriority=90}
   --------------
   Use 3 Diffuse Textures=FunctionInput(Preview=@n14) {bUsePreviewValueAsDefault=true, inputName=Use 3 Textures, inputType=FunctionInput_StaticBool, sortPriority=91}
   Use 4 Diffuse Textures=FunctionInput(Preview=@n14) {bUsePreviewValueAsDefault=true, inputName=Use 4 Textures, inputType=FunctionInput_StaticBool, sortPriority=92}
   Use 5 Diffuse Textures=FunctionInput(Preview=@n14) {bUsePreviewValueAsDefault=true, inputName=Use 5 Textures, inputType=FunctionInput_StaticBool, sortPriority=93}
   Use 6 Diffuse Textures=FunctionInput(Preview=@n14) {bUsePreviewValueAsDefault=true, inputName=Use 6 Textures, inputType=FunctionInput_StaticBool, sortPriority=94}
   Use 7 Diffuse Textures=FunctionInput(Preview=@n14) {bUsePreviewValueAsDefault=true, inputName=Use 7 Textures, inputType=FunctionInput_StaticBool, sortPriority=95}
   Use 8 Diffuse Textures=FunctionInput(Preview=@n14) {bUsePreviewValueAsDefault=true, inputName=Use 8 Textures, inputType=FunctionInput_StaticBool, sortPriority=96}
   Use 9 Diffuse Textures=FunctionInput(Preview=@n14) {bUsePreviewValueAsDefault=true, inputName=Use 9 Textures, inputType=FunctionInput_StaticBool, sortPriority=97}
   Use 10 Diffuse Textures=FunctionInput(Preview=@n14) {bUsePreviewValueAsDefault=true, inputName=Use 10 Textures, inputType=FunctionInput_StaticBool, sortPriority=98}
   "-------------- "
   Dot_Texture1
   TransitionInfo
   Dot_Texture2
   Dot_Texture3
   Dot_Texture4
   Dot_Texture5
   Dot_Texture6
   Dot_Texture7
   Dot_Texture8
   Dot_Texture9
   Dot_Texture10
   Use_Dot_Texture1
   Use_Dot_Texture2
   Use_Dot_Texture3
   Use_Dot_Texture4
   Use_Dot_Texture5
   Use_Dot_Texture6
   Use_Dot_Texture7
   Use_Dot_Texture8
   Use_Dot_Texture10
  ) {materialFunction=&/Fortnite_TextureSetLerp}
  FunctionOutput(None=ComponentMask(None=@n15.Diffuse) {a=true}) {description=Alpha channel from the various diffuse maps, outputName=Diffuse Alpha, sortPriority=8}
  n16 = FunctionInput() {bUsePreviewValueAsDefault=true, inputName=DiffuseFresnelAmount, inputType=FunctionInput_Scalar, sortPriority=7, previewValue=<0.75 0 0 1>}
  n17 = MaterialFunctionCall!(
   DiffuseFresnelAmount=@n16
   Emissive=FunctionInput() {bUsePreviewValueAsDefault=true, inputName=Emissive, sortPriority=2}
   Base Color=@n15.Diffuse
  ) {materialFunction=&/FresnelFunction}
  FunctionOutput(None=@n17.Emissive) {outputName=emissive, sortPriority=1}
  FunctionOutput(None=@n17.d) {outputName=Diffuse}
  FunctionOutput(None=@n15.Specular) {outputName=Specular, sortPriority=3}
  FunctionOutput(None=@n15.Normals) {
   description=If a normal map is being used connect this output into the normal input on the base shader.
   outputName=Tangent Normal
   sortPriority=6
  }
  FunctionOutput(None=@n15.Roughness) {outputName=Roughness, sortPriority=5}
  FunctionOutput(None=@n15.Metalness) {outputName=Metalness, sortPriority=4}
  n18 = VectorParameter() {bUseCustomPrimitiveData=true, group=z_code, parameterName=WSDeformationCenter, primitiveDataIndex=20}
  n19 = ScalarParameter() {group=Code, parameterName=Code_DestructionAmount}
  n20 = ScalarParameter() {bUseCustomPrimitiveData=true, group=Code, parameterName=WSDamageRadius, primitiveDataIndex=26}
  n21 = FunctionInput(Preview=StaticBool()) {
   bUsePreviewValueAsDefault=true
   inputName=ExposeLerpFunctionForInvertingTheHitDistanceCalc
   inputType=FunctionInput_StaticBool
   sortPriority=101
  }
  n22 = ScalarParameter() {bUseCustomPrimitiveData=true, group=Code, parameterName=InvertDamageFalloff, primitiveDataIndex=11}
  n23 = ScalarParameter() {bUseCustomPrimitiveData=true, desc=1 is inverted, group=Code, parameterName=DistanceCullFadeAnimationOffset, primitiveDataIndex=12}
  n24 = FunctionInput() {
   bUsePreviewValueAsDefault=true
   description=insert any additional world position offset code
   inputName=World Position Offset
   sortPriority=6
  }
  FunctionOutput(None=Add(
    A=MaterialFunctionCall!(
     WSDeformationCenter=@n18.RGB
     Code_DestructionAmount=@n19
     Code_UseUvForDamage
     Code_UseHitEffects
     Code_Emissive
     WSDamageRadius=@n20
     CrackDamageShader
     Diffuse
     ScalarModDamageCracks
     ScalarOffsetforDamageCracks
     ExposeLerpFunctionForInvertingTheHitDistanceCalc=@n21
     InvertDamageFalloff=@n22
     CrackWorldPosition
     WorldPosition
     DistanceCullFadeAnimationOffset=@n23
    ) {materialFunction=/Game/Packages/jl_tests/Fortnite_HitEffects}.Hit And Shrink WPO
    B=@n24
   )) {description=Insert into custom lighting diffuse input on base shader., outputName=World Position Offset, sortPriority=7}
  FunctionOutput(None=@n15.SpecularMasks) {outputName=SpecularMasks, sortPriority=2}
 ]
}
