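# Material function that builds hair material attributes from diffuse, normal and packed specular-mask textures, with Assassin Suit hair textures as some defaults.
# It blends two attribute sets: one drives metallic, specular and roughness from mask G and sets a shading model, the other uses R plus a temporary metalness scalar.
# The "Use AO to Blend out Hair" scalar, zero by default, fades the blend from entirely the second set to following M.R.
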MaterialFunction /Game/Packages/Fortress_SharedMaterials/Base_Material_Functions/FN_Hair
graph {nodes [
  n1 = TextureSampleParameter2D() {group=Hair, parameterName=Diffuse, texture=/Game/Accessories/Hats/M_MED_Assassin_Suit_FaceAcc/Textures/T_Assassin_Suit_Hair_D}
  n2 = TextureSampleParameter2D() {
   desc="R = SpecMask G = Metallic B = Roughness"
   group=Hair
   parameterName=SpecularMasks
   samplerType=SAMPLERTYPE_Masks
   texture=&T_Fortnite_Default_S
  }
  n3 = TextureSampleParameter2D() {
   group=Hair
   parameterName=Normals
   samplerType=SAMPLERTYPE_Normal
   texture=/Game/Accessories/Hats/M_MED_Assassin_Suit_FaceAcc/Textures/T_Assassin_Suit_Hair_N
  }
  n4 = SetMaterialAttributes!(MaterialAttributes, BaseColor=@n1.RGB, Metallic=@n2.G, Specular=@n2.G, Roughness=@n2.G, Normal=@n3.RGB, ShadingModel=ShadingModel()) {attributeSetTypes [
    #1
    #2
    #3
    #4
    #5
    #6
   ]}
  n5 = SetMaterialAttributes!(
   MaterialAttributes
   BaseColor=@n1.RGB
   Metallic=ScalarParameter() {defaultValue=0.5, group=Hair, parameterName=TemporaryHairMetalness}
   Specular=@n2.R
   Roughness=@n2.R
   Normal=@n3.RGB
  ) {attributeSetTypes [
    #1
    #2
    #3
    #4
    #5
   ]}
  n6 = TextureSampleParameter2D() {
   group=Hair
   parameterName=M
   samplerType=SAMPLERTYPE_Masks
   texture=/Game/Accessories/Hats/M_MED_Assassin_Suit_FaceAcc/Textures/T_Assassin_Suit_Hair_M
  }
  n7 = ScalarParameter() {group=Hair, parameterName=Use AO to Blend out Hair}
  FunctionOutput(None=BlendMaterialAttributes(A=@n4, B=@n5, Alpha=LinearInterpolate(B=@n6.R, Alpha=@n7) {constA=1}))
 ]}
