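# Material function in the Fortress shared library that picks between a scalar RVT input and a NoRVT fallback.
# A virtual texture feature switch selects RVT when supported, and a quality switch wraps it to force NoRVT at Low quality before the output.
# A second shading-path branch, limiting RVT on Mobile to Medium quality, exists in the graph but is not connected to the output.
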MaterialFunction /Game/Packages/Fortress_SharedMaterials/Base_Material_Functions/FN_RVT_Switch_Scalar
graph {
 properties {bExposeToLibrary=true, libraryCategoriesText [FN]}
 nodes [
  n1 = FunctionInput() {inputName=NoRVT, inputType=FunctionInput_Scalar}
  n2 = FunctionInput() {inputName=RVT, inputType=FunctionInput_Scalar, sortPriority=1}
  FunctionOutput(None=QualitySwitch(Default=VirtualTextureFeatureSwitch!(No=@n1, Yes=@n2), Low=@n1))
  n3 = VirtualTextureFeatureSwitch(No=@n1, Yes=@n2)
  ShadingPathSwitch(Default=@n3, Mobile=QualitySwitch(Default=@n1, Low=@n1, Medium=@n3))
 ]
}
