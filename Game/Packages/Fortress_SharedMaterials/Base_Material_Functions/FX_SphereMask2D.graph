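# FX_SphereMask2D is a library-exposed material function that outputs a circular mask in UV space.
# It feeds the default texture coordinates into a SphereMask node, with Center, Radius and Hardness inputs setting the second point, size and edge falloff.
# Center is a scalar input, so one value serves both U and V.
# All three inputs fall back to their preview values when unconnected.
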MaterialFunction /Game/Packages/Fortress_SharedMaterials/Base_Material_Functions/FX_SphereMask2D
graph {
 properties {bExposeToLibrary=true}
 nodes [
  n1 = FunctionInput() {bUsePreviewValueAsDefault=true, inputName=Center, inputType=FunctionInput_Scalar, previewValue=<0.5 0 0 1>}
  n2 = FunctionInput() {bUsePreviewValueAsDefault=true, inputName=Radius, inputType=FunctionInput_Scalar, previewValue=<0.5 0 0 1>}
  n3 = FunctionInput() {bUsePreviewValueAsDefault=true, inputName=Hardness, inputType=FunctionInput_Scalar, previewValue=<0.5 0 0 1>}
  FunctionOutput(None=SphereMask(A=TextureCoordinate(), B=@n1, Radius=@n2, Hardness=@n3))
 ]
}
