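# Material function that builds a mask selecting vertices by vertex color value, apparently for hiding character elements through opacity masking.
# A hard-edged SphereMask on scalar inputs selects where the channel lies within a shared "Global Character Vertex Hide Radius" parameter of the target value.
# That mask is multiplied by the hide toggle, ceiled and saturated, so it outputs one for matches only while the toggle is positive.
# A leftover Round node on the toggle is unused.
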MaterialFunction /Game/Packages/Fortress_SharedMaterials/Base_Material_Functions/VertexColorMask
graph {nodes [
  n1 = FunctionInput() {inputName=VertexColorChannel, inputType=FunctionInput_Scalar}
  n2 = FunctionInput(Preview=Constant() {r=0.1}) {inputName=Vertex Value to hide, inputType=FunctionInput_Scalar}
  n3 = ScalarParameter() {defaultValue=0.025, group=03_Opacity Mask, parameterName=Global Character Vertex Hide Radius}
  n4 = Constant() {r=1}
  n5 = FunctionInput() {inputName=Hide Element Toggle, inputType=FunctionInput_Scalar}
  n6 = Ceil(None=@n5)
  FunctionOutput(None=Multiply(A=SphereMask(A=@n1, B=@n2, Radius=@n3, Hardness=@n4), B=Saturate(None=@n6)))
  Round(None=@n5)
 ]}
