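# Material function that hides tagged mesh elements by using World Position Offset to collapse their vertices onto the object position.
# A selectable or max-blended vertex color channel feeds VertexColorMask calls, one per toggle at evenly spaced values, combined with Max.
# HideElementsOnlyConsidersRedChannel can suppress that mask based on vertex color; the result scales the collapse added to incoming WPO.
# A UseNaNMethod divide-by-zero switch, noted for mobile, is not wired to the output.
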MaterialFunction /Game/Packages/Fortress_SharedMaterials/Base_Material_Functions/Fortnite_HideElementsWithVertexColor
graph {nodes [
  n1 = FunctionInput(Preview=Constant3Vector()) {bUsePreviewValueAsDefault=true, inputName=WPO}
  n2 = ObjectPositionWS()
  n3 = Add(A=WorldPosition().XYZ, B=@n1)
  n4 = ScalarParameter() {group=HideElements, parameterName=Hide Element 0_5}
  n5 = Constant() {r=0.05}
  n6 = VertexColor()
  n7 = LinearInterpolate(
   A=ChannelMaskParameter(None=AppendVector(A=@n6, B=@n6)) {group=HideElements, parameterName=Hide Vertex Channel}.R
   B=Max(A=Max(A=@n6.R, B=@n6.R), B=@n6.B)
   Alpha=ScalarParameter() {group=HideElements, parameterName=HideMaxChannelColor}
  )
  n8 = ScalarParameter() {group=HideElements, parameterName=Hide Element 1_5}
  n9 = Constant() {r=0.15}
  n10 = Max(
   A=Max(
    A=MaterialFunctionCall!(Hide Element Toggle=@n4, VertexColorChannel=@n7, Vertex Value to hide=@n5) {materialFunction=&/VertexColorMask}.Result
    B=MaterialFunctionCall(Hide Element Toggle=@n8, VertexColorChannel=@n7, Vertex Value to hide=@n9) {materialFunction=&VertexColorMask}.Result
   )
   B=MaterialFunctionCall(
    Hide Element Toggle=ScalarParameter() {group=HideElements, parameterName=Hide Element 2_5}
    VertexColorChannel=@n7
    Vertex Value to hide=Constant() {r=0.25}
   ) {materialFunction=&VertexColorMask}.Result
  )
  n11 = MaterialFunctionCall(
   Hide Element Toggle=ScalarParameter() {group=HideElements, parameterName=Hide Element 3_5}
   VertexColorChannel=@n7
   Vertex Value to hide=Constant() {r=0.35}
  ) {materialFunction=&VertexColorMask}
  n12 = ScalarParameter() {group=HideElements, parameterName=Hide Element 4_5}
  n13 = Constant() {r=0.45}
  n14 = Max(
   A=Max(
    A=Max(A=@n10, B=@n11.Result)
    B=MaterialFunctionCall(Hide Element Toggle=@n12, VertexColorChannel=@n7, Vertex Value to hide=@n13) {materialFunction=&VertexColorMask}.Result
   )
   B=MaterialFunctionCall(
    Hide Element Toggle=ScalarParameter() {group=HideElements, parameterName=Hide Element 5_5}
    VertexColorChannel=@n7
    Vertex Value to hide=Constant() {r=0.55}
   ) {materialFunction=&VertexColorMask}.Result
  )
  n15 = MaterialFunctionCall(
   Hide Element Toggle=ScalarParameter() {group=HideElements, parameterName=Hide Element 6_5}
   VertexColorChannel=@n7
   Vertex Value to hide=Constant() {r=0.65}
  ) {materialFunction=&VertexColorMask}
  n16 = ScalarParameter() {group=HideElements, parameterName=Hide Element 7_5}
  n17 = Constant() {r=0.75}
  n18 = Max(
   A=Max(
    A=Max(A=@n14, B=@n15.Result)
    B=MaterialFunctionCall(Hide Element Toggle=@n16, VertexColorChannel=@n7, Vertex Value to hide=@n17) {materialFunction=&VertexColorMask}.Result
   )
   B=MaterialFunctionCall(
    Hide Element Toggle=ScalarParameter() {group=HideElements, parameterName=Hide Element 8_5}
    VertexColorChannel=@n7
    Vertex Value to hide=Constant() {r=0.85}
   ) {materialFunction=&VertexColorMask}.Result
  )
  n19 = MaterialFunctionCall(
   Hide Element Toggle=ScalarParameter() {group=HideElements, parameterName=Hide Element 9_5}
   VertexColorChannel=@n7
   Vertex Value to hide=Constant() {r=0.95}
  ) {materialFunction=&VertexColorMask}
  n20 = MaterialFunctionCall(
   Hide Element Toggle=ScalarParameter() {group=HideElements, parameterName=Hide Element 00}
   VertexColorChannel=@n7
   Vertex Value to hide=Constant()
  ) {materialFunction=&VertexColorMask}
  n21 = MaterialFunctionCall(
   Hide Element Toggle=ScalarParameter() {group=HideElements, parameterName=Hide Element 01}
   VertexColorChannel=@n7
   Vertex Value to hide=Constant() {r=0.1}
  ) {materialFunction=&VertexColorMask}
  n22 = ScalarParameter() {group=HideElements, parameterName=Hide Element 02}
  n23 = Constant() {r=0.2}
  n24 = Max(
   A=Max(
    A=Max(A=@n20.Result, B=@n21.Result)
    B=MaterialFunctionCall(Hide Element Toggle=@n22, VertexColorChannel=@n7, Vertex Value to hide=@n23) {materialFunction=&VertexColorMask}.Result
   )
   B=MaterialFunctionCall(
    Hide Element Toggle=ScalarParameter() {group=HideElements, parameterName=Hide Element 03}
    VertexColorChannel=@n7
    Vertex Value to hide=Constant() {r=0.3}
   ) {materialFunction=&VertexColorMask}.Result
  )
  n25 = MaterialFunctionCall(
   Hide Element Toggle=ScalarParameter() {group=HideElements, parameterName=Hide Element 04}
   VertexColorChannel=@n7
   Vertex Value to hide=Constant() {r=0.4}
  ) {materialFunction=&VertexColorMask}
  n26 = ScalarParameter() {group=HideElements, parameterName=Hide Element 05}
  n27 = Constant() {r=0.5}
  n28 = Max(
   A=Max(
    A=Max(A=@n24, B=@n25.Result)
    B=MaterialFunctionCall(Hide Element Toggle=@n26, VertexColorChannel=@n7, Vertex Value to hide=@n27) {materialFunction=&VertexColorMask}.Result
   )
   B=MaterialFunctionCall(
    Hide Element Toggle=ScalarParameter() {group=HideElements, parameterName=Hide Element 06}
    VertexColorChannel=@n7
    Vertex Value to hide=Constant() {r=0.6}
   ) {materialFunction=&VertexColorMask}.Result
  )
  n29 = MaterialFunctionCall(
   Hide Element Toggle=ScalarParameter() {group=HideElements, parameterName=Hide Element 07}
   VertexColorChannel=@n7
   Vertex Value to hide=Constant() {r=0.7}
  ) {materialFunction=&VertexColorMask}
  n30 = ScalarParameter() {group=HideElements, parameterName=Hide Element 08}
  n31 = Constant() {r=0.8}
  n32 = Max(
   A=Max(
    A=Max(A=@n28, B=@n29.Result)
    B=MaterialFunctionCall(Hide Element Toggle=@n30, VertexColorChannel=@n7, Vertex Value to hide=@n31) {materialFunction=&VertexColorMask}.Result
   )
   B=MaterialFunctionCall(
    Hide Element Toggle=ScalarParameter() {group=HideElements, parameterName=Hide Element 09}
    VertexColorChannel=@n7
    Vertex Value to hide=Constant() {r=0.9}
   ) {materialFunction=&VertexColorMask}.Result
  )
  n33 = MaterialFunctionCall(
   Hide Element Toggle=ScalarParameter() {group=HideElements, parameterName=Hide Element 10}
   VertexColorChannel=@n7
   Vertex Value to hide=Constant() {r=1}
  ) {materialFunction=&VertexColorMask}
  n34 = VertexColor()
  n35 = Max(A=Ceil(None=@n34), B=Max(A=Ceil(None=@n34.G), B=Ceil(None=@n34.B)))
  n36 = ScalarParameter() {group=HideElements, parameterName=HideElementsOnlyConsidersRedChannel}
  n37 = Multiply(A=Max(A=Max(A=@n32, B=@n33.Result), B=Max(A=@n18, B=@n19.Result)), B=OneMinus(None=Multiply(A=@n35, B=@n36)))
  n38 = Add(A=@n1, B=Multiply(A=Subtract(A=@n2, B=@n3), B=@n37))
  FunctionOutput(None=@n38)
  n39 = Constant() {r=1}
  n40 = Constant()
  StaticSwitchParameter(True=If(A=@n37, B=Constant() {r=0.5}, "A > B"=Divide(A=@n39, B=@n40) {constB=0}, "A < B"=@n1), False=@n38) {desc=when using WPO on mobile this method wont create uncallable mesh Blobs, group=HideElements, parameterName=UseNaNMethod}
 ]}
