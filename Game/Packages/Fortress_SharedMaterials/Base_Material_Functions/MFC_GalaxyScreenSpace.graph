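# Material function for galaxy skin FX, blending an emissive nebula and starfield into incoming material attributes by a SkinFX mask channel, behind a static switch.
# Camera-facing virtual plane coordinates anchored near the actor and scaled by FOV keep the screen-space look object-stabilized, alongside aspect-corrected normal-map distortion.
# Panning nebula, twinkling small stars and bright stars combine with a noise-broken Fresnel rim.
# The galaxy blackens base color and either adds to or replaces emissive.
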MaterialFunction /Game/Packages/Fortress_SharedMaterials/Base_Material_Functions/MFC_GalaxyScreenSpace
graph {
 properties {description=Galaxy FX in screenspace and object stabilized}
 nodes [
  n1 = NamedRerouteUsage()
  n2 = GetMaterialAttributes(None=NamedRerouteUsage()) {attributeGetTypes [#1, #2, #3]}
  n3 = NamedRerouteUsage()
  n4 = ComponentMask(None=VectorParameter() {group=GalaxyScreenSpace, parameterName=Galaxy_PanSpeed, defaultValue=<rgba 0.005 0.015 0 0>}.RGB) {g=true, r=true}
  n5 = TextureSampleParameter2D(UVs=Add(A=Panner(Coordinate=@n3, Speed=@n4) {speedY=0.02}, B=NamedRerouteUsage())) {
   group=GalaxyScreenSpace
   parameterName=Galaxy
   texture=/Game/Characters/Player/Male/Medium/Bodies/M_Med_Soldier_04/Skins/Galaxy/Textures/T_FN_Nebula
  }
  n6 = Desaturation(None=@n5.RGB, Fraction=ScalarParameter() {defaultValue=0.5, group=GalaxyScreenSpace, parameterName=Star Saturation})
  n7 = Multiply(A=NamedRerouteUsage(), B=ScalarParameter() {defaultValue=5, group=GalaxyScreenSpace, parameterName=Small Star Tiling})
  n8 = ComponentMask(None=VectorParameter() {group=GalaxyScreenSpace, parameterName=StarsSmall_PanSpeed, defaultValue=<rgba -0.02 0.02 0 0>}.RGB) {g=true, r=true}
  n9 = TextureObjectParameter() {
   group=GalaxyScreenSpace
   parameterName=Stars
   samplerType=SAMPLERTYPE_Masks
   texture=/Game/Characters/Player/Male/Medium/Bodies/M_Med_Soldier_04/Skins/Galaxy/Textures/Celestial_T_Stars
  }
  n10 = TextureSample(UVs=Panner(Coordinate=@n7, Speed=@n8) {speedY=0.02}, Tex=@n9) {texture=&T_FN_Nebula}
  n11 = MaterialFunctionCall(Float2=Add(A=Multiply(A=@n7) {constB=0.666}) {constB=0.333}) {materialFunction=&BreakOutFloat2Components}
  n12 = Constant2Vector() {g=0.66, r=-0.5}
  n13 = TextureSample(
   UVs=Panner(
    Coordinate=MaterialFunctionCall(X=@n11.G, Y=@n11.G) {materialFunction=&MakeFloat2}.Result
    Speed=Multiply(A=@n8, B=@n12)
   ) {speedY=0.02}
   Tex=@n9
  ) {texture=&T_FN_Nebula}
  n14 = TextureSample(UVs=Panner(Coordinate=NamedRerouteUsage()) {speedX=0.133, speedY=0.2}, Tex=@n9) {
   samplerSource=SSM_Wrap_WorldGroupSettings
   samplerType=SAMPLERTYPE_Masks
   texture=&Celestial_T_Stars
  }
  n15 = LinearInterpolate(
   A=@n5.RGB
   B=Multiply(A=@n6, B=VectorParameter() {group=GalaxyScreenSpace, parameterName=Star Brightness, defaultValue=<rgba 300 300 300 1>}.RGB)
   Alpha=LinearInterpolate(A=Multiply(A=@n10.B, B=@n10.B), B=Multiply(A=@n13.B, B=@n13.B), Alpha=Multiply(A=@n14.G, B=@n14.G))
  )
  n16 = Multiply(A=@n6, B=VectorParameter() {group=GalaxyScreenSpace, parameterName=Star Brightness_1, defaultValue=<rgba 950 950 950 1>}.RGB)
  n17 = TextureSample(
   UVs=Multiply(A=NamedRerouteUsage(), B=ScalarParameter() {defaultValue=2, group=GalaxyScreenSpace, parameterName=Bright Star Tiling})
   Tex=@n9
  ) {texture=&T_FN_Nebula}
  n18 = Multiply(A=Multiply(A=@n17.R, B=@n17.R), B=@n14.G)
  n19 = Multiply(A=ScalarParameter() {defaultValue=8, group=GalaxyScreenSpace, parameterName=AccentRimBrightness}, B=@n5.RGB)
  n20 = VectorParameter() {group=GalaxyScreenSpace, parameterName=EdgeFresnel, defaultValue=<rgba 5.0909 8.586 10 1>}
  n21 = TextureCoordinate() {uTiling=2, vTiling=2}
  n22 = ScalarParameter() {defaultValue=1, group=GalaxyScreenSpace, parameterName=Edge Fresnel Tiling}
  n23 = ScalarParameter() {group=GalaxyScreenSpace, parameterName=GradientSpeed_X}
  n24 = ScalarParameter() {defaultValue=0.1, group=GalaxyScreenSpace, parameterName=GradientSpeed_Y}
  n25 = TextureSample(UVs=Panner(Coordinate=Multiply(A=@n21, B=@n22), Speed=AppendVector(A=@n23, B=@n24))) {
   samplerSource=SSM_Wrap_WorldGroupSettings
   samplerType=SAMPLERTYPE_LinearColor
   texture=/Game/Weapons/WeaponSkins/Textures/T_Swirly_Noise_highContrast1
  }
  n26 = Multiply(A=@n25.R, B=ScalarParameter() {defaultValue=1, group=GalaxyScreenSpace, parameterName=Galaxy Flame Fresnel Intensity})
  n27 = Fresnel(ExponentIn=ScalarParameter() {defaultValue=5, group=GalaxyScreenSpace, parameterName=Galaxy_FresEX}) {baseReflectFraction=0}
  n28 = Multiply(
   A=ScalarParameter() {defaultValue=1, group=GalaxyScreenSpace, parameterName=Galaxy_GlobalBrightness}
   B=LinearInterpolate(A=LinearInterpolate(A=@n15, B=@n16, Alpha=@n18), B=LinearInterpolate(A=@n19, B=@n20.RGB, Alpha=@n26), Alpha=Saturate(None=@n27))
  )
  n29 = SetMaterialAttributes!(
   MaterialAttributes
   BaseColor=Constant()
   Specular=ScalarParameter() {group=GalaxyScreenSpace, parameterName=Galaxy Specular}
   Roughness=Multiply(A=@n2.Roughness, B=ScalarParameter() {defaultValue=1, group=GalaxyScreenSpace, parameterName=Galaxy Roughness})
   Normal=@n2.Normal
   EmissiveColor=StaticSwitch(
    True=Add(A=@n2.EmissiveColor, B=@n28)
    False=@n28
    Value=StaticBoolParameter() {group=GalaxyScreenSpace, parameterName=UseAdditiveEmissiveGalaxy}
   )
  ) {attributeSetTypes [
    #4
    #5
    #1
    #2
    #3
   ]}
  n30 = NamedRerouteUsage()
  FunctionOutput(None=StaticSwitchParameter(True=BlendMaterialAttributes(A=@n1, B=@n29, Alpha=@n30) {vertexAttributeBlendType=UseA}, False=NamedRerouteUsage()) {group=Skin FX, parameterName=Use Galaxy ScreenSpace})
  NamedRerouteDeclaration(None=FunctionInput() {inputName=MA, inputType=FunctionInput_MaterialAttributes}) {name=MA, nodeColor=<rgba 0 0.6484 1 1>}
  NamedRerouteDeclaration(None=ChannelMaskParameter(None=FunctionInput() {inputName=SkinFX, inputType=FunctionInput_Vector4, sortPriority=1}) {group=GalaxyScreenSpace, parameterName=GalaxyScreenSpace_SkinFX_Mask}.R) {name=SkinFX_Mask, nodeColor=<rgba 1 0.6563 0 1>}
  n31 = MaterialFunctionCall() {materialFunction=&CameraDirectionVector}
  n32 = Multiply(A=Constant3Vector() {constant=<rgba 1 1 0 0>}, B=@n31.Result)
  n33 = Add(A=Multiply(B=Normalize(VectorInput=@n32)) {constA=0.25}, B=@n31.Result)
  n34 = Multiply(A=@n31.Result) {constB=50}
  n35 = Add(A=ActorPositionWS(), B=Constant3Vector() {constant=<rgba 0 0 90 1>})
  n36 = ScalarParameter() {defaultValue=128, group=GalaxyScreenSpace, parameterName=TextureResolution, sliderMax=2048}
  n37 = Multiply(A=@n36) {constB=0.5}
  n38 = InverseLinearInterpolate(
   A=Constant() {r=45}
   B=Constant() {r=90}
   Value=MaterialFunctionCall() {materialFunction=/Engine/Functions/Engine_MaterialFunctions02/Utility/FOV}.FOV angle in degrees
  )
  NamedRerouteDeclaration(None=MaterialFunctionCall(Plane Normal Axis=Normalize(VectorInput=@n33), Plane Center=Add(A=@n34, B=@n35), UV Scale=LinearInterpolate(A=@n37, B=@n36, Alpha=@n38)) {materialFunction=&VirtualPlaneCoordinates}.UVs) {name=ScreenSpaceUVs, nodeColor=<rgba 0.5469 0 1 1>}
  n39 = MaterialFunctionCall(Float2=MaterialFunctionCall() {materialFunction=&ScreenResolution}.Visible Resolution) {materialFunction=&BreakOutFloat2Components}
  n40 = MaterialFunctionCall(Float2=Constant() {r=512}) {materialFunction=&BreakOutFloat2Components}
  n41 = Divide(B=Multiply(A=Divide(A=@n39.R, B=@n39.R), B=Divide(A=@n40.G, B=@n40.G) {desc=ratio}))
  n42 = MaterialFunctionCall(
   UVs=ScreenPosition().ViewportUV
   Texture Scale=AppendVector(
    A=MaterialFunctionCall(Float2=@n41) {materialFunction=&BreakOutFloat2Components}.R
    B=Constant() {r=1}
   )
  ) {materialFunction=&ScaleUVsByCenter}
  n43 = TextureSample(UVs=Divide(
    A=Panner(Coordinate=@n42.UVs) {speedX=0.01, speedY=-0.02}
    B=ScalarParameter() {defaultValue=4, group=GalaxyScreenSpace, parameterName=DistortionTiling, sliderMax=32}
   )) {samplerType=SAMPLERTYPE_Normal, texture=/Game/Gadgets/Assets/Telephone_Booth/Textures/T_BlindsNormalNoise_N}
  NamedRerouteDeclaration(None=Multiply(
    A=ComponentMask(None=@n43.RGB) {g=true, r=true}
    B=ScalarParameter() {defaultValue=0.08, group=GalaxyScreenSpace, parameterName=DistortionIntensity, sliderMax=1}
   ) {constB=0.02}) {name=DistortedUVs, nodeColor=<rgba 1 0.2812 0 1>}
 ]
}
