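# Material function that animates a chest opening in the vertex shader, behind the Use Chest Animation static switch.
# Progress comes from CODE_Timestamp scaled by ChestSpeed, or from looping time for debugging.
# It drives world position offset: a height-weighted sine wiggle, a delayable jump, and a delayable pivot rotation plus slide masked by red vertex color.
# Normals get a matching rotation on the deferred path only; Forward and Mobile keep the input normal.
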MaterialFunction /Game/Packages/Fortress_SharedMaterials/Base_Material_Functions/MF_ChestAnimation
graph {nodes [
  n1 = GetMaterialAttributes(None=FunctionInput() {inputType=FunctionInput_MaterialAttributes}) {attributeGetTypes [#1, #2]}
  n2 = VectorParameter() {parameterName=Chest_Wigglestrength, defaultValue=<rgba 3 3 3 1>}
  n3 = Frac(None=Time())
  n4 = ScalarParameter() {defaultValue=3, group=Chest Animation, parameterName=ChestSpeed}
  n5 = ScalarParameter() {parameterName=CODE_Timestamp}
  n6 = StaticSwitchParameter(True=Saturate(None=Multiply(A=@n3, B=@n4) {constB=2}), False=Saturate(None=Multiply(A=@n4, B=@n5))) {parameterName=DebugChestAnim}
  n7 = Sine(None=Multiply(B=@n6) {constA=2})
  n8 = WorldPosition()
  n9 = MaterialFunctionCall() {materialFunction=&ObjectPivotPoint}
  n10 = Divide(
   A=Subtract(A=ComponentMask(None=@n8.XYZ) {b=true}, B=ComponentMask(None=@n9.Object Pivot Location) {b=true})
   B=ComponentMask(None=MaterialFunctionCall() {materialFunction=&ObjectLocalBounds}.Local Bounds Size) {b=true}
  )
  n11 = Multiply(A=Time(), B=ScalarParameter() {defaultValue=0.1, group=Chest Animation, parameterName=Chest_Animation_WiggleSpeed})
  n12 = Sine(None=@n11)
  n13 = Sine(None=Multiply(A=@n11) {constB=1.5})
  n14 = Multiply(A=@n11) {constB=2}
  n15 = Multiply(A=Multiply(A=Multiply(A=@n2.RGB, B=@n7), B=Saturate(None=@n10)), B=AppendVector(A=AppendVector(A=@n12, B=@n13), B=Sine(None=@n14)))
  n16 = ScalarParameter() {parameterName=DelayJump}
  n17 = Multiply(A=Divide(A=@n6, B=OneMinus(None=@n16)), B=Subtract(A=@n6, B=@n16))
  n18 = Saturate(None=Divide(A=Saturate(None=@n17)))
  n19 = Multiply(
   A=VectorParameter() {parameterName=Chest_Anim_JumpHeight, defaultValue=<rgba 0 0 5 1>}.RGB
   B=Power(Base=Sine(None=@n18), Exp=ScalarParameter() {defaultValue=16, parameterName=Chest_Anim_Power2})
  )
  n20 = VertexColor()
  n21 = Normalize(VectorInput=Transform(None=VectorParameter() {group=Chest Animation, parameterName=Chest_Rotation_Axis, defaultValue=<rgba 1 0 0 1>}.RGB) {transformSourceType=TRANSFORMSOURCE_Local})
  n22 = ScalarParameter() {parameterName=DelayOpen}
  n23 = Divide(A=@n6, B=OneMinus(None=@n22))
  n24 = Subtract(A=@n6, B=@n22)
  n25 = Saturate(None=Power(Base=Multiply(A=@n23, B=@n24), Exp=ScalarParameter() {defaultValue=0.5238, parameterName=Chest_Anim_Power}))
  n26 = Multiply(A=ScalarParameter() {group=Chest Animation, parameterName=Chest_Animation_Max_Rotation}, B=@n25)
  n27 = TransformPosition(None=VectorParameter() {group=Chest Animation, parameterName=Chest_Rotation_Pivot, defaultValue=<rgba 0 0 0 1>}.RGB) {transformType=TRANSFORMPOSSOURCE_World}
  n28 = VectorParameter() {group=Chest Animation, parameterName=Chest_Move_to_Location, defaultValue=<rgba 0 0 0 1>}
  n29 = Transform(None=@n28.RGB) {transformSourceType=TRANSFORMSOURCE_Local}
  n30 = Add(
   A=RotateAboutAxis(NormalizedRotationAxis=@n21, RotationAngle=@n26, PivotPoint=@n27, Position=WorldPosition().XYZ)
   B=Multiply(A=@n25, B=@n29)
  )
  n31 = Add(A=@n1.WorldPositionOffset, B=Add(A=Add(A=@n15, B=@n19), B=Multiply(A=@n20.R, B=@n30)))
  n32 = Transform(None=@n1.Normal)
  n33 = RotateAboutAxis(NormalizedRotationAxis=@n21, RotationAngle=@n26, PivotPoint=Constant3Vector(), Position=@n32)
  n34 = LinearInterpolate(
   A=@n1.Normal
   B=Transform(None=Add(A=@n32, B=@n33)) {transformSourceType=TRANSFORMSOURCE_World, transformType=TRANSFORM_Tangent}
   Alpha=@n20.R
  )
  n35 = ShadingPathSwitch(Default=@n34, Deferred=@n34, Forward=@n1.Normal, Mobile=@n1.Normal)
  FunctionOutput(None=StaticSwitchParameter(
    True=SetMaterialAttributes!(MaterialAttributes=@n1.MaterialAttributes, WorldPositionOffset=@n31, Normal=@n35) {attributeSetTypes [#1, #2]}
    False=@n1.MaterialAttributes
   ) {parameterName=Use Chest Animation})
  LinearInterpolate()
  MaterialFunctionCall(
   BaseNormal=@n1.Normal
   AdditionalNormal=VertexInterpolator(VS=MaterialFunctionCall(Rotation Axis=@n21, Rotation Angle=@n26) {materialFunction=&FixRotateAboutAxisNormals}.New Tangent Space Vertex Normal).PS
  ) {materialFunction=&BlendAngleCorrectedNormals}
  Multiply(A=@n5, B=ScalarParameter() {parameterName=CODE_CHEST_ANIM_PROGRESS})
  Subtract(A=@n29, B=@n27)
  Subtract(A=Time(), B=@n5)
  TransformPosition(None=@n28.RGB) {transformType=TRANSFORMPOSSOURCE_World}
 ]}
