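# Material function that outputs unscaled world-aligned 2D coordinates, with no texture sampling despite its name.
# It compares absolute vertex world normal components to a threshold to pick one world-position plane: XY for mostly vertical normals, otherwise YZ for X-facing normals, otherwise XZ.
# Nested If nodes make this a hard box-projection switch, avoiding the blending of a full triplanar projection.
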MaterialFunction /Game/Packages/Fortress_SharedMaterials/Base_Material_Functions/MF_Curie_WorldAlignedTexture
graph {nodes [
  n1 = MaterialFunctionCall(Float3=Abs(None=VertexNormalWS())) {materialFunction=&BreakOutFloat3Components}
  n2 = Constant() {r=0.5}
  n3 = WorldPosition()
  n4 = ComponentMask(None=@n3.XYZ) {b=true, desc=YZ, g=true}
  n5 = ComponentMask(None=@n3.XYZ) {b=true, desc=XZ, r=true}
  FunctionOutput(None=If(A=@n1.B, B=@n2, "A > B"=ComponentMask(None=@n3.XYZ) {desc=YX, g=true, r=true}, "A < B"=If(A=@n1.R, B=@n2, "A > B"=@n4, "A < B"=@n5)))
 ]}
