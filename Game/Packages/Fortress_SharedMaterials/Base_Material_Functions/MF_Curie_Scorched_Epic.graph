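# Material function that layers Curie elemental scorching (char, embers, flames) onto input material attributes, altering base color, subsurface color and emissive.
# Custom primitive data supplies element index, char amount and glow blend; the index reads color, pan speeds and UV scale from a lookup texture.
# Char amount grows a local-bounds sphere mask, or is used directly at Low quality.
# Panned noise forms flames; noise masks use world-aligned or switchable spherical UVs.
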MaterialFunction /Game/Packages/Fortress_SharedMaterials/Base_Material_Functions/MF_Curie_Scorched_Epic
graph {nodes [
  n1 = MaterialFunctionCall(Float3=NamedRerouteUsage()) {desc="red=circular, grn=cloudy, blue=webby", materialFunction=&BreakOutFloat3Components}
  Add(A=@n1.R, B=@n1.R)
  n2 = TextureObject() {samplerType=SAMPLERTYPE_LinearColor, texture=&T_NoisePack_LinearColor}
  n3 = Constant() {r=255}
  n4 = Subtract(A=ComponentMask(None=MaterialFunctionCall(TextureObject=@n2, TextureSize=@n3) {materialFunction=&WorldAlignedTexture}.XYZ Texture) {g=true, r=true}) {constB=0.5}
  n5 = Multiply(A=MaterialFunctionCall() {materialFunction=&/MF_Curie_WorldAlignedTexture}.Result) {constB=1.25}
  n6 = Divide(A=@n5, B=Constant() {r=950})
  Add(A=Multiply(A=@n4, B=Constant() {desc=0.03, r=0.025}), B=@n6)
  n7 = Divide(A=@n5, B=Constant() {r=950})
  Add(A=@n7, B=Multiply(A=@n4, B=Constant() {desc=0.05, r=0.015}))
  n8 = MaterialFunctionCall(Float3=NamedRerouteUsage()) {desc="red=chunks, grn=patches, blue=edges", materialFunction=&BreakOutFloat3Components}
  Add(A=@n8.G, B=@n8.G)
  Constant3Vector() {constant=<rgba 1 0.1222 0.0026 0>}
  n9 = GetMaterialAttributes(None=FunctionInput() {inputType=FunctionInput_MaterialAttributes, previewValue=<1 1 1 1>}) {attributeGetTypes [#1, #2, #3]}
  n10 = Multiply(A=@n9.BaseColor, B=NamedRerouteUsage()) {constB=0}
  n11 = Saturate(None=Multiply(A=NamedRerouteUsage(), B=ScalarParameter() {defaultValue=2, desc=char alpha, group=Curie, parameterName=Curie Char Alpha, sliderMax=1}) {constB=2})
  n12 = NamedRerouteUsage()
  n13 = Multiply(A=@n9.SubsurfaceColor) {constB=0.1}
  n14 = Power(Base=@n11) {constExponent=0.2}
  FunctionOutput(None=SetMaterialAttributes!(
    MaterialAttributes=@n9.MaterialAttributes
    BaseColor=LinearInterpolate(A=@n9.BaseColor, B=@n10, Alpha=@n11)
    EmissiveColor=Add(A=@n9.EmissiveColor, B=@n12) {constB=0}
    SubsurfaceColor=LinearInterpolate(A=@n9.SubsurfaceColor, B=@n13, Alpha=@n14)
   ) {attributeSetTypes [#1, #2, #3]})
  MaterialFunctionCall() {materialFunction=&LocalPosition}
  MaterialFunctionCall() {materialFunction=&MF_Curie_WorldAlignedTexture}
  MaterialFunctionCall() {desc="red=chunks, grn=patches, blue=edges", materialFunction=&BreakOutFloat3Components}
  MaterialFunctionCall() {
   desc="red=circular, grn=cloudycells, blue=webby"
   materialFunction=&BreakOutFloat3Components
  }
  MaterialFunctionCall() {desc="red=chunks, grn=patches, blue=edges", materialFunction=&BreakOutFloat3Components}
  MaterialFunctionCall() {materialFunction=&MF_Curie_WorldAlignedTexture}
  MaterialFunctionCall() {materialFunction=&MF_Curie_WorldAlignedTexture}
  Multiply(A=Time() {bOverride_Period=true, period=10000}, B=AppendVector(A=Constant() {r=-0.0775}, B=Constant() {r=-0.25}))
  Multiply(A=ScalarParameter() {defaultValue=1, group=Curie, parameterName=Curie PatchScale, sliderMax=2}, B=@n1.G) {constB=1.5}
  n15 = Time() {bOverride_Period=true, period=10000}
  n16 = TextureSample(UVs=AppendVector(
    A=ScalarParameter() {bUseCustomPrimitiveData=true, group=Curie, parameterName=CurieElementIndex, primitiveDataIndex=27}
    B=Constant() {r=0.35}
   )) {samplerSource=SSM_Wrap_WorldGroupSettings, samplerType=SAMPLERTYPE_LinearColor, texture=/Game/Curie/Effects/T_Curie_LinearColors}
  NamedRerouteDeclaration(None=Multiply(A=@n15, B=AppendVector(A=@n16.R, B=@n16.R))) {name=WorldPannerA, nodeColor=<rgba 0.5469 0 1 1>}
  n17 = MaterialFunctionCall(Float3=NamedRerouteUsage()) {desc="red=circular, grn=cloudy, blue=webby", materialFunction=&BreakOutFloat3Components}
  n18 = Multiply(A=Power(Base=@n17.R) {desc=block the flames}, B=Divide(B=@n17.B))
  n19 = Multiply(
   A=Constant() {r=0}
   B=MaterialFunctionCall() {materialFunction=&MF_Curie_WorldAlignedTexture}.Result
  )
  n20 = TextureSample(UVs=@n19) {samplerType=SAMPLERTYPE_Normal, texture=/Game/Global/Textures/Default/Noise/T_Curie_Noise_Normal_Epic}
  n21 = Multiply(A=ComponentMask(None=Transform(None=@n20.RGB)) {g=true, r=true}, B=@n19) {constA=1}
  n22 = TextureSample(UVs=Add(A=Add(A=@n7, B=@n21), B=NamedRerouteUsage())) {samplerType=SAMPLERTYPE_LinearColor, texture=/Game/Global/Textures/Default/Noise/T_Curie_NoisePack}
  n23 = TextureSample(UVs=Add(A=Add(A=@n6, B=@n21), B=NamedRerouteUsage())) {samplerType=SAMPLERTYPE_LinearColor, texture=&T_Curie_NoisePack}
  n24 = Add(A=@n22.G, B=@n23.R)
  n25 = Multiply(A=Clamp(None=Multiply(A=@n18) {constB=2}) {maxDefault=25}, B=Multiply(A=Multiply(A=@n22.B, B=@n23.B), B=Power(Base=@n24))) {constA=1}
  n26 = ScalarParameter() {defaultValue=0.584, group=Curie, parameterName=Curie FlameGlowScale, sliderMax=1}
  NamedRerouteDeclaration(None=Multiply(A=Multiply(A=@n25, B=@n26), B=NamedRerouteUsage())) {name=Flames, nodeColor=<rgba 1 0.115 0.0153 1>}
  n27 = TextureSample(UVs=AppendVector(
    A=ScalarParameter() {bUseCustomPrimitiveData=true, group=Curie, parameterName=CurieElementIndex, primitiveDataIndex=27}
    B=Constant() {r=0.05}
   )) {samplerSource=SSM_Wrap_WorldGroupSettings, samplerType=SAMPLERTYPE_LinearColor, texture=&T_Curie_LinearColors}
  NamedRerouteDeclaration(None=@n27.RGB) {name=CurieColor, nodeColor=<rgba 1 0.0222 0 1>}
  n28 = NamedRerouteUsage()
  n29 = ScalarParameter() {defaultValue=1.5, group=Curie, parameterName=Curie DetailMaskScale_Char, sliderMax=15}
  NamedRerouteDeclaration(None=TextureSample(
    UVs=Divide(A=@n28, B=@n29) {constB=5}
    Tex=TextureObject() {samplerType=SAMPLERTYPE_LinearColor, texture=/Game/Global/Textures/Default/Noise/T_Curie_NoisePack_Epic}
   ) {samplerType=SAMPLERTYPE_LinearColor, texture=&T_Curie_NoisePack_Epic}.RGB) {name=Detail Char, nodeColor=<rgba 0 1 0.267 1>}
  n30 = NamedRerouteUsage()
  n31 = ScalarParameter() {defaultValue=1.9, group=Curie, parameterName=Curie LargeMaskScale, sliderMax=5000}
  NamedRerouteDeclaration(None=TextureSample(UVs=Divide(A=@n30, B=@n31) {constB=600}) {texture=/Game/Effects/Fort_Effects/Textures/Fire/HF_FireNoisesPacked}.RGB) {name=LargeScale Cloudy, nodeColor=<rgba 1 0 0.1641 1>}
  n32 = NamedRerouteUsage()
  n33 = ScalarParameter() {defaultValue=0.64, group=Curie, parameterName=Curie LargeMaskScale_Char, sliderMax=5000}
  NamedRerouteDeclaration(None=TextureSample(UVs=Divide(A=@n32, B=@n33) {constB=600}) {samplerType=SAMPLERTYPE_LinearColor, texture=&T_Curie_NoisePack_Epic}.RGB) {name=LargeScale Char, nodeColor=<rgba 0 1 0.1887 1>}
  n34 = Constant3Vector() {constant=<rgba 1 0 0 0>}
  n35 = Normalize(VectorInput=Subtract(A=ObjectPositionWS(), B=WorldPosition().XYZ))
  n36 = DotProduct(A=Normalize(VectorInput=Transform(None=@n34) {transformSourceType=TRANSFORMSOURCE_Local}), B=@n35)
  n37 = Constant3Vector() {constant=<rgba 0 1 0 0>}
  n38 = DotProduct(A=@n35, B=Normalize(VectorInput=Transform(None=@n37) {transformSourceType=TRANSFORMSOURCE_Local}))
  n39 = Add(A=Divide(A=Arctangent2Fast(Y=@n36, X=@n38)) {constB=6.2832}) {constB=0.5}
  n40 = ScalarParameter() {defaultValue=8, desc=box volume, group=Curie, parameterName=Xscale, sliderMax=1}
  n41 = Divide(
   A=ComponentMask(None=WorldPosition() {worldPositionShaderOffset=WPT_ExcludeAllShaderOffsets}.XYZ) {b=true}
   B=ScalarParameter() {defaultValue=2500, group=Curie, parameterName=CurieSphericalMappingScale}
  )
  NamedRerouteDeclaration(None=AppendVector(A=Multiply(A=@n39, B=@n40), B=Multiply(A=@n40, B=@n41))) {name=SphericalMapping, nodeColor=<rgba 0 0.9062 1 1>}
  n42 = MaterialFunctionCall(Float3=NamedRerouteUsage()) {desc="red=chunks, grn=patches, blue=edges", materialFunction=&BreakOutFloat3Components}
  n43 = ScalarParameter() {defaultValue=0.2, desc=was .3, group=Curie, parameterName=Curie Stain Trail, sliderMax=1}
  n44 = NamedRerouteUsage()
  n45 = Multiply(
   A=Add(A=Add(A=@n1.R, B=@n42.R), B=LinearInterpolate(B=@n43, Alpha=@n44) {constB=0.3, desc=longer wipe trail - user higher values}) {desc=add contrast}
   B=@n44
  )
  NamedRerouteDeclaration(None=Saturate(None=Subtract(A=@n45, B=@n1.G))) {name=Bleed Mask, nodeColor=<rgba 0.4629 0 1 1>}
  n46 = MaterialFunctionCall() {materialFunction=&ObjectLocalBounds}
  n47 = ScalarParameter() {bUseCustomPrimitiveData=true, group=Curie, parameterName=CurieCharred, primitiveDataIndex=28}
  n48 = ScalarParameter() {defaultValue=3000, group=Curie, parameterName=Curie Wipe Mask Radius, sliderMax=15}
  n49 = Saturate(None=SphereMask(
    A=MaterialFunctionCall() {materialFunction=&LocalPosition}.Local Position
    B=LinearInterpolate(A=@n46.Local Bounds Minimum, B=@n46.Local Bounds Minimum)
    Radius=Multiply(A=@n47, B=@n48) {constB=5}
    Hardness=ScalarParameter() {group=Curie, parameterName=Curie Wipe Mask Hardness, sliderMax=1}
   ) {hardnessPercent=81.6})
  NamedRerouteDeclaration(None=QualitySwitch(Default=@n49, Low=@n47, Epic=@n49)) {name=Char Mask, nodeColor=<rgba 1 0 0.7031 1>}
  n50 = NamedRerouteUsage()
  n51 = TextureSample(UVs=AppendVector(
    A=ScalarParameter() {bUseCustomPrimitiveData=true, group=Curie, parameterName=CurieElementIndex, primitiveDataIndex=27}
    B=Constant() {r=0.55}
   )) {samplerSource=SSM_Wrap_WorldGroupSettings, samplerType=SAMPLERTYPE_LinearColor, texture=&T_Curie_LinearColors}
  n52 = Divide(
   A=MaterialFunctionCall() {materialFunction=&MF_Curie_WorldAlignedTexture}.Result
   B=@n51.R
  )
  n53 = StaticBool()
  NamedRerouteDeclaration(None=StaticSwitch(
    True=@n50
    False=StaticSwitchParameter(True=@n50, False=@n52) {group=Curie, parameterName=UseSphericalMapping}
    Value=FunctionInput(Preview=@n53) {bUsePreviewValueAsDefault=true, inputName=UseSphericalMappingOverride, inputType=FunctionInput_StaticBool, sortPriority=1}
   )) {name=FinalUVs, nodeColor=<rgba 1 0.0938 0 1>}
  n54 = Add(A=@n8.R, B=@n8.R)
  n55 = ScalarParameter() {defaultValue=3, desc=was 2, group=Curie, parameterName=Curie Char Contrast, sliderMax=3}
  n56 = MaterialFunctionCall(Float3=NamedRerouteUsage()) {desc="red=chunks, grn=patches, blue=edges", materialFunction=&BreakOutFloat3Components}
  n57 = Multiply(
   A=Multiply(A=Power(Base=@n54, Exp=@n55), B=@n56.B)
   B=ScalarParameter() {defaultValue=2, desc=was 2, group=Curie, parameterName=Curie Char Scale, sliderMax=3}
  ) {constB=15}
  n58 = Saturate(None=Multiply(A=NamedRerouteUsage(), B=NamedRerouteDeclaration(None=@n57) {name=Char BaseColor, nodeColor=<rgba 0 0.8594 1 1>}))
  n59 = ScalarParameter() {defaultValue=12, group=Curie, parameterName=Curie Ember Scale, sliderMax=12}
  n60 = Add(A=Multiply(A=Multiply(A=@n58, B=@n59) {constB=0.2}, B=NamedRerouteUsage()), B=NamedRerouteUsage()) {constB=0, desc=Add Flames}
  n61 = NamedRerouteUsage()
  n62 = ScalarParameter() {bUseCustomPrimitiveData=true, group=Curie, parameterName=CurieLERP, primitiveDataIndex=29}
  NamedRerouteDeclaration(None=Multiply(A=Multiply(A=@n60, B=@n61) {desc=Mask by char}, B=Multiply(A=@n62) {constB=1.5})) {name=Final_Emissive, nodeColor=<rgba 1 0.4717 0 1>}
  n63 = TextureSample(UVs=AppendVector(
    A=ScalarParameter() {bUseCustomPrimitiveData=true, group=Curie, parameterName=CurieElementIndex, primitiveDataIndex=27}
    B=Constant() {r=0.45}
   )) {samplerSource=SSM_Wrap_WorldGroupSettings, samplerType=SAMPLERTYPE_LinearColor, texture=&T_Curie_LinearColors}
  NamedRerouteDeclaration(None=Multiply(A=@n15, B=AppendVector(A=@n63.R, B=@n63.R))) {name=WorldPannerB, nodeColor=<rgba 0.7812 0 1 1>}
  Power(Base=@n42.R) {constExponent=1}
  n64 = Multiply(A=Time() {bOverride_Period=true, period=10000}, B=AppendVector(A=Constant() {r=0.0512}, B=Constant() {r=-0.35}))
  Reroute(None=@n64)
  Reroute(None=Multiply(A=@n64) {constB=0.66})
  Reroute(None=@n1.B)
  ShadingPathSwitch(Default=QualitySwitch())
  VectorParameter() {parameterName=EmissionColor, defaultValue=<rgba 1 0.1221 0.0056 1>}
 ]}
