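# MF_Curie_Scorched_Med is a material function that adds Curie scorching, with char and animated flames, to incoming material attributes.
# Custom primitive data supplies char amount, flame blend and an element index into a color lookup texture.
# Packed noise, on spherical or world-aligned UVs, builds char masks and two-layer panning flames.
# The default path darkens base, specular and subsurface and adjusts emissive; Low, Forward and Mobile use simpler uniform darkening plus glow.
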MaterialFunction /Game/Packages/Fortress_SharedMaterials/Base_Material_Functions/MF_Curie_Scorched_Med
graph {nodes [
  Constant3Vector() {constant=<rgba 1 0.1222 0.0026 0>}
  Constant3Vector()
  n1 = Constant3Vector() {constant=<rgba 0.151 0.151 0.151 0>}
  n2 = FunctionInput(Preview=SetMaterialAttributes!(MaterialAttributes)) {inputType=FunctionInput_MaterialAttributes, previewValue=<1 1 1 1>}
  n3 = GetMaterialAttributes(None=@n2) {attributeGetTypes [
    #1
    #2
    #3
    #4
   ]}
  n4 = NamedRerouteUsage()
  n5 = Multiply(A=NamedRerouteUsage(), B=Power(Base=@n4) {constExponent=3})
  n6 = Constant3Vector()
  n7 = SetMaterialAttributes!(
   MaterialAttributes=@n3.MaterialAttributes
   BaseColor=LinearInterpolate(A=@n3.BaseColor, B=Multiply(A=@n3.BaseColor, B=@n1) {constB=0}, Alpha=@n4)
   EmissiveColor=MaterialFunctionCall(Emissive=Add(A=@n3.EmissiveColor, B=@n5) {constB=0}) {materialFunction=&MF_FX_EmissiveAdjustments}.Result
   SubsurfaceColor=LinearInterpolate(A=@n3.SubsurfaceColor, B=Multiply(A=@n3.SubsurfaceColor) {constB=0.1}, Alpha=Power(Base=@n4) {constExponent=0.2})
   Specular=LinearInterpolate(A=@n3.Specular, B=Multiply(A=@n3.Specular, B=@n6) {constB=0}, Alpha=@n4)
  ) {attributeSetTypes [
    #1
    #2
    #3
    #4
   ]}
  n8 = GetMaterialAttributes(None=@n2) {attributeGetTypes [#1, #2, #4]}
  n9 = ScalarParameter() {bUseCustomPrimitiveData=true, group=Curie, parameterName=CurieCharred, primitiveDataIndex=28}
  n10 = NamedRerouteUsage()
  n11 = SetMaterialAttributes!(
   MaterialAttributes=@n8.MaterialAttributes
   BaseColor=LinearInterpolate(A=@n8.BaseColor, B=Multiply(A=@n8.BaseColor) {constB=0.07}, Alpha=@n9)
   EmissiveColor=Add(A=@n8.EmissiveColor, B=Multiply(A=@n10) {constB=2})
   Specular=LinearInterpolate(A=@n8.Specular, B=Multiply(A=@n8.Specular) {constB=0.07}, Alpha=@n9)
  ) {attributeSetTypes [#1, #2, #4]}
  FunctionOutput(None=QualitySwitch(Default=ShadingPathSwitch(Default=@n7, Forward=@n11, Mobile=@n11), Low=@n11))
  n12 = Divide(A=NamedRerouteUsage(), B=Constant() {r=1.5}) {constB=600}
  n13 = NamedRerouteDeclaration(None=TextureSample(UVs=@n12) {samplerType=SAMPLERTYPE_LinearColor, texture=/Game/Global/Textures/Default/Noise/T_Curie_NoisePack}.B) {name=LargeMask, nodeColor=<rgba 0 0.9766 1 1>}
  n14 = Power(Base=OneMinus(None=@n13), Exp=ScalarParameter() {defaultValue=2, group=Curie, parameterName=Curie MaskContrast_low, sliderMax=5}) {constExponent=3.5}
  LinearInterpolate(A=@n13, B=@n14) {constAlpha=1}
  Multiply(A=NamedRerouteUsage(), B=@n9)
  n15 = Multiply(A=NamedRerouteUsage()) {constB=0.7}
  n16 = Multiply(A=Time() {bOverride_Period=true, period=10000}, B=AppendVector(A=Constant() {r=0.0512}, B=Constant() {r=-0.35}))
  n17 = Add(A=@n15, B=Multiply(A=@n16) {constB=0.66})
  n18 = Add(
   A=TextureSample(UVs=Add(A=@n15, B=@n16)) {samplerType=SAMPLERTYPE_LinearColor, texture=&T_Curie_NoisePack}.G
   B=Power(Base=TextureSample(UVs=@n17) {samplerType=SAMPLERTYPE_LinearColor, texture=&T_Curie_NoisePack}.R) {constExponent=4}
  )
  n19 = Multiply(A=QualitySwitch(Default=ShadingPathSwitch(Default=@n18, Forward=@n18, Mobile=@n18), Low=@n18)) {constB=2}
  n20 = TextureSample(UVs=AppendVector(
    A=ScalarParameter() {bUseCustomPrimitiveData=true, group=Curie, parameterName=CurieElementIndex, primitiveDataIndex=27}
    B=Constant() {r=0.05}
   )) {samplerSource=SSM_Wrap_WorldGroupSettings, samplerType=SAMPLERTYPE_LinearColor, texture=/Game/Curie/Effects/T_Curie_LinearColors}
  NamedRerouteDeclaration(None=Multiply(
    A=Multiply(A=@n19, B=@n20.RGB)
    B=ScalarParameter() {bUseCustomPrimitiveData=true, group=Curie, parameterName=CurieLERP, primitiveDataIndex=29}
   )) {name=Flames, nodeColor=<rgba 1 0.115 0.0153 1>}
  NamedRerouteDeclaration(None=@n14) {name=LargeScale Cloudy, nodeColor=<rgba 1 0 0.1641 1>}
  n21 = Constant3Vector() {constant=<rgba 1 0 0 0>}
  n22 = Normalize(VectorInput=Subtract(A=ObjectPositionWS(), B=WorldPosition().XYZ))
  n23 = DotProduct(A=Normalize(VectorInput=Transform(None=@n21) {transformSourceType=TRANSFORMSOURCE_Local}), B=@n22)
  n24 = Constant3Vector() {constant=<rgba 0 1 0 0>}
  n25 = DotProduct(A=@n22, B=Normalize(VectorInput=Transform(None=@n24) {transformSourceType=TRANSFORMSOURCE_Local}))
  n26 = Add(A=Divide(A=Arctangent2Fast(Y=@n23, X=@n25)) {constB=6.2832}) {constB=0.5}
  n27 = ScalarParameter() {defaultValue=8, desc=box volume, parameterName=Xscale, sliderMax=1}
  n28 = Divide(
   A=ComponentMask(None=WorldPosition() {worldPositionShaderOffset=WPT_ExcludeAllShaderOffsets}.XYZ) {b=true}
   B=ScalarParameter() {defaultValue=2500, group=Curie, parameterName=CurieSphericalMappingScale}
  )
  NamedRerouteDeclaration(None=AppendVector(A=Multiply(A=@n26, B=@n27), B=Multiply(A=@n27, B=@n28))) {name=SphericalMapping, nodeColor=<rgba 0 0.9062 1 1>}
  n29 = MaterialFunctionCall(Float3=NamedRerouteUsage()) {desc="red=circular, grn=cloudy, blue=webby", materialFunction=&BreakOutFloat3Components}
  n30 = NamedRerouteUsage()
  n31 = Add(
   A=Multiply(A=Power(Base=@n29.R) {constExponent=1.5}, B=@n30)
   B=Multiply(A=@n30, B=ScalarParameter() {defaultValue=0.4, group=Curie, parameterName=Curie CharContrast_low, sliderMax=3})
  )
  n32 = ScalarParameter() {defaultValue=1.77, desc=char scale, group=Curie, parameterName=Curie CharScale_Low, sliderMax=3}
  NamedRerouteDeclaration(None=Saturate(None=Multiply(A=@n31, B=@n32) {constB=2})) {name=Char BaseColor, nodeColor=<rgba 0 0.0351 1 1>}
  n33 = ScalarParameter() {bUseCustomPrimitiveData=true, group=Curie, parameterName=CurieCharred, primitiveDataIndex=28}
  NamedRerouteDeclaration(None=@n33) {name=Char Alpha, nodeColor=<rgba 1 0 0.7031 1>}
  n34 = NamedRerouteUsage()
  n35 = TextureSample(UVs=AppendVector(
    A=ScalarParameter() {bUseCustomPrimitiveData=true, group=Curie, parameterName=CurieElementIndex, primitiveDataIndex=27}
    B=Constant() {r=0.55}
   )) {samplerSource=SSM_Wrap_WorldGroupSettings, samplerType=SAMPLERTYPE_LinearColor, texture=&T_Curie_LinearColors}
  n36 = Divide(
   A=MaterialFunctionCall() {materialFunction=&/MF_Curie_WorldAlignedTexture}.Result
   B=@n35.R
  )
  n37 = StaticBool()
  NamedRerouteDeclaration(None=StaticSwitch(
    True=@n34
    False=StaticSwitchParameter(True=@n34, False=@n36) {group=Curie, parameterName=UseSphericalMapping}
    Value=FunctionInput(Preview=@n37) {bUsePreviewValueAsDefault=true, inputName=UseSphericalMappingOverride, inputType=FunctionInput_StaticBool, sortPriority=1}
   )) {name=FinalUVs, nodeColor=<rgba 1 0.0938 0 1>}
  NamedRerouteUsage()
  OneMinus()
  Power(Base=@n29.G) {constExponent=0.66}
  n38 = MaterialFunctionCall() {materialFunction=&LocalPosition}
  n39 = MaterialFunctionCall() {materialFunction=&ObjectLocalBounds}
  n40 = LinearInterpolate(A=@n39.Local Bounds Minimum, B=@n39.Local Bounds Minimum)
  n41 = Multiply(A=@n33, B=ScalarParameter() {defaultValue=3000, parameterName=Mask Radius, sliderMax=15}) {constB=5}
  n42 = ScalarParameter() {group=Curie, parameterName=Curie Mask Hardness, sliderMax=1}
  QualitySwitch(Default=Saturate(None=SphereMask(A=@n38.Local Position, B=@n40, Radius=@n41, Hardness=@n42) {hardnessPercent=81.6}), Low=@n33)
  Saturate(None=@n4)
  ScalarParameter() {defaultValue=2, parameterName=FlameGlowStrength_Low, sliderMax=1}
  ShadingPathSwitch(Default=QualitySwitch())
  TextureSample(UVs=@n12) {texture=/Game/Effects/Fort_Effects/Textures/Fire/HF_FireNoisesPacked}
 ]}
