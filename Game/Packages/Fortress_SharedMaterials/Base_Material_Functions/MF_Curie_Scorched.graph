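# Material function that applies the Curie scorched look to incoming material attributes, choosing an implementation by quality level.
# A quality switch runs MF_Curie_Scorched_Epic at Epic and MF_Curie_Scorched_Med at the default, labelled medium and below.
# Both variants receive the same attributes plus an optional UseSphericalMappingOverride static bool, which is passed straight through.
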MaterialFunction /Game/Packages/Fortress_SharedMaterials/Base_Material_Functions/MF_Curie_Scorched
graph {nodes [
  n1 = FunctionInput() {inputType=FunctionInput_MaterialAttributes, previewValue=<0.7 0.7 0.7 1>}
  n2 = FunctionInput(Preview=StaticBool()) {bUsePreviewValueAsDefault=true, inputName=UseSphericalMappingOverride, inputType=FunctionInput_StaticBool, sortPriority=1}
  FunctionOutput(None=QualitySwitch(
    Default=MaterialFunctionCall!(In=@n1, UseSphericalMappingOverride=@n2) {desc=Med and Below, materialFunction=&/MF_Curie_Scorched_Med}.Result
    Epic=MaterialFunctionCall!(In=@n1, UseSphericalMappingOverride=@n2) {desc=Epic, materialFunction=&/MF_Curie_Scorched_Epic}.Result
   ))
 ]}
