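# Material function meant as a Curie visual effect on meshes, but its output is wired straight to its input, passing attributes through unchanged.
# The unconnected graph sends default shading to MF_Curie_Scorched, and Forward and Mobile to a cheaper inline fallback.
# That fallback uses world-aligned or angle-and-height noise UVs to mask a base-color blend and panning orange emissive.
# Custom primitive data drives char amount (CurieChared) and glow strength (CurieLERP).
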MaterialFunction /Game/Packages/Fortress_SharedMaterials/Base_Material_Functions/MF_CurieFallbackEffect
graph {
 properties {description=Used to provide a visual effect on a mesh affected by the Curie system.}
 nodes [
  n1 = FunctionInput() {inputType=FunctionInput_MaterialAttributes}
  FunctionOutput(None=@n1)
  n2 = Normalize(VectorInput=Transform(None=Constant3Vector() {constant=<rgba 1 1 0 0>}) {transformSourceType=TRANSFORMSOURCE_Local})
  n3 = Normalize(VectorInput=Subtract(A=ObjectPositionWS(), B=WorldPosition().XYZ))
  n4 = Normalize(VectorInput=Transform(None=Constant3Vector() {constant=<rgba -1 1 0 0>}) {transformSourceType=TRANSFORMSOURCE_Local})
  n5 = Divide(A=Arctangent2Fast(Y=DotProduct(A=@n2, B=@n3), X=DotProduct(A=@n4, B=@n3))) {constB=6.2832}
  n6 = ComponentMask(None=WorldPosition() {worldPositionShaderOffset=WPT_ExcludeAllShaderOffsets}.XYZ) {b=true}
  n7 = ScalarParameter() {defaultValue=1500, group=Curie, parameterName=CurieSphericalMappingScale}
  NamedRerouteDeclaration(None=AppendVector(A=Add(A=@n5) {constB=0.5}, B=Divide(A=@n6, B=@n7))) {name=SphericalMapping, nodeColor=<rgba 0 0.9062 1 1>}
  n8 = Time() {bOverride_Period=true, period=10000}
  NamedRerouteDeclaration(None=Multiply(A=@n8, B=Constant2Vector() {g=-0.3395, r=0.1019})) {name=WorldPannerA, nodeColor=<rgba 0.5469 0 1 1>}
  NamedRerouteDeclaration(None=Multiply(A=@n8, B=Constant2Vector() {g=-0.5375, r=-0.1268})) {name=WorldPannerB, nodeColor=<rgba 0.7813 0 1 1>}
  NamedRerouteDeclaration(None=MaterialFunctionCall() {materialFunction=&/MF_Curie_WorldAlignedTexture}.Result) {name=WorldAlignedTexture, nodeColor=<rgba 1 0.2109 0 1>}
  n9 = NamedRerouteUsage()
  n10 = Divide(A=NamedRerouteUsage(), B=Constant() {r=470})
  NamedRerouteDeclaration(None=StaticSwitch(True=@n9, False=StaticSwitchParameter(True=@n9, False=@n10) {group=Curie, parameterName=UseSphericalMapping}, Value=NamedRerouteUsage())) {name=FinalUVs, nodeColor=<rgba 1 0.0938 0 1>}
  n11 = NamedRerouteDeclaration(None=FunctionInput(Preview=StaticBool()) {bUsePreviewValueAsDefault=true, inputName=UseSphericalMappingOverride, inputType=FunctionInput_StaticBool, sortPriority=1}) {name=UseSphericalMappingOverride, nodeColor=<rgba 1 0.6094 0 1>}
  n12 = GetMaterialAttributes(None=@n1) {attributeGetTypes [#1]}
  n13 = Multiply(A=@n12.BaseColor)
  n14 = TextureSample(UVs=NamedRerouteUsage()) {
   desc=World Aligned Mask
   samplerSource=SSM_Wrap_WorldGroupSettings
   samplerType=SAMPLERTYPE_LinearColor
   texture=/Game/Global/Textures/Default/Noise/T_Curie_NoisePack
  }
  n15 = Add(
   A=ConstantBiasScale(None=ScalarParameter() {bUseCustomPrimitiveData=true, group=Curie, parameterName=CurieChared, primitiveDataIndex=28}) {bias=-0.5, scale=2}
   B=@n14.B
  )
  n16 = GetMaterialAttributes(None=BlendMaterialAttributes(
    A=@n1
    B=SetMaterialAttributes!(MaterialAttributes=@n12.MaterialAttributes, BaseColor=@n13) {attributeSetTypes [#1]}
    Alpha=Saturate(None=@n15)
   ) {vertexAttributeBlendType=UseA}) {attributeGetTypes [#2]}
  n17 = ScalarParameter() {bUseCustomPrimitiveData=true, group=Curie, parameterName=CurieLERP, primitiveDataIndex=29}
  n18 = TextureSample(UVs=Add(A=NamedRerouteUsage(), B=NamedRerouteUsage())) {
   samplerSource=SSM_Wrap_WorldGroupSettings
   samplerType=SAMPLERTYPE_LinearColor
   texture=&T_Curie_NoisePack
  }
  n19 = Multiply(A=Constant3Vector() {constant=<rgba 3.2832 0.7314 0.1652 0>}, B=Multiply(A=@n14.B, B=@n18.G))
  n20 = SetMaterialAttributes!(MaterialAttributes=@n16.MaterialAttributes, EmissiveColor=Add(A=@n16.EmissiveColor, B=Multiply(A=@n17, B=@n19))) {attributeSetTypes [#2]}
  ShadingPathSwitch(
   Default=ShadingPathSwitch(
    Default=MaterialFunctionCall!(In=@n1, UseSphericalMappingOverride=@n11) {materialFunction=&/MF_Curie_Scorched}.Result
    Mobile=@n20
   ) {desc=Migrated from FeatureLevelSwitch}
   Forward=@n20
   Mobile=@n20
  )
 ]
}
