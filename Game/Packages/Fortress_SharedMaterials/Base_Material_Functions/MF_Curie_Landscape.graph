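# Landscape material function for the Curie fire system, exposing a fire render target sample and a burned-state mask alongside material attributes.
# World position, remapped by collection bias and divisor, samples the render target's red channel.
# A charring and flame-emissive blend uses panned noise, quality and shading-path switches, grass toggles and RVT-height falloff.
# That blend feeds no output; the main output passes input attributes through unchanged.
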MaterialFunction /Game/Packages/Fortress_SharedMaterials/Base_Material_Functions/MF_Curie_Landscape
graph {nodes [
  n1 = FunctionInput() {inputType=FunctionInput_MaterialAttributes}
  FunctionOutput(None=@n1)
  n2 = WorldPosition()
  n3 = CollectionParameter() {collection=&FortniteMaterialParameters, parameterName=CurieLandscapeBias}
  n4 = Add(A=AppendVector(A=TruncateLWC!(None=@n2.XYZ), B=Constant()), B=@n3)
  n5 = CollectionParameter() {collection=&FortniteMaterialParameters, parameterName=CurieLandscapeDivisor}
  n6 = ComponentMask(None=Saturate(None=Divide(A=@n4, B=@n5) {constB=260096})) {g=true, r=true}
  FunctionOutput(None=NamedRerouteDeclaration(None=TextureSample(UVs=@n6) {samplerType=SAMPLERTYPE_LinearColor, texture=/Game/Curie/Textures/LandscapeGrassRenderTarget}.R) {name=LandscapeRenderTarget, nodeColor=<rgba 1 0 0.6574 1>}) {outputName=LandscapeRenderTarget, sortPriority=1}
  n7 = NamedRerouteUsage()
  n8 = Multiply(A=Add(A=NamedRerouteUsage(), B=Constant() {r=-0.1}), B=Constant() {r=1.25})
  n9 = NamedRerouteUsage()
  n10 = Constant() {r=2.4}
  n11 = Multiply(A=NamedRerouteUsage(), B=Constant() {r=1.8})
  n12 = ShadingPathSwitch(Default=Multiply(A=Add(A=@n7, B=@n8), B=Multiply(A=@n9, B=@n10)), Forward=@n11, Mobile=@n11)
  FunctionOutput(None=NamedRerouteDeclaration(None=Saturate(None=@n12)) {name=StateLerp, nodeColor=<rgba 1 0.7969 0 1>}) {outputName=BurnedState, sortPriority=2}
  n13 = NamedRerouteUsage()
  n14 = ScalarParameter() {defaultValue=610, group=z_GameplayParameters_FireSystem, parameterName=FlamesDivisorA}
  n15 = Time() {bOverride_Period=true, period=10000}
  n16 = VectorParameter() {group=z_GameplayParameters_FireSystem, parameterName=CuriePanSpeedTerrainWSv4, defaultValue=<rgba 0.1563 -0.1601 -0.1432 0.1567>}
  n17 = AppendVector(A=@n16.R, B=@n16.R)
  n18 = TextureSample(UVs=Add(A=Divide(A=@n13, B=@n14), B=Multiply(A=@n15, B=@n17))) {
   samplerSource=SSM_Wrap_WorldGroupSettings
   samplerType=SAMPLERTYPE_LinearColor
   texture=/Game/Global/Textures/Default/Noise/T_Curie_NoisePack
  }
  n19 = NamedRerouteUsage()
  n20 = ScalarParameter() {defaultValue=680, group=z_GameplayParameters_FireSystem, parameterName=FlamesDivisorB}
  n21 = AppendVector(A=@n16.B, B=@n16.B)
  n22 = TextureSample(UVs=Add(A=Divide(A=@n19, B=@n20), B=Multiply(A=@n15, B=@n21))) {
   samplerSource=SSM_Wrap_WorldGroupSettings
   samplerType=SAMPLERTYPE_LinearColor
   texture=&T_Curie_NoisePack
  }
  Multiply(A=@n18.R, B=@n22.R)
  n23 = NamedRerouteDeclaration(None=@n1) {name=Material Attributes, nodeColor=<rgba 0 0.9531 1 1>}
  n24 = GetMaterialAttributes(None=@n23) {attributeGetTypes [
    #1
    #2
    #3
    #4
    #5
   ]}
  NamedRerouteDeclaration(None=@n24.BaseColor) {name=BaseColorIn, nodeColor=<rgba 0.8047 0 1 1>}
  NamedRerouteDeclaration(None=@n24.Metallic) {name=MetallicIn, nodeColor=<rgba 0.3594 1 0 1>}
  n25 = FunctionInput(Preview=StaticBool()) {bUsePreviewValueAsDefault=true, desc=WildEstate Sylisation, inputName=Is_Stylised_WE, inputType=FunctionInput_StaticBool, sortPriority=9999}
  NamedRerouteDeclaration(None=@n25) {name=IsWildEstateStylised, nodeColor=<rgba 0 1 0.8828 1>}
  NamedRerouteDeclaration(None=@n24.Specular) {name=SpecularIn, nodeColor=<rgba 1 0.4922 0 1>}
  NamedRerouteDeclaration(None=@n24.Roughness) {name=RoughnessIn, nodeColor=<rgba 1 0 0.0937 1>}
  NamedRerouteDeclaration(None=@n24.EmissiveColor) {name=EmissiveIn, nodeColor=<rgba 1 0.1641 0 1>}
  NamedRerouteDeclaration(None=ComponentMask(None=WorldPosition() {worldPositionShaderOffset=WPT_ExcludeAllShaderOffsets}.XYZ) {desc=XY, g=true, r=true}) {name=AbsoluteWorldPosition.XY, nodeColor=<rgba 1 0 0.6797 1>}
  n26 = TextureSample(UVs=Divide(A=NamedRerouteUsage(), B=Constant() {r=371})) {
   samplerSource=SSM_Wrap_WorldGroupSettings
   samplerType=SAMPLERTYPE_LinearColor
   texture=&T_Curie_NoisePack
  }
  NamedRerouteDeclaration(None=@n26.B) {name=WorldMask, nodeColor=<rgba 0 0 1 1>}
  n27 = FunctionInput(Preview=StaticBool()) {bUsePreviewValueAsDefault=true, inputName=In_Grass, inputType=FunctionInput_StaticBool, sortPriority=200}
  NamedRerouteDeclaration(None=@n27) {name=IsGrass, nodeColor=<rgba 1 0.9844 0 1>}
  Reroute(None=@n1)
  Reroute(None=@n27)
  n28 = Constant() {r=0.4}
  n29 = Add(A=NamedRerouteUsage(), B=NamedRerouteUsage())
  n30 = Constant3Vector() {constant=<rgba 0.1806 0.1498 0.1411 1>}
  n31 = QualitySwitch(Default=Constant3Vector() {constant=<rgba 0.0537 0.055 0.0457 1>}, Low=@n30, Medium=@n30)
  n32 = Multiply(
   A=Desaturation(None=NamedRerouteUsage(), Fraction=Multiply(A=@n28, B=@n29))
   B=LinearInterpolate(A=Constant3Vector() {constant=<rgba 0.03 0.03 0.03 1>}, B=@n31, Alpha=@n29)
  )
  n33 = NamedRerouteUsage()
  n34 = FunctionInput(Preview=Constant3Vector() {constant=<rgba 1.8 0.4025 0.0531 1>}) {bUsePreviewValueAsDefault=true, inputName=OnFireEmissiveOverride, sortPriority=1}
  n35 = Divide(A=NamedRerouteUsage(), B=Constant() {r=1125})
  n36 = Time() {bOverride_Period=true, period=10000}
  n37 = Constant4Vector() {constant=<rgba 0.0569 -0.0455 -0.0407 0.0445>}
  n38 = ComponentMask(None=@n37.RGBA) {g=true, r=true}
  n39 = TextureSample(UVs=Add(A=@n35, B=Multiply(A=@n36, B=@n38))) {
   samplerSource=SSM_Wrap_WorldGroupSettings
   samplerType=SAMPLERTYPE_LinearColor
   texture=&T_Curie_NoisePack
  }
  n40 = ComponentMask(None=@n37.RGBA) {a=true, b=true}
  n41 = TextureSample(UVs=Add(A=@n35, B=Multiply(A=@n36, B=@n40))) {
   samplerSource=SSM_Wrap_WorldGroupSettings
   samplerType=SAMPLERTYPE_LinearColor
   texture=&T_Curie_NoisePack
  }
  n42 = Multiply(A=Multiply(A=Multiply(A=@n39.B, B=@n41.B), B=Add(A=@n39.G, B=@n41.G)), B=Add(A=@n39.B, B=@n41.B))
  n43 = Add(A=@n18.R, B=@n22.R)
  n44 = ScalarParameter() {defaultValue=4, group=z_GameplayParameters_FireSystem, parameterName=EpicFlamesExponent}
  n45 = Add(
   A=@n42
   B=Multiply(A=Power(Base=@n43, Exp=@n44), B=ScalarParameter() {defaultValue=0.025, group=z_GameplayParameters_FireSystem, parameterName=EpicFlamesScale})
  )
  n46 = MaterialFunctionCall!(min=Constant() {r=0.8}, max=Constant() {r=0.95}, value=NamedRerouteUsage()) {materialFunction=/Game/Effects/Fort_Effects/Materials/_Functions/InverseLerp_MF}
  n47 = Add(A=NamedRerouteUsage(), B=Constant() {r=-0.3517})
  n48 = Constant() {r=3}
  n49 = Multiply(A=@n46.Clamped, B=Add(A=@n46.Clamped, B=Multiply(A=@n47, B=@n48)))
  n50 = Multiply(
   A=ShadingPathSwitch(Default=QualitySwitch(Default=@n42, Epic=@n45), Forward=@n39.G, Mobile=@n39.G)
   B=ShadingPathSwitch(Default=Saturate(None=@n49), Forward=@n46.Clamped, Mobile=@n46.Clamped)
  )
  n51 = StaticSwitch(True=@n33, False=Add(A=@n33, B=Multiply(A=@n34, B=@n50)) {desc=Add original emissive back in}, Value=NamedRerouteUsage())
  n52 = NamedRerouteUsage()
  n53 = Constant() {r=0.1}
  n54 = SetMaterialAttributes!(
   MaterialAttributes=@n24.MaterialAttributes
   BaseColor=@n32
   Metallic=StaticSwitch(True=@n24.Metallic, False=Multiply(A=@n52, B=@n53) {desc=Metallic}, Value=NamedRerouteUsage())
   Specular=Multiply(A=NamedRerouteUsage(), B=Constant() {r=0.065}) {desc=Specular}
   Roughness=Multiply(A=NamedRerouteUsage(), B=Constant() {r=0.5}) {desc=Roughness}
   EmissiveColor=@n51
   SubsurfaceColor=Constant()
  ) {attributeSetTypes [
    #1
    #2
    #3
    #4
    #5
    #6
   ]}
  n55 = StaticSwitch(
   True=SetMaterialAttributes!(MaterialAttributes=@n24.MaterialAttributes, BaseColor=@n32, EmissiveColor=@n51) {attributeSetTypes [#1, #5]}
   False=@n54
   Value=NamedRerouteUsage()
  )
  n56 = SetMaterialAttributes(MaterialAttributes=@n23, BaseColor=Multiply(A=NamedRerouteUsage(), B=@n31), EmissiveColor=@n51) {attributeSetTypes [#1, #5]}
  n57 = RuntimeVirtualTextureSampleParameter() {
   materialType=WorldHeight
   parameterName=RVT Landscape Height Target
   virtualTexture=/Game/Athena/Apollo/Maps/Landscape/VT/RVT_Apollo_Landscape_Height
  }
  n58 = SphereMask(
   A=@n57.WorldHeight
   B=ComponentMask(None=WorldPosition().XYZ) {b=true}
   Radius=FunctionInput(Preview=ScalarParameter() {defaultValue=170, group=z_GameplayParameters_FireSystem, parameterName=HeightFalloffRadius}) {bUsePreviewValueAsDefault=true, inputName=HeightFalloffRadius, inputType=FunctionInput_Scalar, sortPriority=3}
   Hardness=FunctionInput(Preview=ScalarParameter() {defaultValue=0.4, group=z_GameplayParameters_FireSystem, parameterName=HeightFalloffHardness}) {bUsePreviewValueAsDefault=true, inputName=HeightFalloffHardness, inputType=FunctionInput_Scalar, sortPriority=4}
  )
  n59 = NamedRerouteUsage()
  n60 = QualitySwitch(Default=@n59, Epic=ShadingPathSwitch(Default=@n59, Deferred=Multiply(A=@n58, B=@n59)))
  n61 = FunctionInput(Preview=StaticBool() {value=true}) {bUsePreviewValueAsDefault=true, inputName=UseHeightFalloff, inputType=FunctionInput_StaticBool, sortPriority=1}
  n62 = GetMaterialAttributes(None=NamedRerouteUsage()) {attributeGetTypes [#1, #7]}
  n63 = ComponentMask(None=@n62.Normal) {b=true}
  n64 = Multiply(A=@n63, B=@n63)
  n65 = FunctionInput() {bUsePreviewValueAsDefault=true, inputName=VariationMasksMobile, inputType=FunctionInput_Vector4, sortPriority=10, previewValue=<1 1 1 1>}
  n66 = Multiply(A=Min(A=NamedRerouteUsage()) {constB=0.7}, B=Add(A=OneMinus(None=@n64), B=ComponentMask(None=@n65) {b=true}))
  n67 = Saturate(None=Subtract(A=Multiply(A=@n66) {constB=6.5}) {constB=0.05})
  n68 = Multiply(A=@n62.BaseColor, B=OneMinus(None=Multiply(A=@n67) {constB=0.9}))
  n69 = NamedRerouteUsage()
  ShadingPathSwitch(
   Default=BlendMaterialAttributes(A=@n23, B=ShadingPathSwitch(Default=@n55, Forward=@n56, Mobile=@n56), Alpha=StaticSwitch(True=@n60, False=@n59, Value=@n61)) {desc="Vertex blend type altered!", vertexAttributeBlendType=UseA}
   Mobile=SetMaterialAttributes!(MaterialAttributes=@n62.MaterialAttributes, BaseColor=StaticSwitch(True=@n62.BaseColor, False=@n68, Value=@n69)) {attributeSetTypes [#1]}
  )
 ]}
