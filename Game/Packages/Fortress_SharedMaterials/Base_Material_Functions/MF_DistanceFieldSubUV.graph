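# Material function that turns an animated distance field flipbook texture into a sharp-edged scalar mask.
# Time, wrapped on a four-second period and scaled by the speed input, drives the frame of the engine SubUV_Function, which samples the input texture.
# The red channel is offset to threshold just below mid-gray, then steeply amplified and saturated, leaving a narrow soft edge around the encoded shape.
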MaterialFunction /Game/Packages/Fortress_SharedMaterials/Base_Material_Functions/MF_DistanceFieldSubUV
graph {nodes [
  n1 = FunctionInput() {bUsePreviewValueAsDefault=true, inputName=SubUV_TextureObject, inputType=FunctionInput_Texture2D, previewValue=<2 0 0 1>}
  n2 = FunctionInput() {bUsePreviewValueAsDefault=true, inputName=SubUV_Frames, inputType=FunctionInput_Scalar, previewValue=<2 0 0 1>}
  n3 = FunctionInput() {bUsePreviewValueAsDefault=true, inputName=SubUV_Speed, inputType=FunctionInput_Scalar, previewValue=<2 0 0 1>}
  n4 = MaterialFunctionCall!(Texture=@n1, UVs, SubImages=@n2, Frame=Multiply(A=Time() {period=4}, B=@n3) {constB=2}) {materialFunction=/Engine/Functions/Engine_MaterialFunctions03/Texturing/SubUV_Function}
  n5 = Saturate(None=Add(A=MaterialFunctionCall(Float3=@n4.RGB) {materialFunction=&BreakOutFloat3Components}.R) {constB=-0.48})
  n6 = Constant() {r=16}
  FunctionOutput(None=Saturate(None=Multiply(A=@n5, B=@n6)))
 ]}
