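# Library material function that turns an input value into a binary, temporally shifting screen-space dither mask without a texture, like DitherTemporalAA.
# Pixel position, offset by the temporal sample index and multiplied by Scale, feeds a five-step diagonal grid and an Iterations-driven Bayer loop in Custom HLSL.
# Their normalized sum offsets the input before rounding and saturation.
# The Bayer loop uses logical `&&` where bit extraction normally uses `&`.
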MaterialFunction /Game/Packages/Fortress_SharedMaterials/Base_Material_Functions/MF_DitherBayerNoTex
graph {
 properties {bExposeToLibrary=true}
 nodes [
  n1 = ViewProperty() {property=MEVP_TemporalSampleIndex}
  n2 = FunctionInput() {
   bUsePreviewValueAsDefault=true
   description="1 is default, 4 has good quality"
   inputName=Scale
   inputType=FunctionInput_Scalar
   sortPriority=3
   previewValue=<1 0 0 1>
  }
  MaterialExpressionMultiply_0 = Multiply(A=Add(A=ScreenPosition().PixelPosition, B=AppendVector(A=@n1.Property, B=@n1.Property)), B=@n2)
  n3 = Custom!(None=@MaterialExpressionMultiply_0) {
   code="Mod(((uint)(Input.x) + 2 * (uint)(Input.y)), 5);"
   description=Grid
   outputType=CMOT_Float1
   showCode=true
   inputs [{
     inputName=Input
     input {expression=&self:MaterialExpressionMultiply_0}
    }]
  }
  n4 = FunctionInput(Preview=Constant() {r=1}) {
   bUsePreviewValueAsDefault=true
   description="1 is default, 4 has good quality"
   inputName=Iterations
   inputType=FunctionInput_Scalar
   sortPriority=1
   previewValue=<2 0 0 1>
  }
  MaterialExpressionMax_0 = Max(A=@n4)
  n5 = Custom!(UV=@MaterialExpressionMultiply_0, Iterations=@MaterialExpressionMax_0) {
   code="// Bayer matrix\r\n\r\nint2 p = int2(UV);\r\n\r\nint sum = 0;\r\n\r\nfor(int i = 0; i < int(Iterations); i++)\r\n{\r\n    int2 t = p && 1;\r\n    sum = sum * 4 | (t.x ^ t.y) * 2 | t.x;\r\n    p *= 0.5;\r\n}    \r\n\r\nreturn float(sum) / float(1 << (2 * int(Iterations)));"
   description=Bayer
   outputType=CMOT_Float1
   showCode=true
   inputs [
    {
     inputName=UV
     input {expression=&self:MaterialExpressionMultiply_0}
    }
    {
     inputName=Iterations
     input {expression=&self:MaterialExpressionMax_0}
    }
   ]
  }
  n6 = Multiply(A=Divide(A=Add(A=@n3, B=@n5) {desc=address banding issues}) {constB=6}) {constB=0.998}
  n7 = ComponentMask(None=TextureCoordinate()) {r=true}
  n8 = Subtract(A=Add(A=Add(A=@n6) {constB=0.001}, B=FunctionInput(Preview=@n7) {inputType=FunctionInput_Vector4})) {constB=0.5}
  FunctionOutput(None=Saturate(None=Round(None=@n8)))
 ]
}
