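# Material function for first-person weapon scopes that outputs eye UV, a soft circular eye mask and camera-to-eye-point distance.
# C++-updated eye, target and up vectors define the scope axis, with points remapped through first-person transforms to match weapon rendering.
# Screen-space view rays, rebuilt from FOV, analytically intersect a virtual plane to produce scaled UVs.
# A static switch selects camera or world space; custom HLSL applies View.FirstPersonTransform for radius scaling.
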MaterialFunction /Game/Packages/Fortress_SharedMaterials/Base_Material_Functions/MF_FirstPersonScope
graph {nodes [
  Constant3Vector() {constant=<rgba 1 0 0 0>}
  Constant() {r=1}
  n1 = FunctionInput() {bUsePreviewValueAsDefault=true, inputName=Plane Depth Offset, inputType=FunctionInput_Scalar}
  MaterialExpressionMultiply_12 = Multiply(A=@n1, B=Constant3Vector() {constant=<rgba 0 0 1 0>})
  MaterialExpressionNamedRerouteUsage_15 = NamedRerouteUsage()
  n2 = Custom!(Radius=@MaterialExpressionMultiply_12, Center=@MaterialExpressionNamedRerouteUsage_15) {
   code="const float3 ScaleFP = mul((Center + Radius), (float3x3)View.FirstPersonTransform);\r\nreturn ScaleFP;"
   showCode=true
   inputs [
    {
     inputName=Radius
     input {expression=&self:MaterialExpressionMultiply_12}
    }
    {
     inputName=Center
     input {expression=&self:MaterialExpressionNamedRerouteUsage_15}
    }
   ]
  }
  Distance(A=@n2, B=NamedRerouteUsage())
  n3 = NamedRerouteUsage()
  n4 = ScalarParameter() {defaultValue=1, parameterName=Scope Eye Mask Hardness}
  FunctionOutput(None=SphereMask(
    A=FunctionOutput(None=@n3) {outputName=Scope Eye UV, sortPriority=60}
    B=Constant2Vector() {g=0.5, r=0.5}
    Hardness=FunctionInput(Preview=@n4) {bUsePreviewValueAsDefault=true, inputName=Scope Eye Mask Hardness, inputType=FunctionInput_Scalar}
   ) {attenuationRadius=0.5}) {outputName=Scope Eye Mask, sortPriority=50}
  FunctionOutput(None=Distance(A=NamedRerouteUsage(), B=NamedRerouteUsage())) {outputName=Camera - Eye Point Distance, sortPriority=70}
  n5 = CameraVectorWS()
  n6 = Normalize(VectorInput=CrossProduct(A=@n5, B=Constant3Vector() {constant=<rgba 0 0 1 0>}))
  n7 = NamedRerouteUsage()
  n8 = NamedRerouteUsage()
  FunctionOutput(None=SphereMask(
    A=MaterialFunctionCall(Plane Normal Axis=@n5, Plane X Axis=@n6, Plane Center=@n7, UV Scale=@n8) {materialFunction=&VirtualPlaneCoordinates}.UVs
    B=Constant2Vector() {g=0.5, r=0.5}
   ) {attenuationRadius=0.5}) {outputName=Debug, sortPriority=999}
  MaterialFunctionCall() {materialFunction=&VirtualPlaneCoordinates}
  Multiply(A=CameraVectorWS()) {constB=-1}
  n9 = DotProduct(A=NamedRerouteUsage(), B=NamedRerouteUsage())
  n10 = DotProduct(A=NamedRerouteUsage(), B=NamedRerouteUsage()) {desc=Distance to Plane}
  n11 = Multiply(A=Subtract(A=ScreenPosition().ViewportUV) {constB=0.5}, B=Constant2Vector() {g=-2, r=2}) {constB=2}
  n12 = MaterialFunctionCall() {materialFunction=/Game/Global/MaterialFunctions/MF_FirstPerson_TanHalfFov}
  n13 = Normalize(VectorInput=AppendVector(A=Multiply(A=@n11, B=@n12."Tan(0.5 * Field Of View)"), B=Constant() {r=1}))
  n14 = StaticSwitch(True=@n13, False=Transform(None=@n13) {transformSourceType=TRANSFORMSOURCE_View}, Value=NamedRerouteUsage())
  n15 = Divide(
   A=Multiply(A=Subtract(A=@n9, B=@n10) {desc=Plane Distance}) {constB=-1}
   B=DotProduct(A=NamedRerouteUsage(), B=@n14) {desc=VectorDotPlaneNormal}
  ) {desc=Intersection Time1}
  n16 = Constant()
  n17 = VectorParameter() {
   desc="This value is updated in C++\r\nWeapon Scope Attachment"
   group=C++ Parameters
   parameterName=ScopeUpVectorWorldSpace
   defaultValue=<rgba 0 0 1 0>
  }
  n18 = Normalize(VectorInput=NamedRerouteDeclaration(None=FunctionInput(Preview=@n17.RGB) {bUsePreviewValueAsDefault=true, inputName=Up Vector, sortPriority=24}) {name=Up Vector, nodeColor=<rgba 1 0.8672 0 1>})
  n19 = NamedRerouteUsage()
  n20 = Constant3Vector() {constant=<rgba 0 0.0002 -1 0>}
  n21 = Subtract(A=Add(A=NamedRerouteUsage(), B=Multiply(A=@n15, B=@n14)) {desc=Intersection Position}, B=NamedRerouteUsage())
  n22 = DotProduct(
   A=CrossProduct(
    A=@n18
    B=MaterialFunctionCall!(Vector1=@n19, Vector2=@n20) {materialFunction=/Engine/Functions/Engine_MaterialFunctions02/Math/CreateThirdOrthogonalVector}.Vector1
   )
   B=@n21
  )
  n23 = DotProduct(A=@n18, B=@n21)
  n24 = FunctionInput() {
   bUsePreviewValueAsDefault=true
   description=The size of the virtual plane
   inputName=Scale
   inputType=FunctionInput_Vector2
   sortPriority=3
   previewValue=<1 1 0 1>
  }
  n25 = Divide(
   A=If(A=@n15, B=@n16, "A > B"=AppendVector(A=@n22, B=@n23), "A == B"=@n16, "A < B"=@n16)
   B=NamedRerouteDeclaration(None=Multiply(A=@n24) {constB=2}) {name=ScopeRadiusScaleOrg, nodeColor=<rgba 1 0 0.5625 1>}
  )
  n26 = Constant2Vector() {g=-1, r=1}
  NamedRerouteDeclaration(None=Add(A=Multiply(A=@n25, B=@n26)) {constB=0.5}) {name=Virtual Plane Coordinates, nodeColor=<rgba 0 0.3672 1 1>}
  n27 = VectorParameter() {
   desc="This value is updated in C++\r\nWeapon Scope Attachment"
   group=C++ Parameters
   parameterName=ScopeTargetPointWorldSpace
   defaultValue=<rgba 0 1 0 0>
  }
  n28 = Add(
   A=FunctionInput(Preview=@n27.RGB) {bUsePreviewValueAsDefault=true, inputName=Scope Target Point WS, sortPriority=20}
   B=Multiply(A=ScalarParameter() {parameterName=ScopeTargetUpOffset}, B=NamedRerouteUsage())
  )
  n29 = TransformPosition(None=@n28) {transformSourceType=TRANSFORMPOSSOURCE_World, transformType=TRANSFORMPOSSOURCE_TranslatedWorld}
  n30 = Constant() {r=1}
  NamedRerouteDeclaration(None=TransformPosition(None=MaterialFunctionCall!(CameraRelativePosition=@n29, Scale, FirstPersonLerpAlpha=@n30) {materialFunction=/Game/Global/MaterialFunctions/MF_FirstPersonTransforms}.CameraRealtiveFP) {transformSourceType=TRANSFORMPOSSOURCE_TranslatedWorld, transformType=TRANSFORMPOSSOURCE_World}) {name=ScopeTarget Point Transformed, nodeColor=<rgba 0 1 0.8359 1>}
  n31 = Add(
   A=FunctionInput(Preview=VectorParameter() {desc="This value is updated in C++\r\nWeapon Scope Attachment", group=C++ Parameters, parameterName=ScopeEyePointWorldSpace}.RGB) {bUsePreviewValueAsDefault=true, inputName=Scope Eye Point WS, sortPriority=22}
   B=Multiply(A=ScalarParameter() {parameterName=ScopeEyeUpOffset}, B=NamedRerouteUsage())
  )
  n32 = MaterialFunctionCall(
   CameraRelativePosition=NamedRerouteDeclaration(None=TransformPosition(None=@n31) {transformSourceType=TRANSFORMPOSSOURCE_World, transformType=TRANSFORMPOSSOURCE_TranslatedWorld}) {name=EyeCamSpace_Pre, nodeColor=<rgba 0 0.7656 1 1>}
   FirstPersonLerpAlpha=Constant() {r=1}
  ) {materialFunction=&MF_FirstPersonTransforms}
  NamedRerouteDeclaration(None=TransformPosition(None=NamedRerouteDeclaration(None=@n32.CameraRealtiveFP) {name=EyeCamSpace_Post, nodeColor=<rgba 0 0.7656 1 1>}) {transformSourceType=TRANSFORMPOSSOURCE_TranslatedWorld, transformType=TRANSFORMPOSSOURCE_World}) {name=Scope Eye Point Transformed, nodeColor=<rgba 0 1 0.8359 1>}
  n33 = NamedRerouteDeclaration(None=@n28) {name=Scope Target Point WS, nodeColor=<rgba 1 0 0.375 1>}
  n34 = NamedRerouteDeclaration(None=@n31) {name=Scope Eye Point WS, nodeColor=<rgba 1 0 0.375 1>}
  n35 = NamedRerouteDeclaration(None=Normalize(VectorInput=Subtract(A=@n33, B=@n34))) {name=Scope Forward Vector, nodeColor=<rgba 0.5469 1 0 1>}
  NamedRerouteDeclaration(None=Add(A=Multiply(A=@n35, B=@n1), B=@n34)) {name=Scope Glass Location WS, nodeColor=<rgba 1 0 0.1875 1>}
  n36 = FunctionInput(Preview=StaticBool()) {bUsePreviewValueAsDefault=true, inputType=FunctionInput_StaticBool, sortPriority=10000}
  NamedRerouteDeclaration(None=StaticSwitch(
    True=Constant3Vector()
    False=CameraPositionWS() {desc=Ray Origin}
    Value=NamedRerouteDeclaration(None=@n36) {name=isCameraSpace, nodeColor=<rgba 0 0.9297 1 1>}
   )) {name=Camera Position, nodeColor=<rgba 1 0 0.1807 1>}
  MaterialExpressionMultiply_11 = Multiply(A=ComponentMask(None=NamedRerouteUsage()) {r=true}, B=Constant3Vector() {constant=<rgba 0 0 1 0>})
  MaterialExpressionNamedRerouteUsage_14 = NamedRerouteUsage()
  n37 = Custom(Radius=@MaterialExpressionMultiply_11, Center=@MaterialExpressionNamedRerouteUsage_14) {
   code="const float3 ScaleFP = mul((Center + Radius), (float3x3)View.FirstPersonTransform);\r\nreturn ScaleFP;"
   showCode=true
   inputs [
    {
     inputName=Radius
     input {expression=&self:MaterialExpressionMultiply_11}
    }
    {
     inputName=Center
     input {expression=&self:MaterialExpressionNamedRerouteUsage_14}
    }
   ]
  }
  NamedRerouteDeclaration(None=Distance(A=NamedRerouteUsage(), B=@n37)) {name=ScopeRadiusScaleTransformed, nodeColor=<rgba 1 0 0.0364 1>}
  NamedRerouteUsage()
  NamedRerouteUsage()
  n38 = TransformPosition() {transformSourceType=TRANSFORMPOSSOURCE_View, transformType=TRANSFORMPOSSOURCE_TranslatedWorld}
  n39 = Constant() {r=1}
  n40 = CameraPositionWS()
  Normalize(VectorInput=Subtract(
    A=MaterialFunctionCall(CameraRelativePosition=@n38, FirstPersonLerpAlpha=@n39) {materialFunction=&MF_FirstPersonTransforms}.CameraRealtiveFP
    B=TransformPosition(None=@n40) {transformSourceType=TRANSFORMPOSSOURCE_World, transformType=TRANSFORMPOSSOURCE_TranslatedWorld}
   ))
 ]}
