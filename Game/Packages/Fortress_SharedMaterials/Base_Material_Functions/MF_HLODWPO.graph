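# Material function that adds a world-collapse effect, sucking geometry toward CollapseCenter as SuckProgress rises, presumably for HLOD meshes.
# World position offset combines an exponential inward pull, radial downward drop and sine ripple, fading original offset near the center.
# IsTerrain limits vertical pull to near the center; IsOcean uses distance-masked pull and recolors base, emissive and specular.
# A custom node keeps original offset at zero progress; a global static switch bypasses everything.
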MaterialFunction /Game/Packages/Fortress_SharedMaterials/Base_Material_Functions/MF_HLODWPO
graph {nodes [
  n1 = ScalarParameter() {defaultValue=55477.6133, parameterName=WobbleRaidusFrequencey}
  n2 = VectorParameter() {parameterName=CollapseCenter, defaultValue=<rgba 66048 -42752 -716 1>}
  n3 = WorldPosition()
  n4 = Distance(A=@n2.RGB, B=@n3.XYZ)
  n5 = Divide(A=@n4, B=@n1)
  n6 = ScalarParameter() {defaultValue=4, parameterName=SingleWaveDistance}
  MaterialExpressionReroute_2 = Reroute(None=Saturate(None=ScalarParameter() {parameterName=SuckProgress}))
  n7 = LinearInterpolate(A=Constant() {r=-1}, B=@n6, Alpha=@MaterialExpressionReroute_2)
  n8 = Sine(None=Clamp(None=Subtract(A=@n5, B=@n7)))
  Add(A=@n8)
  n9 = ScalarParameter() {defaultValue=-5000, parameterName=SuckDownHeight}
  n10 = Constant3Vector() {constant=<rgba 1 1 0 0>}
  n11 = Constant3Vector() {constant=<rgba 1 1 1 0>}
  n12 = Saturate(None=Divide(A=@n4, B=ScalarParameter() {defaultValue=82302.2656, parameterName=SuckDownRadius}))
  n13 = Saturate(None=Power(Base=OneMinus(None=@n12), Exp=ScalarParameter() {defaultValue=5, parameterName=SuckDownExponent}))
  n14 = Constant()
  n15 = ScalarParameter() {defaultValue=7000000, parameterName=SuckRadius_Max}
  n16 = Saturate(None=Power(Base=@MaterialExpressionReroute_2, Exp=ScalarParameter() {defaultValue=8, parameterName=ExponentialTime}))
  n17 = Multiply(
   A=Normalize(VectorInput=Subtract(A=@n2.RGB, B=@n3.XYZ))
   B=Clamp(None=LinearInterpolate(A=@n14, B=@n15, Alpha=@n16), Max=@n4) {clampMode=CMODE_ClampMax}
  )
  n18 = StaticSwitch(
   True=Multiply(A=LinearInterpolate(A=@n10, B=@n11, Alpha=@n13), B=@n17)
   False=@n17
   Value=FunctionInput(Preview=StaticBool()) {bUsePreviewValueAsDefault=true, inputName=IsTerrain, inputType=FunctionInput_StaticBool}
  )
  n19 = Add(A=@n18, B=WorldPosition().XYZ)
  n20 = Saturate(None=Multiply(A=@MaterialExpressionReroute_2, B=ScalarParameter() {defaultValue=3, parameterName=SuckDownSpeedup}))
  n21 = LinearInterpolate(A=Constant(), B=Subtract(A=@n9, B=ComponentMask(None=@n19) {b=true}), Alpha=@n20)
  n22 = AppendVector(A=ComponentMask(None=@n2.RGB) {g=true, r=true}, B=@n9)
  Clamp(None=@n21, Max=Distance(A=@n19, B=@n22)) {clampMode=CMODE_ClampMax}
  n23 = Frac(None=Multiply(A=Time(), B=ScalarParameter() {defaultValue=0.05, parameterName=ColapseSpeed}))
  Divide(A=@n23)
  Frac(None=Multiply(A=Time(), B=ScalarParameter() {defaultValue=-0.4, parameterName=WobbleSpeed}))
  n24 = FunctionInput() {inputType=FunctionInput_MaterialAttributes}
  MaterialExpressionGetMaterialAttributes_0 = GetMaterialAttributes(None=@n24) {attributeGetTypes [
    #1
    #2
    #3
    #4
   ]}
  n25 = Constant()
  n26 = Distance(A=@n19, B=@n2.RGB)
  n27 = Saturate(None=Divide(A=@n26, B=ScalarParameter() {defaultValue=82302.2656, parameterName=SuckDownRadius}))
  n28 = Multiply(A=@n10, B=@n17)
  n29 = ComponentMask(None=@n2.RGB) {g=true, r=true}
  n30 = ComponentMask(None=ObjectPositionWS()) {b=true}
  n31 = Saturate(None=Power(Base=@MaterialExpressionReroute_2, Exp=ScalarParameter() {defaultValue=3, parameterName=ExponentialTimeOcean}))
  n32 = LinearInterpolate(A=Constant3Vector(), B=Subtract(A=AppendVector(A=@n29, B=@n30), B=@n3.XYZ), Alpha=@n31)
  n33 = ScalarParameter() {defaultValue=300000, parameterName=OceanWPOMaks_Offset}
  n34 = Saturate(None=Divide(A=Subtract(A=@n4, B=@n33), B=ScalarParameter() {defaultValue=2000000, parameterName=OceanWPOMaks_Fade}))
  n35 = ScalarParameter() {defaultValue=3568.3557, parameterName=ZWObble_Distance}
  n36 = Divide(A=@n4, B=Multiply(A=@n1, B=@n6))
  n37 = Multiply(A=@n35, B=OneMinus(None=Saturate(None=@n36)))
  n38 = Add(A=@n18, B=AppendVector(A=Constant2Vector(), B=Multiply(A=@n37, B=@n8)))
  n39 = Saturate(None=Divide(A=@n26, B=ScalarParameter() {defaultValue=82302.2656, parameterName=SuckDownRadius}))
  n40 = OneMinus(None=@n39)
  n41 = ScalarParameter() {defaultValue=5, parameterName=SuckDownExponent}
  n42 = AppendVector(A=Constant2Vector(), B=Multiply(A=Power(Base=@n40, Exp=@n41), B=@n21))
  n43 = FunctionInput(Preview=StaticBool()) {bUsePreviewValueAsDefault=true, inputName=IsOcean, inputType=FunctionInput_StaticBool}
  MaterialExpressionAdd_0 = Add(
   A=LinearInterpolate(A=Constant3Vector(), B=@MaterialExpressionGetMaterialAttributes_0.WorldPositionOffset, Alpha=LinearInterpolate(A=@n25, B=@n27, Alpha=@n20))
   B=StaticSwitch(True=LinearInterpolate(A=@n28, B=@n32, Alpha=@n34), False=Add(A=@n38, B=@n42), Value=@n43)
  )
  n44 = Custom!(WPO=@MaterialExpressionAdd_0, Progress=@MaterialExpressionReroute_2, OrgWPO=@MaterialExpressionGetMaterialAttributes_0.WorldPositionOffset) {
   code="float3 newWPO = {0,0,0};\r\n\r\nif (Progress <= 0)\r\n{\r\n\treturn OrgWPO;\r\n}\r\nelse\r\n{\r\n\treturn WPO;\r\n}\r\n"
   inputs [
    {inputName=WPO, input {expression=&self:MaterialExpressionAdd_0}}
    {
     inputName=Progress
     input {expression=&self:MaterialExpressionReroute_2}
    }
    {
     inputName=OrgWPO
     input {
      expression=&self:MaterialExpressionGetMaterialAttributes_0
      outputIndex=1
     }
    }
   ]
  }
  n45 = Saturate(None=Multiply(A=@n31, B=ScalarParameter() {defaultValue=3, parameterName=OceanCollpaseSpeedup}))
  n46 = LinearInterpolate(
   A=VectorParameter() {parameterName=OceanBlend, defaultValue=<rgba 0.5249 1.9842 3 1>}.RGB
   B=VectorParameter() {parameterName=OceanBlend2, defaultValue=<rgba 0 0 0 1>}.RGB
   Alpha=@n45
  )
  n47 = Add(
   A=Distance(A=WorldPosition() {worldPositionShaderOffset=WPT_ExcludeAllShaderOffsets}.XYZ, B=@n2.RGB)
   B=ScalarParameter() {defaultValue=-1000000, parameterName=OffsetOceanBlend}
  )
  n48 = ScalarParameter() {defaultValue=2001652, parameterName=FadeDistanceOceanBlend}
  n49 = LinearInterpolate(A=Constant(), B=Saturate(None=Divide(A=@n47, B=@n48)), Alpha=@n31)
  n50 = Constant()
  n51 = SetMaterialAttributes!(
   MaterialAttributes=@MaterialExpressionGetMaterialAttributes_0.MaterialAttributes
   WorldPositionOffset=@n44
   EmissiveColor=StaticSwitch(
    True=LinearInterpolate(A=@MaterialExpressionGetMaterialAttributes_0.EmissiveColor, B=@n46, Alpha=@n49)
    False=@MaterialExpressionGetMaterialAttributes_0.EmissiveColor
    Value=@n43
   )
   BaseColor=StaticSwitch(
    True=LinearInterpolate(A=@MaterialExpressionGetMaterialAttributes_0.BaseColor, B=@n46, Alpha=@n49)
    False=@MaterialExpressionGetMaterialAttributes_0.BaseColor
    Value=@n43
   )
   Specular=StaticSwitch(
    True=LinearInterpolate(A=@MaterialExpressionGetMaterialAttributes_0.Specular, B=@n50, Alpha=@n45)
    False=@MaterialExpressionGetMaterialAttributes_0.Specular
    Value=@n43
   )
  ) {attributeSetTypes [
    #1
    #2
    #3
    #4
   ]}
  FunctionOutput(None=StaticSwitchParameter(True=@n51, False=@n24) {parameterName=GlobalSwitchToEnableFunction})
  LinearInterpolate()
  Multiply(A=@n39, B=@n35)
  Multiply(A=@n23, B=ScalarParameter() {defaultValue=1, parameterName=PauseAdd}) {constB=2}
  Normalize(VectorInput=Subtract(A=@n19, B=@n22))
  ScalarParameter() {defaultValue=1, parameterName=SingleWaveAnim}
  ScalarParameter() {defaultValue=-62378.5156, parameterName=SuckDownDistance}
  WorldPosition()
 ]}
