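# Material function that adds a time-of-day character rim light to the emissive of incoming material attributes.
# Sunny and dark-side colours blend by atmospheric light direction against vertex normal.
# Named reroutes supply this colour, contrast-adjusted Fresnel, pre-skinned height gradient, AO, GMap and distance-blend inputs for shaping and boosting the rim.
# A global collection parameter, per-character multiplier and custom primitive data scale it, while ceiled base colour masks out pure black.
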MaterialFunction /Game/Packages/Fortress_SharedMaterials/Base_Material_Functions/MF_RimV3
graph {nodes [
  n1 = GetMaterialAttributes(None=FunctionInput() {inputName=Material, inputType=FunctionInput_MaterialAttributes}) {attributeGetTypes [#1, #2, #3]}
  n2 = NamedRerouteUsage()
  n3 = NamedRerouteUsage()
  n4 = Multiply(A=NamedRerouteUsage(), B=Multiply(A=NamedRerouteUsage(), B=Multiply(A=@n2, B=@n3)))
  n5 = ScalarParameter() {defaultValue=8, group=RimLight, parameterName=Distance Rim Boost}
  n6 = CollectionParameter() {
   collection=&FortniteMaterialParameters
   parameterName=bEnableCharacterRimLight
  }
  n7 = Multiply(A=LinearInterpolate(A=@n4, B=Multiply(A=@n4, B=@n5) {constB=3}, Alpha=NamedRerouteUsage()), B=@n6)
  n8 = ScalarParameter() {defaultValue=1, group=RimLight, parameterName=PerCharacterRimMultiplier}
  n9 = Add(
   A=@n1.EmissiveColor
   B=Multiply(A=ScalarParameter() {bUseCustomPrimitiveData=true, defaultValue=1, parameterName=PrimitiveDataRimisActive}, B=Multiply(A=@n7, B=@n8))
  )
  n10 = Saturate(None=Ceil(None=@n1.BaseColor))
  FunctionOutput(None=SetMaterialAttributes!(MaterialAttributes=@n1.MaterialAttributes, EmissiveColor=LinearInterpolate(A=@n1.EmissiveColor, B=@n9, Alpha=@n10)) {attributeSetTypes [#3]})
  n11 = VertexInterpolator(VS=PreSkinnedPosition())
  n12 = ComponentMask(None=Divide(A=FunctionInput(Preview=@n11.PS) {bUsePreviewValueAsDefault=true, inputName=Pre Skinned Local Position}, B=Constant() {r=120})) {b=true}
  n13 = Constant() {r=-0.45}
  NamedRerouteDeclaration(None=Saturate(None=Add(A=@n12, B=@n13))) {name=Z-Gradient, nodeColor=<rgba 0 1 0.7891 1>}
  NamedRerouteDeclaration(None=FunctionInput(Preview=Constant() {r=400}) {inputName=Distance Blend, inputType=FunctionInput_Scalar}) {name=DistanceBlend, nodeColor=<rgba 0.5469 0 1 1>}
  NamedRerouteDeclaration(None=FunctionInput() {inputName=Gmap, inputType=FunctionInput_Scalar}) {name=GMap, nodeColor=<rgba 0 0.4609 1 1>}
  n14 = Fresnel(ExponentIn=ScalarParameter() {defaultValue=3, group=RimLight, parameterName=Fresnel Rim Light}, Normal=Transform(None=NamedRerouteUsage())) {baseReflectFraction=0}
  n15 = Constant() {r=0.03}
  NamedRerouteDeclaration(None=Saturate(None=MaterialFunctionCall(In=@n14, Contrast=@n15) {materialFunction=&CheapContrast}.Result)) {name=RimLightFresnel, nodeColor=<rgba 1 0.375 0 1>}
  NamedRerouteDeclaration(None=FunctionInput() {inputName=AO, inputType=FunctionInput_Scalar}) {name=AO, nodeColor=<rgba 1 0 0.3984 1>}
  n16 = Constant3Vector() {constant=<rgba 0.1591 0.1267 0.3802 1>}
  n17 = Multiply(
   A=MaterialFunctionCall(
    Color_Morning=@n16
    Color_Day=Constant3Vector() {constant=<rgba 0.266 0.3806 0.4479 1>}
    Color_Evening=@n16
    Color_Night=Constant3Vector() {constant=<rgba 0.0218 0.1585 0.349 1>}
   ) {materialFunction=&Color_Time_of_Day}.Result
   B=ScalarParameter() {defaultValue=0.25, group=RimLight, parameterName=Darkside}
  ) {constB=0.2}
  n18 = Constant3Vector() {constant=<rgba 0.6146 0.292 0.149 1>}
  n19 = Multiply(
   A=MaterialFunctionCall(
    Color_Morning=@n18
    Color_Day=Constant3Vector() {constant=<rgba 0.599 0.3955 0.1809 1>}
    Color_Evening=@n18
    Color_Night=Constant3Vector() {constant=<rgba 0.1209 0.2902 0.3686 1>}
   ) {materialFunction=&Color_Time_of_Day}.Result
   B=ScalarParameter() {defaultValue=2, group=RimLight, parameterName=SunnySideIntensity}
  ) {constB=50}
  n20 = AtmosphericLightVector()
  n21 = VertexNormalWS()
  n22 = Saturate(None=Power(Base=DotProduct(A=@n20, B=@n21), Exp=Constant() {r=4}) {constExponent=6})
  NamedRerouteDeclaration(None=Multiply(A=ScalarParameter() {defaultValue=2, group=RimLight, parameterName=baseBrightness}, B=LinearInterpolate(A=@n17, B=@n19, Alpha=@n22))) {name=Rimlight Color, nodeColor=<rgba 1 0 0.6797 1>}
  NamedRerouteDeclaration(None=Normalize(VectorInput=@n1.Normal)) {name=Normal, nodeColor=<rgba 0.8047 1 0 1>}
 ]}
