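# MF_Rimlight_SingleLight is a material function that computes a single camera-relative directional light term for rim lighting.
# It builds a per-pixel basis from the camera-relative position and view up axis, then maps the normalized LightVector into it.
# The saturated dot product with the pixel normal is scaled by Gmap and BaseColor, and the transformed vector is also output.
# A vertex-normal clamp and a direct view transform are left unconnected.
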MaterialFunction /Game/Packages/Fortress_SharedMaterials/Base_Material_Functions/MF_Rimlight_SingleLight
graph {nodes [
  n1 = FunctionInput(Preview=Constant() {r=1}) {inputName=Gmap, inputType=FunctionInput_Scalar}
  n2 = FunctionInput() {inputName=LightVector, previewValue=<1 0 0 1>}
  n3 = Normalize(VectorInput=@n2)
  n4 = ComponentMask(None=@n3) {r=true}
  n5 = Constant3Vector() {constant=<rgba 0 1 0 0>}
  n6 = Normalize(VectorInput=WorldPosition() {worldPositionShaderOffset=WPT_CameraRelative}.XYZ)
  n7 = Normalize(VectorInput=CrossProduct(A=Transform(None=@n5) {transformSourceType=TRANSFORMSOURCE_View}, B=@n6))
  n8 = ComponentMask(None=@n3) {g=true}
  n9 = CrossProduct(A=@n6, B=@n7)
  n10 = Add(A=Add(A=Multiply(A=@n4, B=@n7), B=Multiply(A=@n8, B=@n9)), B=Multiply(A=ComponentMask(None=@n3) {b=true}, B=@n6))
  n11 = Saturate(None=DotProduct(A=@n10, B=PixelNormalWS()))
  n12 = Constant() {r=1}
  FunctionOutput(None=Multiply(A=Multiply(A=@n11, B=@n1), B=FunctionInput(Preview=@n12) {inputName=BaseColor}))
  FunctionOutput(None=@n10) {outputName=NormalizedAndTransformedVector, sortPriority=10}
  n13 = VertexNormalWS()
  Min(A=Saturate(None=DotProduct(A=@n13, B=@n10)), B=@n11)
  Transform(None=Normalize(VectorInput=@n2)) {transformSourceType=TRANSFORMSOURCE_View}
 ]}
