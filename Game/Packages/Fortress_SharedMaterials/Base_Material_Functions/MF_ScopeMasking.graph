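# MF_ScopeMasking is a material function that cuts a scope-view hole in materials through OpacityMask.
# GlassLocation, adjusted by first-person transforms, centres a virtual plane whose UVs drive a texture or radial mask.
# View-angle, distance and local-Y side tests limit the cut, and ShadowReplace keeps shadows unmasked.
# Static switches enable masking or an emissive debug view; unconnected cone and radius experiments remain.
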MaterialFunction /Game/Packages/Fortress_SharedMaterials/Base_Material_Functions/MF_ScopeMasking
graph {nodes [
  CollectionParameter() {collection=&FortniteMaterialParameters, parameterName=FP_Active}
  n1 = SetMaterialAttributes!(MaterialAttributes, EmissiveColor=AppendVector(A=NamedRerouteUsage(), B=NamedRerouteUsage())) {attributeSetTypes [#1]}
  n2 = FunctionInput() {bUsePreviewValueAsDefault=true, inputType=FunctionInput_MaterialAttributes}
  n3 = GetMaterialAttributes(None=@n2) {attributeGetTypes [#2]}
  n4 = Transform(None=Constant3Vector() {constant=<rgba 0 1 0 0>}) {transformSourceType=TRANSFORMSOURCE_Local}
  n5 = VectorParameter() {group=Scope Masking, parameterName=GlassLocation, defaultValue=<rgba 0 0 0 1>}
  n6 = TransformPosition(None=@n5.RGB) {transformType=TRANSFORMPOSSOURCE_TranslatedWorld}
  n7 = Normalize(VectorInput=@n6)
  n8 = Sign(None=Subtract(A=DotProduct(A=@n4, B=@n7), B=ScalarParameter() {defaultValue=0.5, group=Scope Masking, parameterName=MaskAngleThreshold}))
  n9 = VectorParameter() {group=Scope Masking, parameterName=MaskingPlaneNormal, defaultValue=<rgba 0 -1 0 1>}
  n10 = ScalarParameter() {defaultValue=1, group=Scope Masking, parameterName=ScopeMaskRadius}
  n11 = MaterialFunctionCall!(CameraRelativePosition=@n6, Scale=AppendVector(A=@n10, B=Constant2Vector()), FirstPersonLerpAlpha=Constant() {r=1}) {materialFunction=/Game/Global/MaterialFunctions/MF_FirstPersonTransforms}
  n12 = TransformPosition(None=@n11.CameraRealtiveFP) {transformSourceType=TRANSFORMPOSSOURCE_TranslatedWorld, transformType=TRANSFORMPOSSOURCE_World}
  n13 = If(
   A=Distance(A=@n12, B=TransformPosition(None=@n5.RGB) {transformType=TRANSFORMPOSSOURCE_World})
   "A > B"=ComponentMask(None=@n11.Scale) {r=true}
   "A == B"=@n10
   "A < B"=@n10
  ) {constB=1}
  n14 = MaterialFunctionCall(
   Plane Normal Axis=Transform(None=Normalize(VectorInput=@n9.RGB)) {transformSourceType=TRANSFORMSOURCE_Local}
   Plane X Axis=Transform(None=Constant3Vector() {constant=<rgba 1 0 0 0>}) {transformSourceType=TRANSFORMSOURCE_Instance}
   Plane Center=@n12
   UV Scale=@n13
  ) {materialFunction=&VirtualPlaneCoordinates}
  n15 = Multiply(A=@n13) {constB=3}
  n16 = TextureSampleParameter2D(UVs=NamedRerouteDeclaration(None=Saturate(None=@n14.UVs)) {name=UVs, nodeColor=<rgba 0 1 0.1328 1>}) {
   group=Scope Masking Texture
   parameterName=Mask Texture
   samplerType=SAMPLERTYPE_Masks
   texture=/Game/Weapons/FORT_Rifles/SniperScope/T_HolographicMask
  }
  n17 = SphereMask(
   A=ChannelMaskParameter(None=@n16.RGBA) {group=Scope Masking Texture, parameterName=Mask Texture Channel}.R
   B=Constant() {r=1}
   Radius=ScalarParameter() {defaultValue=0.001, group=Scope Masking Texture, parameterName=Mask Texture Threshold}
  ) {attenuationRadius=0.001}
  n18 = SphereMask(A=@n14.UVs, B=Constant2Vector() {g=0.5, r=0.5}, Radius=Constant() {r=1}) {attenuationRadius=1.2148}
  n19 = VertexInterpolator(VS=LocalPosition() {localOrigin=InstancePreSkinning}.XYZ)
  n20 = Sign(None=ComponentMask(None=Subtract(A=@n19.PS, B=@n5.RGB)) {g=true})
  n21 = Multiply(
   A=Multiply(A=Saturate(None=@n8), B=SphereMask(A=@n14.World Position, B=@n12, Radius=@n15) {attenuationRadius=3}) {constA=1}
   B=Multiply(A=StaticSwitchParameter(True=@n17, False=@n18) {group=Scope Masking, parameterName=UseTextureMasking, sortPriority=2}, B=Saturate(None=@n20))
  )
  n22 = OneMinus(None=@n21)
  n23 = Constant() {r=1}
  n24 = SetMaterialAttributes!(
   MaterialAttributes=@n3.MaterialAttributes
   OpacityMask=Min(A=@n3.OpacityMask, B=ShadowReplace(Default=@n22, Shadow=@n23) {desc=Bypass masking for shadows})
  ) {attributeSetTypes [#2]}
  FunctionOutput(None=StaticSwitchParameter(True=StaticSwitchParameter(True=@n1, False=@n24) {group=Scope Masking, parameterName=UseScopeDebug, sortPriority=1000}, False=@n2) {group=Scope Masking, parameterName=UseScopeMasking, sortPriority=1})
  MaterialExpressionConstant3Vector_6 = Constant3Vector() {constant=<rgba 0 0 1 0>}
  MaterialExpressionConstant_5 = Constant() {r=1}
  n25 = Custom!(TranslatedWorldPosition=@MaterialExpressionConstant3Vector_6, FirstPersonLerpAlpha=@MaterialExpressionConstant_5) {
   code="const float3 FirstPersonPosition = mul(TranslatedWorldPosition, (float3x3)View.FirstPersonTransform);\r\nreturn lerp(TranslatedWorldPosition, FirstPersonPosition , FirstPersonLerpAlpha);"
   showCode=true
   inputs [
    {
     inputName=TranslatedWorldPosition
     input {
      expression=&self:MaterialExpressionConstant3Vector_6
      mask=1
      maskB=1
      maskG=1
      maskR=1
     }
    }
    {
     inputName=FirstPersonLerpAlpha
     input {expression=&self:MaterialExpressionConstant_5}
    }
   ]
  }
  Length(None=TransformPosition(None=@n25) {transformSourceType=TRANSFORMPOSSOURCE_TranslatedWorld, transformType=TRANSFORMPOSSOURCE_World}) {is=class MaterialExpressionLength}
  MaterialFunctionCall!(VectorToTransform, BasisX, BasisY, BasisZ) {materialFunction=/Engine/Functions/Engine_MaterialFunctions02/Math/InverseTransformMatrix}
  MaterialFunctionCall() {materialFunction=&InverseTransformMatrix}
  NamedRerouteDeclaration(None=@n21) {name=Mask, nodeColor=<rgba 0 1 0.0859 1>}
  n26 = Divide(A=Distance(A=@n14.World Position, B=@n12), B=ScalarParameter() {defaultValue=1.3, group=Scope Masking, parameterName=ScopeMaskRadiusTest2})
  OneMinus(None=Saturate(None=Floor(None=@n26)))
  n27 = Subtract(A=VertexInterpolator(VS=LocalPosition() {localOrigin=InstancePreSkinning}.XYZ).PS, B=@n5.RGB)
  n28 = Subtract(A=Add(A=@n5.RGB, B=Constant3Vector() {constant=<rgba 0 15 0 0>}), B=@n5.RGB)
  n29 = DotProduct(A=@n27, B=@n28)
  Power(Base=@n29, Exp=ScalarParameter() {defaultValue=0.704, group=Scope Masking, parameterName=ScopeMaskRadiusConeFalloff})
  n30 = ScalarParameter() {defaultValue=1.2837, group=Scope Masking, parameterName=ScopeMaskRadiusTest}
  n31 = LinearInterpolate(
   A=@n30
   B=Add(A=@n30, B=ScalarParameter() {defaultValue=0.4934, group=Scope Masking, parameterName=ScopeMaskRadiusConeAdd})
   Alpha=Saturate(None=@n29)
  )
  n32 = Divide(A=Length(None=CrossProduct(A=@n27, B=@n28)), B=Max(A=Length(None=@n28)) {constB=0.001})
  Saturate(None=Floor(None=Divide(A=@n32, B=@n31)))
  ScalarParameter() {defaultValue=1, group=Scope Masking, parameterName=ScopeMaskRadius}
  ScalarParameter() {defaultValue=1, group=Scope Masking, parameterName=ScopeMaskRadius}
  TransformPosition(None=@n14.World Position) {transformSourceType=TRANSFORMPOSSOURCE_World, transformType=TRANSFORMPOSSOURCE_TranslatedWorld}
  Transform(None=Constant3Vector() {constant=<rgba 1 0 0 0>}) {transformSourceType=TRANSFORMSOURCE_Camera, transformType=TRANSFORM_Local}
 ]}
