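# Material function that adds a colored team outline glow to the emissive channel of the input material attributes.
# A Fresnel term from the camera vector and world-transformed material normal forms a hard rim band plus a soft gradient.
# Camera-to-object distance over a threshold blends close and distant falloff and glow strength.
# The glow is tinted by team color and scaled by an enable parameter; an inside-mask branch is left unconnected.
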MaterialFunction /Game/Packages/Fortress_SharedMaterials/Base_Material_Functions/MF_TeamOutline
graph {nodes [
  n1 = GetMaterialAttributes(None=FunctionInput() {inputName=Material, inputType=FunctionInput_MaterialAttributes}) {attributeGetTypes [#1, #2]}
  n2 = Transform(None=@n1.Normal)
  Add(A=VertexNormalWS(), B=@n2)
  n3 = DotProduct(A=CameraVectorWS(), B=@n2)
  n4 = CameraPositionWS()
  n5 = ObjectPositionWS()
  n6 = Clamp(None=Divide(A=Distance(A=@n4, B=@n5), B=ScalarParameter() {defaultValue=5000, group=TeamOutline, parameterName=TeamOutlineDistanceThreshold}))
  n7 = LinearInterpolate(
   A=ScalarParameter() {defaultValue=0.5, group=TeamOutline, parameterName=CloseTeamOutlineBaseFresnelFalloff}
   B=ScalarParameter() {defaultValue=0.6, group=TeamOutline, parameterName=DistantTeamOutlineBaseFresnelFalloff}
   Alpha=@n6
  )
  n8 = Add(A=ScalarParameter() {defaultValue=0.1, parameterName=TeamOutlineInsideMaskFalloff}, B=@n7)
  Ceil(None=Clamp(None=Add(A=@n3, B=@n8)))
  n9 = OneMinus(None=@n3)
  n10 = Clamp(None=Subtract(A=@n9, B=@n7))
  n11 = Subtract(
   A=@n9
   B=LinearInterpolate(
    A=ScalarParameter() {defaultValue=0.5, group=TeamOutline, parameterName=CloseTeamOutlineFresnelFalloff}
    B=ScalarParameter() {group=TeamOutline, parameterName=DistantTeamOutlineFresnelFalloff}
    Alpha=@n6
   )
  )
  n12 = Multiply(
   A=Add(A=Ceil(None=@n10), B=Clamp(None=@n11))
   B=LinearInterpolate(
    A=ScalarParameter() {defaultValue=1, group=TeamOutline, parameterName=CloseTeamOutlineGlowStrength}
    B=ScalarParameter() {defaultValue=1.5, group=TeamOutline, parameterName=DistantTeamOutlineGlowStrength}
    Alpha=@n6
   )
  )
  n13 = Multiply(
   A=Multiply(A=Power(Base=@n12), B=VectorParameter() {group=TeamOutline, parameterName=TeamOutlineColor, defaultValue=<rgba 1 0.0091 0 1>}.RGB)
   B=ScalarParameter() {group=TeamOutline, parameterName=EnableTeamOutline}
  )
  FunctionOutput(None=SetMaterialAttributes!(MaterialAttributes=@n1.MaterialAttributes, EmissiveColor=Add(A=@n1.EmissiveColor, B=@n13)) {attributeSetTypes [#1]})
 ]}
