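# Material function that supplies the blue base look for the vehicle derez (death) transition, output as material attributes and separate emissive.
# A Fresnel-default rim input drives a glossy, black-based surface's cyan glow, reusing normals from the incoming base material.
# Epic quality reshapes the rim and adds time-animated grid squares, projected triplanar in local position and blended by world vertex normal.
# A DeRezPulseA timestamp makes emissive start boosted, then ease down.
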MaterialFunction /Game/Packages/Fortress_SharedMaterials/Base_Material_Functions/MF_VehicleDerez_BaseBlue
graph {
 properties {description=Base Colors for the death Transition}
 nodes [
  n1 = Constant()
  n2 = GetMaterialAttributes(None=FunctionInput(Preview=SetMaterialAttributes!(MaterialAttributes)) {bUsePreviewValueAsDefault=true, inputName=BaseMaterial, inputType=FunctionInput_MaterialAttributes}) {attributeGetTypes [#1, #2]}
  n3 = FunctionInput(Preview=Fresnel() {baseReflectFraction=0, exponent=1}) {bUsePreviewValueAsDefault=true, inputName=Rim, inputType=FunctionInput_Scalar}
  n4 = Clamp(None=@n3)
  n5 = Power(Base=OneMinus(None=@n4)) {constExponent=4}
  n6 = LinearInterpolate(Alpha=Power(Base=@n4)) {constA=0.3, constB=0.9}
  n7 = QualitySwitch(Default=@n3, Epic=Multiply(A=Add(A=@n5, B=@n6)) {constB=0.6})
  n8 = Add(
   A=Constant3Vector() {constant=<rgba 0.0583 0.3109 0.4 1>}
   B=Multiply(A=Constant3Vector() {constant=<rgba 0 3.8199 6 1>}, B=Power(Base=@n7) {constExponent=3})
  )
  n9 = VectorParameter() {group=Derez, parameterName=DeRezSquare_TextureSize, defaultValue=<rgba 100 100 100 0>}
  n10 = Divide(
   A=MaterialFunctionCall() {materialFunction=&LocalPosition}.Local Position
   B=Multiply(A=Abs(None=@n9.RGB)) {constB=-1}
  )
  n11 = TextureSample(UVs=ComponentMask(None=@n10) {b=true, r=true}) {
   samplerSource=SSM_Wrap_WorldGroupSettings
   samplerType=SAMPLERTYPE_LinearColor
   texture=/Game/Packages/Fortress_SharedMaterials/Textures/T_Slowfield_GridSquares_01
  }
  n12 = TextureSample(UVs=ComponentMask(None=@n10) {b=true, g=true}) {
   samplerSource=SSM_Wrap_WorldGroupSettings
   samplerType=SAMPLERTYPE_LinearColor
   texture=&T_Slowfield_GridSquares_01
  }
  n13 = VertexNormalWS()
  n14 = Constant() {r=1}
  n15 = MaterialFunctionCall(In=Abs(None=ComponentMask(None=@n13) {r=true}), Contrast=@n14) {materialFunction=&CheapContrast}
  n16 = TextureSample(UVs=ComponentMask(None=@n10) {g=true, r=true}) {
   samplerSource=SSM_Wrap_WorldGroupSettings
   samplerType=SAMPLERTYPE_LinearColor
   texture=&T_Slowfield_GridSquares_01
  }
  n17 = Abs(None=ComponentMask(None=@n13) {b=true})
  n18 = LinearInterpolate(
   A=LinearInterpolate(A=@n11.RGB, B=@n12.RGB, Alpha=ComponentMask(None=@n15.Result) {r=true})
   B=@n16.RGB
   Alpha=ComponentMask(None=MaterialFunctionCall(In=@n17, Contrast=@n14) {materialFunction=&CheapContrast}.Result) {r=true}
  )
  n19 = Sine(None=Add(A=@n18, B=Time()))
  n20 = Multiply(A=Power(Base=@n8), B=Multiply(A=ConstantBiasScale(None=@n19), B=@n18))
  n21 = Multiply(A=@n8) {constB=0.3}
  n22 = ScalarParameter() {parameterName=DeRezPulseA}
  n23 = ScalarParameter() {defaultValue=0.5, parameterName=TransitionLength_In}
  n24 = Time()
  n25 = StaticBool() {value=true}
  n26 = Multiply(
   A=QualitySwitch(Default=@n8, Epic=Add(A=@n20, B=@n21))
   B=LinearInterpolate(Alpha=MaterialFunctionCall(TimestampA=@n22, TransitionLength_In=@n23, Time=@n24, Transition In Only?=@n25) {materialFunction=&MF_Timestamp_InOut_01}.SmoothStep) {constA=10}
  )
  n27 = SetMaterialAttributes!(
   MaterialAttributes
   BaseColor=@n1
   Metallic=@n1
   Specular=Constant() {r=1}
   Roughness=Constant() {r=0.2}
   EmissiveColor=@n26
   Opacity
   OpacityMask
   Anisotropy
   Normal=@n2.Normal
   WorldPositionOffset=@n2.Normal
   WorldDisplacement
   TessellationMultiplier
   SubsurfaceColor=Constant()
   CustomData0=Constant()
  ) {attributeSetTypes [
    #3
    #4
    #5
    #6
    #7
    #8
    #9
    #10
    #1
    #2
    #11
    #12
    #13
    #14
   ]}
  FunctionOutput(None=@n27)
  FunctionOutput(None=@n26) {outputName=Emissive, sortPriority=1}
 ]
}
