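# Material function that applies a vehicle de-rez transition, sweeping a blue derez look from MF_VehicleDerez_BaseBlue over the input material attributes.
# The sweep follows local X scaled by CarLength and is driven by the NewDissolveGradient custom primitive data.
# World-aligned Perlin noise in local position breaks up the blend edge and shapes an emissive hot band at the front.
# Mobile replaces the noise with a constant.
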MaterialFunction /Game/Packages/Fortress_SharedMaterials/Base_Material_Functions/MF_VehicleDeRez_Transition
graph {nodes [
  n1 = FunctionInput() {inputType=FunctionInput_MaterialAttributes}
  n2 = PixelNormalWS()
  n3 = MaterialFunctionCall!(
   BaseMaterial=@n1
   Rim=FunctionInput(Preview=Fresnel(Normal=@n2) {baseReflectFraction=0, exponent=1}) {bUsePreviewValueAsDefault=true, inputName=Fresnel, inputType=FunctionInput_Scalar}
  ) {materialFunction=&/MF_VehicleDerez_BaseBlue}
  n4 = OneMinus(None=ComponentMask(None=MaterialFunctionCall() {materialFunction=&LocalPosition}.Local Position) {r=true})
  n5 = ScalarParameter() {defaultValue=750, parameterName=CarLength}
  n6 = Multiply(A=@n5) {constB=0.5}
  n7 = Clamp(None=OneMinus(None=ScalarParameter() {bUseCustomPrimitiveData=true, parameterName=NewDissolveGradient, primitiveDataIndex=13}))
  n8 = Saturate(None=Subtract(A=Add(A=@n4, B=@n6), B=Multiply(A=@n5, B=@n7)))
  n9 = Multiply(A=@n5, B=@n7)
  n10 = ShadingPathSwitch(
   Default=MaterialFunctionCall!(
    TextureObject=TextureObject() {samplerType=SAMPLERTYPE_LinearGrayscale, texture=&T_SimplePerlin_Greyscale}
    TextureSize=VectorParameter() {group=Derez, parameterName=HotEdge_TextureSize, defaultValue=<rgba 150 150 150 0>}.RGB
    WorldPosition=MaterialFunctionCall() {materialFunction=&LocalPosition}.Local Position
    Export Float 4
    In=Constant2Vector() {r=-0.5}
    World Space Normal
    ProjectionTransitionContrast
   ) {materialFunction=/Game/Athena/SafeZone/WorldAlignedTexture_ScalePan}.XYZ Texture
   Mobile=Constant() {r=1}
  )
  n11 = Multiply(A=ScalarParameter() {defaultValue=74, group=Derez, parameterName=Hot Edge Offset}, B=@n10)
  n12 = Saturate(None=Subtract(A=Add(A=@n4, B=@n6), B=Add(A=@n9, B=@n11)))
  n13 = Add(
   A=@n3.Emissive
   B=Multiply(A=VectorParameter() {parameterName=HotSweepEdgeColor, defaultValue=<rgba 0 20 30 1>}.RGB, B=Subtract(A=@n8, B=@n12))
  ) {constB=0}
  n14 = ComponentMask(None=MaterialFunctionCall() {materialFunction=&LocalPosition}.Local Position) {r=true}
  n15 = ScalarParameter() {defaultValue=750, parameterName=CarLength}
  n16 = Clamp(None=OneMinus(None=ScalarParameter() {bUseCustomPrimitiveData=true, parameterName=NewDissolveGradient, primitiveDataIndex=13}))
  n17 = ScalarParameter() {defaultValue=25, group=Derez, parameterName=Raw Edge Offset}
  n18 = Subtract(A=Add(A=OneMinus(None=@n14), B=Multiply(A=@n15) {constB=0.5}), B=Add(A=Multiply(A=@n15, B=@n16), B=Multiply(A=@n17, B=@n10)))
  FunctionOutput(None=BlendMaterialAttributes(
    A=@n1
    B=SetMaterialAttributes!(MaterialAttributes=@n3.Result, EmissiveColor=@n13) {attributeSetTypes [#1]}
    Alpha=Saturate(None=@n18)
   ))
 ]}
