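# Material function that rotates a vector, the camera vector by default, by the offset between the skinned world vertex normal and the pre-skinned local normal.
# The normals' unnormalized cross product serves as the axis, splitting the vector into parallel and perpendicular parts.
# A Rodrigues-style blend uses their dot product as cosine and the square root of saturated one-minus-dot as sine term, then normalizes.
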MaterialFunction /Game/Packages/Fortress_SharedMaterials/Base_Material_Functions/RotateVector_FromSkinning
graph {nodes [
  n1 = FunctionInput(Preview=VertexNormalWS()) {bUsePreviewValueAsDefault=true, inputName=VertexlNormalWS, sortPriority=2}
  n2 = FunctionInput(Preview=VertexInterpolator(VS=PreSkinnedNormal()).PS) {bUsePreviewValueAsDefault=true, inputName=PreSkinnedLocalNormal, sortPriority=1}
  n3 = CrossProduct(A=@n1, B=@n2)
  n4 = FunctionInput(Preview=CameraVectorWS()) {bUsePreviewValueAsDefault=true, inputName=VectorToRotate}
  n5 = Multiply(A=DotProduct(A=@n4, B=@n3), B=@n3)
  n6 = Subtract(A=@n4, B=@n5)
  n7 = DotProduct(A=@n2, B=@n1)
  n8 = Saturate(None=OneMinus(None=Multiply(A=@n7)))
  n9 = Add(A=Multiply(A=@n6, B=@n7), B=Multiply(A=CrossProduct(A=@n3, B=@n6), B=SquareRoot(None=@n8)))
  FunctionOutput(None=Normalize(VectorInput=Add(A=@n5, B=@n9))) {outputName=RotatedVectorFromSkinning}
 ]}
