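# Shared Fortnite material function that samples a packed SpecularMasks texture parameter and splits it into Spec, Metallic and Spec Roughness outputs.
# Red passes through as Spec, green is multiplied by a Metallic scalar, and blue interpolates between roughness minimum and maximum values.
# UVs and scalar inputs default to the texture coordinate and Specular-group parameters, and the sampler uses shared world-group wrap settings.
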MaterialFunction /Game/Packages/Fortress_SharedMaterials/Base_Material_Functions/SpecularMasks
graph {
 properties {bExposeToLibrary=true, description=Samples the specular mask and seperates it's components, libraryCategoriesText [Misc,Fortnite]}
 nodes [
  n1 = TextureSampleParameter2D(UVs=FunctionInput(Preview=TextureCoordinate()) {bUsePreviewValueAsDefault=true, inputName=UVs, inputType=FunctionInput_Vector2}) {
   group=Specular
   parameterName=SpecularMasks
   samplerSource=SSM_Wrap_WorldGroupSettings
   samplerType=SAMPLERTYPE_Masks
   texture=&T_Fortnite_Default_S
  }
  FunctionOutput(None=@n1.R) {outputName=Spec, sortPriority=1}
  n2 = ScalarParameter() {defaultValue=0.1, group=Specular, parameterName=SpecRoughnessMin}
  n3 = ScalarParameter() {defaultValue=1, group=Specular, parameterName=SpecRoughnessMax}
  FunctionOutput(None=LinearInterpolate(
    A=FunctionInput(Preview=@n2) {bUsePreviewValueAsDefault=true, inputName=SpecRoughnessMin, inputType=FunctionInput_Scalar, sortPriority=1}
    B=FunctionInput(Preview=@n3) {bUsePreviewValueAsDefault=true, inputName=SpecRoughnessMax, inputType=FunctionInput_Scalar, sortPriority=2}
    Alpha=@n1.B
   )) {outputName=Spec Roughness, sortPriority=2}
  n4 = ScalarParameter() {defaultValue=1, group=Specular, parameterName=Metallic}
  FunctionOutput(None=Multiply(A=@n1.G, B=FunctionInput(Preview=@n4) {bUsePreviewValueAsDefault=true, inputName=Metallic, inputType=FunctionInput_Scalar, sortPriority=3})) {outputName=Metallic}
 ]
}
