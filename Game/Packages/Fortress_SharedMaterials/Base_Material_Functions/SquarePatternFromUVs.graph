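# SquarePatternFromUVs is a library material function that creates a centered square from input UVs for masking or animation.
# It offsets both the UVs and their inverse by Size times SizeScale, then feeds each through LinearGradient.
# Rounding turns the four U and V gradients into hard edge thresholds.
# Summing these, subtracting three and saturating acts as a branchless AND, leaving only the square where all four overlap.
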MaterialFunction /Game/Packages/Fortress_SharedMaterials/Base_Material_Functions/SquarePatternFromUVs
graph {
 properties {
  bExposeToLibrary=true
  description=Creates a centered square from UV coordinates to mask or animate.
  libraryCategoriesText [Misc,Coordinates]
 }
 nodes [
  n1 = FunctionInput(Preview=TextureCoordinate()) {description=Input UVs, inputName=UVs, inputType=FunctionInput_Vector2}
  n2 = Multiply(
   A=FunctionInput(Preview=Constant() {r=0.25}) {bUsePreviewValueAsDefault=true, inputName=Size, inputType=FunctionInput_Scalar, sortPriority=1}
   B=FunctionInput(Preview=Constant() {r=1}) {bUsePreviewValueAsDefault=true, inputName=SizeScale, inputType=FunctionInput_Scalar, sortPriority=2}
  )
  n3 = MaterialFunctionCall(UV Channel=Add(A=@n1, B=@n2) {constB=0}) {materialFunction=&LinearGradient}
  n4 = MaterialFunctionCall(UV Channel=Add(A=OneMinus(None=@n1), B=@n2) {constB=0}) {materialFunction=&LinearGradient}
  n5 = Add(A=Add(A=Round(None=@n3.UGradient), B=Round(None=@n4.UGradient)), B=Add(A=Round(None=@n3.VGradient), B=Round(None=@n4.VGradient)))
  FunctionOutput(None=Saturate(None=Subtract(A=@n5) {constB=3})) {outputName=PatternOutput}
 ]
}
