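# Material function that outputs world-space planar UVs and selection masks for blending or masking textures in a triplanar, world-aligned projection.
# Front, Side and Top UVs take the XZ, YZ and XY of world position, divided by the negated absolute size set for each plane.
# Matching masks pass absolute normal Y, X and Z through CheapContrast with a shared transition contrast, without normalizing their sum.
# Position defaults to world position without shader offsets, and normal to the vertex normal.
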MaterialFunction /Game/Packages/Fortress_SharedMaterials/Base_Material_Functions/WorldAlignedTexture_UVsMasks
graph {
 properties {
  description="Outputs world space UVs and UV selection masks, to blend or masks textures in world space"
  libraryCategoriesText [Coordinates,Texturing,Misc]
 }
 nodes [
  n1 = FunctionInput(Preview=WorldPosition() {worldPositionShaderOffset=WPT_ExcludeAllShaderOffsets}.XYZ) {bUsePreviewValueAsDefault=true, inputName=WorldPosition}
  n2 = FunctionInput() {
   bUsePreviewValueAsDefault=true
   description=Size of texture in worldspace units
   inputName=UVSize_Top
   inputType=FunctionInput_Vector2
   sortPriority=4
   previewValue=<64 64 1 1>
  }
  n3 = Abs(None=@n2)
  FunctionOutput(None=Divide(A=ComponentMask(None=@n1) {g=true, r=true}, B=Multiply(A=@n3) {constB=-1})) {outputName=UV_Top, sortPriority=3}
  n4 = FunctionInput(Preview=VertexNormalWS()) {bUsePreviewValueAsDefault=true, inputName=World Space Normal, sortPriority=1, previewValue=<0 0 1 1>}
  n5 = Abs(None=ComponentMask(None=@n4) {r=true})
  n6 = FunctionInput() {
   bUsePreviewValueAsDefault=true
   inputName=ProjectionTransitionContrast
   inputType=FunctionInput_Scalar
   sortPriority=5
   previewValue=<1 0 0 1>
  }
  FunctionOutput(None=ComponentMask(None=MaterialFunctionCall(In=@n5, Contrast=@n6) {materialFunction=&CheapContrast}.Result) {r=true}) {outputName=Mask_Side, sortPriority=5}
  n7 = FunctionInput() {
   bUsePreviewValueAsDefault=true
   description=Size of texture in worldspace units
   inputName=UVSize_Front
   inputType=FunctionInput_Vector2
   sortPriority=2
   previewValue=<64 64 1 1>
  }
  n8 = Abs(None=@n7)
  FunctionOutput(None=Divide(A=ComponentMask(None=@n1) {b=true, r=true}, B=Multiply(A=@n8) {constB=-1})) {outputName=UV_Front, sortPriority=1}
  n9 = FunctionInput() {
   bUsePreviewValueAsDefault=true
   description=Size of texture in worldspace units
   inputName=UVSize_Side
   inputType=FunctionInput_Vector2
   sortPriority=3
   previewValue=<64 64 1 1>
  }
  n10 = Abs(None=@n9)
  FunctionOutput(None=Divide(A=ComponentMask(None=@n1) {b=true, g=true}, B=Multiply(A=@n10) {constB=-1})) {outputName=UV_Side, sortPriority=2}
  n11 = Abs(None=ComponentMask(None=@n4) {b=true})
  FunctionOutput(None=ComponentMask(None=MaterialFunctionCall(In=@n11, Contrast=@n6) {materialFunction=&CheapContrast}.Result) {r=true}) {outputName=Mask_Top, sortPriority=6}
  n12 = Abs(None=ComponentMask(None=@n4) {g=true})
  FunctionOutput(None=ComponentMask(None=MaterialFunctionCall(In=@n12, Contrast=@n6) {materialFunction=&CheapContrast}.Result) {r=true}) {outputName=Mask_Front, sortPriority=4}
 ]
}
